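# Physics asset for the lower bandolier straps on the large male ToughPack body cosmetic, with capsules on strap bones, pelvis and thighs.
# Five-link front and back chains and short mid links hang from base bones constrained to the pelvis, except left mid, anchored to the left thigh.
# Left-side strap joints have tight swing limits, while right-side joints rotate freely.
# Locked-rotation constraints tie left front and back chain ends to left and right mid links respectively.
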PhysicsAsset /BRCosmetics/Characters/Player/Male/Large/Bodies/M_LRG_ToughPack/Meshes/M_LRG_ToughPack_Bandoliers_Physics
bodies {
 base_bandoliersBotBack_1_l [{shapeType=Capsule, length=1, radius=0.5, shapeName=base_bandoliersBotBack_1_l_capsule}]
 base_bandoliersBotBack_1_r [{shapeType=Capsule, length=1, radius=0.5, shapeName=base_bandoliersBotBack_1_l_capsule}]
 base_bandoliersBotFront_1_l [{shapeType=Capsule, length=1, radius=0.5, shapeName=base_bandoliersBotFront_1_l_capsule}]
 base_bandoliersBotFront_1_r [{shapeType=Capsule, length=0.505, radius=0.505, shapeName=base_bandoliersBotFront_1_r_capsule, rotation=<rot 0 -90 0>}]
 base_bandoliersBotMid_1_l [{shapeType=Capsule, length=1, radius=0.5, shapeName=base_bandoliersBotFront_1_l_capsule}]
 base_bandoliersBotMid_1_r [{shapeType=Capsule, length=0.505, radius=0.505, shapeName=base_bandoliersBotMid_1_r_capsule, rotation=<rot 0 -90 0>}]
 dyn_bandoliersBotBack_1_l [{
   shapeType=Capsule
   length=2.6873
   radius=2.576
   shapeName=dyn_bandoliersBotBack_4_l_capsule
   center=<3.9523 -1.2776 -0.3337>
   rotation=<rot 3.4409 100.965 -82.0385>
  }]
 dyn_bandoliersBotBack_1_r [{
   shapeType=Capsule
   length=2.6873
   radius=2.576
   shapeName=dyn_bandoliersBotBack_4_l_capsule
   center=<3.9128 -0.6716 -0.038>
   rotation=<rot 3.2269 94.6449 -92.1242>
  }]
 dyn_bandoliersBotBack_2_l [{
   shapeType=Capsule
   length=1.8838
   radius=2.7132
   shapeName=dyn_bandoliersBotBack_4_l_capsule
   center=<3.4159 -1.5237 -0.4224>
   rotation=<rot 4.4389 93.5841 -82.5463>
  }]
 dyn_bandoliersBotBack_2_r [{
   shapeType=Capsule
   length=2.129
   radius=2.4679
   shapeName=dyn_bandoliersBotBack_4_l_capsule
   center=<3.3158 -1.2742 0.5183>
   rotation=<rot 5.2162 87.4093 -83.0665>
  }]
 dyn_bandoliersBotBack_3_l [{
   shapeType=Capsule
   length=2.268
   radius=2.5954
   shapeName=dyn_bandoliersBotBack_4_l_capsule
   center=<3.6532 -1.1492 0.0056>
   rotation=<rot -6.8453 88.8926 -88.5097>
  }]
 dyn_bandoliersBotBack_3_r [{
   shapeType=Capsule
   length=2.2568
   radius=2.1562
   shapeName=dyn_bandoliersBotBack_3_r_capsule
   center=<4.4339 -0.5521 0.3897>
   rotation=<rot -4.0829 -90 -92.1971>
  }]
 dyn_bandoliersBotBack_4_l [{
   shapeType=Capsule
   length=2.7591
   radius=2.6015
   shapeName=dyn_bandoliersBotBack_4_l_capsule
   center=<3.4822 -1.8638 -0.3768>
   rotation=<rot -9.0242 91.4835 -90.4447>
  }]
 dyn_bandoliersBotBack_4_r [{
   shapeType=Capsule
   length=4.3979
   radius=2.2822
   shapeName=dyn_bandoliersBotBack_4_r_capsule
   center=<4.8893 -0.5551 0.2127>
   rotation=<rot 0 89.9749 -87.7429>
  }]
 dyn_bandoliersBotBack_5_l [{
   shapeType=Capsule
   length=1.9344
   radius=3.0324
   shapeName=dyn_bandoliersBotBack_5_l_capsule
   center=<2.5281 -2.079 -0.3859>
   rotation=<rot 7.5528 92.9449 -93.4647>
  }]
 dyn_bandoliersBotBack_5_r [{shapeType=Capsule, length=0.505, radius=0.505, shapeName=dyn_bandoliersBotBack_5_r_capsule, rotation=<rot 0 -90 0>}]
 dyn_bandoliersBotFront_1_l [{
   shapeType=Capsule
   length=7.761
   radius=3.1941
   shapeName=dyn_bandoliersBotFront_1_l_capsule
   center=<0.9992 -1.972 0.8755>
   rotation=<rot -5.2251 95.3112 -95.1423>
  }]
 dyn_bandoliersBotFront_1_r [{shapeType=Capsule, length=0.505, radius=0.505, shapeName=dyn_bandoliersBotFront_1_r_capsule, rotation=<rot 0 -90 0>}]
 dyn_bandoliersBotFront_2_l [{
   shapeType=Capsule
   length=2.9983
   radius=3.129
   shapeName=dyn_bandoliersBotFront_2_l_capsule
   center=<2.8251 -1.6489 1.1568>
   rotation=<rot -14.7175 88.5781 -88.2306>
  }]
 dyn_bandoliersBotFront_2_r [{
   shapeType=Capsule
   length=4.4152
   radius=3.8067
   shapeName=dyn_bandoliersBotFront_2_r_capsule
   center=<1.1459 -0.4379 -0.9657>
   rotation=<rot -7.1286 -90 -123.0469>
  }]
 dyn_bandoliersBotFront_3_l [{
   shapeType=Capsule
   length=3.1061
   radius=4.0442
   shapeName=dyn_bandoliersBotFront_3_l_capsule
   center=<2.8802 -3.531 0.69>
   rotation=<rot 10.4087 -95.8866 -89.8241>
  }]
 dyn_bandoliersBotFront_3_r [{
   shapeType=Capsule
   length=5.2107
   radius=4.0511
   shapeName=dyn_bandoliersBotFront_3_r_capsule
   center=<3.2809 -1.5485 -0.7183>
   rotation=<rot 0 84.9351 -76.1962>
  }]
 dyn_bandoliersBotFront_4_l [{
   shapeType=Capsule
   length=4.584
   radius=3.7531
   shapeName=dyn_bandoliersBotFront_3_l_capsule
   center=<3.8804 -3.538 0.6693>
   rotation=<rot 10.1527 -83.9547 -87.6871>
  }]
 dyn_bandoliersBotFront_4_r [{
   shapeType=Capsule
   length=3.3028
   radius=3.0286
   shapeName=dyn_bandoliersBotFront_4_r_capsule
   center=<3.5017 -2.1323 -0.3953>
   rotation=<rot -7.5873 -90 -2.7423>
  }]
 dyn_bandoliersBotFront_5_l [{
   shapeType=Capsule
   length=1.8902
   radius=3.7531
   shapeName=dyn_bandoliersBotFront_3_l_capsule
   center=<1.2683 -3.2999 0.6832>
   rotation=<rot 10.4093 -97.5823 -90.1305>
  }]
 dyn_bandoliersBotFront_5_r [{
   shapeType=Capsule
   length=4.3462
   radius=3.3993
   shapeName=dyn_bandoliersBotFront_5_r_capsule
   center=<3.146 -1.7786 0.0428>
   rotation=<rot -10.6002 -90 -94.9469>
  }]
 dyn_bandoliersBotMid_1_l [{
   shapeType=Capsule
   length=5.0766
   radius=2.6733
   shapeName=dyn_bandoliersBotMid_1_l_capsule
   center=<20.2868 -0.7931 0.7216>
   rotation=<rot 8.4807 -46.2012 -173.9076>
  }]
 dyn_bandoliersBotMid_1_r [{
   shapeType=Capsule
   length=2.3461
   radius=2.1325
   shapeName=dyn_bandoliersBotMid_1_r_capsule
   center=<19.3545 -0.1382 1.2308>
   rotation=<rot 0 134.8952 -165.7721>
  }]
 pelvis [{
   shapeType=Capsule
   length=19.6613
   radius=9.1324
   shapeName=pelvis_capsule
   center=<-1.9573 -0.6502 -0.1758>
   rotation=<rot 2.8444 -90 0.491>
  }]
 thigh_twist_01_l [{
   shapeType=Capsule
   length=25.3285
   radius=9.0803
   shapeName=thigh_twist_01_l_capsule
   center=<4.5427 -1.3951 1.1652>
   rotation=<rot 85.1192 179.9996 179.2859>
  }]
 thigh_twist_01_r [{
   shapeType=Capsule
   length=25.0726
   radius=9.3362
   shapeName=thigh_twist_01_r_capsule
   center=<-1.1195 0.9612 -1.9127>
   rotation=<rot 88.002 179.9998 -0.0002>
  }]
}
constraints [
 {bone1Name=dyn_bandoliersBotBack_1_l, bone2Name=base_bandoliersBotBack_1_l, swing1LimitDegrees=15, swing2LimitDegrees=10, twistLimitDegrees=0}
 {bone1Name=dyn_bandoliersBotBack_2_l, bone2Name=dyn_bandoliersBotBack_1_l, swing1LimitDegrees=15, swing2LimitDegrees=10, twistLimitDegrees=0}
 {bone1Name=dyn_bandoliersBotBack_3_l, bone2Name=dyn_bandoliersBotBack_2_l, swing1LimitDegrees=15, swing2LimitDegrees=10, twistLimitDegrees=0}
 {bone1Name=dyn_bandoliersBotBack_4_l, bone2Name=dyn_bandoliersBotBack_3_l, swing1LimitDegrees=15, swing2LimitDegrees=10, twistLimitDegrees=0}
 {bone1Name=dyn_bandoliersBotBack_5_l, bone2Name=dyn_bandoliersBotBack_4_l, swing1LimitDegrees=15, swing2LimitDegrees=10, twistLimitDegrees=0}
 {
  bone1Name=dyn_bandoliersBotFront_5_l
  bone2Name=dyn_bandoliersBotFront_4_l
  swing1LimitDegrees=15
  swing2LimitDegrees=10
  twistLimitDegrees=0
 }
 {
  bone1Name=dyn_bandoliersBotFront_1_l
  bone2Name=base_bandoliersBotFront_1_l
  swing1LimitDegrees=15
  swing2LimitDegrees=10
  twistLimitDegrees=0
 }
 {
  bone1Name=dyn_bandoliersBotFront_2_l
  bone2Name=dyn_bandoliersBotFront_1_l
  swing1LimitDegrees=15
  swing2LimitDegrees=10
  twistLimitDegrees=0
 }
 {
  bone1Name=dyn_bandoliersBotFront_3_l
  bone2Name=dyn_bandoliersBotFront_2_l
  swing1LimitDegrees=15
  swing2LimitDegrees=10
  twistLimitDegrees=0
 }
 {
  bone1Name=dyn_bandoliersBotFront_4_l
  bone2Name=dyn_bandoliersBotFront_3_l
  swing1LimitDegrees=15
  swing2LimitDegrees=10
  twistLimitDegrees=0
 }
 {bone1Name=dyn_bandoliersBotMid_1_l, bone2Name=base_bandoliersBotMid_1_l, swing1LimitDegrees=35, swing2LimitDegrees=20, twistLimitDegrees=0}
 {
  bone1Name=base_bandoliersBotMid_1_l
  bone2Name=thigh_twist_01_l
  swing1LimitDegrees=45
  swing1Motion=Free
  swing2LimitDegrees=45
  swing2Motion=Free
  twistLimitDegrees=45
  twistMotion=Free
 }
 {
  bone1Name=thigh_twist_01_l
  bone2Name=pelvis
  swing1LimitDegrees=45
  swing1Motion=Free
  swing2LimitDegrees=45
  swing2Motion=Free
  twistLimitDegrees=45
  twistMotion=Free
 }
 {
  bone1Name=base_bandoliersBotBack_1_l
  bone2Name=pelvis
  swing1LimitDegrees=45
  swing1Motion=Free
  swing2LimitDegrees=45
  swing2Motion=Free
  twistLimitDegrees=45
  twistMotion=Free
 }
 {
  bone1Name=base_bandoliersBotFront_1_l
  bone2Name=pelvis
  swing1LimitDegrees=45
  swing1Motion=Free
  swing2LimitDegrees=45
  swing2Motion=Free
  twistLimitDegrees=45
  twistMotion=Free
 }
 {
  bone1Name=base_bandoliersBotBack_1_r
  bone2Name=pelvis
  swing1LimitDegrees=45
  swing1Motion=Free
  swing2LimitDegrees=45
  swing2Motion=Free
  twistLimitDegrees=45
  twistMotion=Free
 }
 {
  bone1Name=dyn_bandoliersBotBack_1_r
  bone2Name=base_bandoliersBotBack_1_r
  swing1LimitDegrees=45
  swing1Motion=Free
  swing2LimitDegrees=45
  swing2Motion=Free
  twistLimitDegrees=45
  twistMotion=Free
 }
 {
  bone1Name=dyn_bandoliersBotBack_2_r
  bone2Name=dyn_bandoliersBotBack_1_r
  swing1LimitDegrees=45
  swing1Motion=Free
  swing2LimitDegrees=45
  swing2Motion=Free
  twistLimitDegrees=45
  twistMotion=Free
 }
 {
  bone1Name=dyn_bandoliersBotBack_3_r
  bone2Name=dyn_bandoliersBotBack_2_r
  swing1LimitDegrees=45
  swing1Motion=Free
  swing2LimitDegrees=45
  swing2Motion=Free
  twistLimitDegrees=45
  twistMotion=Free
 }
 {
  bone1Name=dyn_bandoliersBotBack_4_r
  bone2Name=dyn_bandoliersBotBack_3_r
  swing1LimitDegrees=45
  swing1Motion=Free
  swing2LimitDegrees=45
  swing2Motion=Free
  twistLimitDegrees=45
  twistMotion=Free
 }
 {
  bone1Name=dyn_bandoliersBotBack_5_r
  bone2Name=dyn_bandoliersBotBack_4_r
  swing1LimitDegrees=45
  swing1Motion=Free
  swing2LimitDegrees=45
  swing2Motion=Free
  twistLimitDegrees=45
  twistMotion=Free
 }
 {
  bone1Name=base_bandoliersBotFront_1_r
  bone2Name=pelvis
  swing1LimitDegrees=45
  swing1Motion=Free
  swing2LimitDegrees=45
  swing2Motion=Free
  twistLimitDegrees=45
  twistMotion=Free
 }
 {
  bone1Name=dyn_bandoliersBotFront_1_r
  bone2Name=base_bandoliersBotFront_1_r
  swing1LimitDegrees=45
  swing1Motion=Free
  swing2LimitDegrees=45
  swing2Motion=Free
  twistLimitDegrees=45
  twistMotion=Free
 }
 {
  bone1Name=dyn_bandoliersBotFront_2_r
  bone2Name=dyn_bandoliersBotFront_1_r
  swing1LimitDegrees=45
  swing1Motion=Free
  swing2LimitDegrees=45
  swing2Motion=Free
  twistLimitDegrees=45
  twistMotion=Free
 }
 {
  bone1Name=dyn_bandoliersBotFront_3_r
  bone2Name=dyn_bandoliersBotFront_2_r
  swing1LimitDegrees=45
  swing1Motion=Free
  swing2LimitDegrees=45
  swing2Motion=Free
  twistLimitDegrees=45
  twistMotion=Free
 }
 {
  bone1Name=dyn_bandoliersBotFront_4_r
  bone2Name=dyn_bandoliersBotFront_3_r
  swing1LimitDegrees=45
  swing1Motion=Free
  swing2LimitDegrees=45
  swing2Motion=Free
  twistLimitDegrees=45
  twistMotion=Free
 }
 {
  bone1Name=dyn_bandoliersBotFront_5_r
  bone2Name=dyn_bandoliersBotFront_4_r
  swing1LimitDegrees=45
  swing1Motion=Free
  swing2LimitDegrees=45
  swing2Motion=Free
  twistLimitDegrees=45
  twistMotion=Free
 }
 {
  bone1Name=base_bandoliersBotMid_1_r
  bone2Name=pelvis
  swing1LimitDegrees=45
  swing1Motion=Free
  swing2LimitDegrees=45
  swing2Motion=Free
  twistLimitDegrees=45
  twistMotion=Free
 }
 {
  bone1Name=dyn_bandoliersBotMid_1_r
  bone2Name=base_bandoliersBotMid_1_r
  swing1LimitDegrees=45
  swing1Motion=Free
  swing2LimitDegrees=45
  swing2Motion=Free
  twistLimitDegrees=45
  twistMotion=Free
 }
 {
  bone1Name=thigh_twist_01_r
  bone2Name=pelvis
  swing1LimitDegrees=45
  swing1Motion=Free
  swing2LimitDegrees=45
  swing2Motion=Free
  twistLimitDegrees=45
  twistMotion=Free
 }
 {
  bone1Name=dyn_bandoliersBotFront_5_l
  bone2Name=dyn_bandoliersBotMid_1_l
  swing1LimitDegrees=45
  swing1Motion=Locked
  swing2LimitDegrees=45
  swing2Motion=Locked
  twistLimitDegrees=45
  twistMotion=Locked
 }
 {
  bone1Name=dyn_bandoliersBotBack_5_l
  bone2Name=dyn_bandoliersBotMid_1_r
  swing1LimitDegrees=45
  swing1Motion=Locked
  swing2LimitDegrees=45
  swing2Motion=Locked
  twistLimitDegrees=45
  twistMotion=Locked
 }
]
