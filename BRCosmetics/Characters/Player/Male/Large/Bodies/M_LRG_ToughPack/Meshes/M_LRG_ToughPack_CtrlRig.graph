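# Control Rig for the large male ToughPack player body, procedurally driving corrective deformation and accessory bones.
# Weighted constraints attach arm-bend, bandolier, shoulder-strap, knee, groin and glute helpers to body bones.
# DeltaRotation swing/twist readings and spherical or pose-space readers weight transform offsets and FBIK chain corrections on limbs.
# isTargeting and isFrontEnd gate some arm fixes, and a debug branch copies control transforms onto bones.
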ControlRigBlueprint /BRCosmetics/Characters/Player/Male/Large/Bodies/M_LRG_ToughPack/Meshes/M_LRG_ToughPack_CtrlRig
graphs [
 CollapseNode_1_ContainedGraph {nodes [
   RigVMFunctionEntryNode()
   > &RigUnit_ParentConstraint(
    AdvancedSettings=(InterpolationType=Average,RotationOrderForFilter=XZY)
    Child=(Type=Null,Name="arm_bend_1_r_ctrl_null")
    Parents=((Item=(Type=Bone,Name="upperarmBend_r"),Weight=1))
    Weight=1
   )
   > &RigUnit_ParentConstraint(
    AdvancedSettings=(InterpolationType=Average,RotationOrderForFilter=XZY)
    Child=(Type=Null,Name="arm_bend_2_r_ctrl_null")
    Filter=(TranslationFilter=(bX=false,bY=false,bZ=false),RotationFilter=(bX=True,bY=True,bZ=True),ScaleFilter=(bX=True,bY=True,bZ=True))
    Parents=((Item=(Type=Bone,Name="upperarmBend_r"),Weight=1))
    Weight=1
   )
   > /Script/ControlRig.RigUnit_PositionConstraintLocalSpaceOffset(
    Child=(Type=Null,Name="arm_bend_2_r_ctrl_null")
    Filter=(bX=True,bY=True,bZ=True)
    Parents=((Item=(Type=Bone,Name="upperarmBend_r"),Weight=0.5),(Item=(Type=Bone,Name="lowerarmBend_r"),Weight=0.5))
    Weight=1
    bMaintainOffset=True
   ) {is=struct RigVMUnitNode}
   > &RigUnit_ParentConstraint(
    AdvancedSettings=(InterpolationType=Average,RotationOrderForFilter=XZY)
    Child=(Type=Null,Name="arm_bend_3_r_ctrl_null")
    Parents=((Item=(Type=Bone,Name="lowerarmBend_r"),Weight=1))
    Weight=1
   )
   > &RigUnit_ParentConstraint(
    AdvancedSettings=(InterpolationType=Average,RotationOrderForFilter=XZY)
    Child=(Type=Null,Name="arm_bend_4_r_ctrl_null")
    Filter=(TranslationFilter=(bX=false,bY=false,bZ=false),RotationFilter=(bX=True,bY=True,bZ=True),ScaleFilter=(bX=True,bY=True,bZ=True))
    Parents=((Item=(Type=Bone,Name="lowerarmBend_r"),Weight=1))
    Weight=1
   )
   > &RigUnit_PositionConstraintLocalSpaceOffset(
    Child=(Type=Null,Name="arm_bend_4_r_ctrl_null")
    Filter=(bX=True,bY=True,bZ=True)
    Parents=((Item=(Type=Bone,Name="lowerarmBend_r"),Weight=0.5),(Item=(Type=Bone,Name="lowerarmBend_r_end"),Weight=0.5))
    Weight=1
    bMaintainOffset=True
   )
   > &RigUnit_ParentConstraint(
    AdvancedSettings=(InterpolationType=Average,RotationOrderForFilter=XZY)
    Child=(Type=Null,Name="arm_bend_5_r_ctrl_null")
    Parents=((Item=(Type=Bone,Name="lowerarmBend_r_end"),Weight=1))
    Weight=1
   )
   > RigVMFunctionReturnNode()
  ]}
 CollapseNode_ContainedGraph {nodes [
   RigVMFunctionEntryNode()
   > &RigUnit_ParentConstraint(
    AdvancedSettings=(InterpolationType=Average,RotationOrderForFilter=XZY)
    Child=(Type=Null,Name="arm_bend_1_l_ctrl_null")
    Parents=((Item=(Type=Bone,Name="upperarmBend_l"),Weight=1))
    Weight=1
   )
   > &RigUnit_ParentConstraint(
    AdvancedSettings=(InterpolationType=Average,RotationOrderForFilter=XZY)
    Child=(Type=Null,Name="arm_bend_2_l_ctrl_null")
    Filter=(TranslationFilter=(bX=false,bY=false,bZ=false),RotationFilter=(bX=True,bY=True,bZ=True),ScaleFilter=(bX=True,bY=True,bZ=True))
    Parents=((Item=(Type=Bone,Name="upperarmBend_l"),Weight=1))
    Weight=1
   )
   > &RigUnit_PositionConstraintLocalSpaceOffset(
    Child=(Type=Null,Name="arm_bend_2_l_ctrl_null")
    Filter=(bX=True,bY=True,bZ=True)
    Parents=((Item=(Type=Bone,Name="upperarmBend_l"),Weight=0.5),(Item=(Type=Bone,Name="lowerarmBend_l"),Weight=0.5))
    Weight=1
    bMaintainOffset=True
   )
   > &RigUnit_ParentConstraint(
    AdvancedSettings=(InterpolationType=Average,RotationOrderForFilter=XZY)
    Child=(Type=Null,Name="arm_bend_3_l_ctrl_null")
    Parents=((Item=(Type=Bone,Name="lowerarmBend_l"),Weight=1))
    Weight=1
   )
   > &RigUnit_ParentConstraint(
    AdvancedSettings=(InterpolationType=Average,RotationOrderForFilter=XZY)
    Child=(Type=Null,Name="arm_bend_4_l_ctrl_null")
    Filter=(TranslationFilter=(bX=false,bY=false,bZ=false),RotationFilter=(bX=True,bY=True,bZ=True),ScaleFilter=(bX=True,bY=True,bZ=True))
    Parents=((Item=(Type=Bone,Name="lowerarmBend_l"),Weight=1))
    Weight=1
   )
   > &RigUnit_PositionConstraintLocalSpaceOffset(
    Child=(Type=Null,Name="arm_bend_4_l_ctrl_null")
    Filter=(bX=True,bY=True,bZ=True)
    Parents=((Item=(Type=Bone,Name="lowerarmBend_l"),Weight=0.5),(Item=(Type=Bone,Name="lowerarmBend_l_end"),Weight=0.5))
    Weight=1
    bMaintainOffset=True
   )
   > &RigUnit_ParentConstraint(
    AdvancedSettings=(InterpolationType=Average,RotationOrderForFilter=XZY)
    Child=(Type=Null,Name="arm_bend_5_l_ctrl_null")
    Parents=((Item=(Type=Bone,Name="lowerarmBend_l_end"),Weight=1))
    Weight=1
   )
   > RigVMFunctionReturnNode()
  ]}
 CollapseNode_ContainedGraph {nodes [
   n1 = RigVMFunctionEntryNode()
   > DeltaRotation = RigVMFunctionReferenceNode(
    Rotation Order=ZYX
    child=lowerarm_l
    debugRotation=false
    max=110
    parent=lowerarm_l_null
    parentType=Null
    premult=1
    primAxis=2
    twistAxis=(X=1,Y=0,Z=0)
   )
   RigVMFunction_ControlFlowBranch = &RigVMFunction_ControlFlowBranch(ExecuteContext=@DeltaRotation.ExecuteContext, Condition=@n1.enable)
   RigVMFunctionReturnNode(ExecuteContext=@RigVMFunction_ControlFlowBranch.Completed)
   &RigUnit_ModifyTransforms(
    ItemToModify="((Item=(Type=Null,Name=\"arm_bend_3_l_ctrl_null\"),Transform=(Rotation=(X=0,Y=0,Z=0,W=1),Translation=(X=-5,Y=0,Z=-15),Scale3D=(X=1,Y=1,Z=1))),(Item=(Type=Null,Name=\"arm_bend_2_l_ctrl_null\"),Transform=(Rotation=(X=0,Y=0,Z=0,W=1),Translation=(X=0,Y=0,Z=-10),Sc ...(truncated)"
    WeightMinimum=0
    Weight=@DeltaRotation.outputWeight
    ExecutePin=@RigVMFunction_ControlFlowBranch.False
   )
   > &RigUnit_OffsetTransformForItem(
    Item=(Type=Bone,Name="arm_bend_10_r")
    OffsetTransform=(Rotation=(X=0,Y=0,Z=0,W=1),Translation=(X=0,Y=0,Z=0),Scale3D=(X=1,Y=0.5,Z=1))
    Weight=1
    bPropagateToChildren=false
   )
   > &RigUnit_OffsetTransformForItem(
    Item=(Type=Bone,Name="arm_bend_8_l")
    OffsetTransform=(Rotation=(X=0,Y=0,Z=0,W=1),Translation=(X=0,Y=0,Z=0),Scale3D=(X=0.8,Y=0.8,Z=0.8))
    Weight=1
    bPropagateToChildren=false
   )
   > &RigUnit_OffsetTransformForItem(
    Item=(Type=Bone,Name="arm_bend_9_l")
    OffsetTransform=(Rotation=(X=0,Y=0,Z=0,W=1),Translation=(X=0,Y=0,Z=0),Scale3D=(X=0.7,Y=0.7,Z=0.7))
    Weight=1
    bPropagateToChildren=false
   )
   > &RigUnit_OffsetTransformForItem(
    Item=(Type=Bone,Name="arm_bend_10_l")
    OffsetTransform=(Rotation=(X=0,Y=0,Z=0,W=1),Translation=(X=0,Y=0,Z=0),Scale3D=(X=1,Y=0.5,Z=1))
    Weight=1
    bPropagateToChildren=false
   )
   &RigUnit_ModifyTransforms(
    ItemToModify="((Item=(Type=Null,Name=\"arm_bend_2_l_ctrl_null\"),Transform=(Rotation=(X=0,Y=0,Z=0,W=1),Translation=(X=0,Y=0,Z=-12),Scale3D=(X=1,Y=1,Z=1))),(Item=(Type=Null,Name=\"arm_bend_4_l_ctrl_null\"),Transform=(Rotation=(X=0,Y=0,Z=0,W=1),Translation=(X=0,Y=0,Z=-8),Scal ...(truncated)"
    WeightMinimum=0
    Weight=@DeltaRotation.outputWeight
    ExecutePin=@RigVMFunction_ControlFlowBranch.True
   )
   > n2 = &RigUnit_SphericalPoseReader(
    ActiveRegionScaleFactors=(PositiveWidth=1,NegativeWidth=1,PositiveHeight=1,NegativeHeight=1)
    ActiveRegionSize=0.08
    Debug=(bDrawDebug=false,bDraw2D=False,bDrawLocalAxes=False,DebugScale=25,DebugSegments=20,DebugThickness=0.25)
    DriverAxis=(X=1,Y=0,Z=0)
    DriverItem=(Type=Bone,Name="upperarmBend_l")
    FalloffRegionScaleFactors=(PositiveWidth=1,NegativeWidth=1,PositiveHeight=1,NegativeHeight=1)
    FalloffSize=0.22
    FlipHeightScaling=False
    FlipWidthScaling=False
    OptionalParentItem=(Type=Bone,Name="None")
    RotationOffset=(X=3,Y=60,Z=40)
   )
   Multiply = &RigVMFunction_MathDoubleMul(A=@n2.OutputParam, B=RigVMVariableNode() {variable=pub isTargeting}.Value)
   > &RigUnit_ModifyTransforms(
    ItemToModify="((Item=(Type=Null,Name=\"arm_bend_2_l_ctrl_null\"),Transform=(Rotation=(X=0,Y=0,Z=0,W=1),Translation=(X=0,Y=-5,Z=-2),Scale3D=(X=1,Y=1,Z=1))),(Item=(Type=Null,Name=\"arm_bend_1_l_ctrl_null\"),Transform=(Rotation=(X=0.5,Y=0,Z=0,W=0.866),Translation=(X=0,Y=-2,Z=0 ...(truncated)"
    WeightMinimum=0
    Weight=@Multiply.Result
   )
  ]}
 CollapseNode_ContainedGraph {nodes [
   n1 = RigVMFunctionEntryNode()
   DeltaRotation_1 = RigVMFunctionReferenceNode(
    Rotation Order=ZYX
    child=lowerarm_r
    debugRotation=false
    max=110
    parent=lowerarm_r_null
    parentType=Null
    premult=1
    primAxis=2
    twistAxis=(X=1,Y=0,Z=0)
    ExecuteContext=@n1.ExecutePin
   )
   RigVMFunction_ControlFlowBranch = &RigVMFunction_ControlFlowBranch(ExecuteContext=@DeltaRotation_1.ExecuteContext, Condition=@n1.enable)
   RigVMFunctionReturnNode(ExecutePin=@RigVMFunction_ControlFlowBranch.Completed)
   &RigUnit_ModifyTransforms(
    ItemToModify="((Item=(Type=Null,Name=\"arm_bend_3_r_ctrl_null\"),Transform=(Rotation=(X=0,Y=0,Z=0,W=1),Translation=(X=5,Y=0,Z=15),Scale3D=(X=1,Y=1,Z=1))),(Item=(Type=Null,Name=\"arm_bend_2_r_ctrl_null\"),Transform=(Rotation=(X=0,Y=0,Z=0,W=1),Translation=(X=0,Y=0,Z=10),Scale ...(truncated)"
    WeightMinimum=0
    Weight=@DeltaRotation_1.outputWeight
    ExecutePin=@RigVMFunction_ControlFlowBranch.False
   )
   > &RigUnit_OffsetTransformForItem(
    Item=(Type=Bone,Name="arm_bend_9_r")
    OffsetTransform=(Rotation=(X=0,Y=0,Z=0,W=1),Translation=(X=0,Y=0,Z=0),Scale3D=(X=0.5,Y=0.5,Z=0.5))
    Weight=1
    bPropagateToChildren=false
   )
   > &RigUnit_OffsetTransformForItem(
    Item=(Type=Bone,Name="arm_bend_8_r")
    OffsetTransform=(Rotation=(X=0,Y=0,Z=0,W=1),Translation=(X=0,Y=0,Z=0),Scale3D=(X=0.7,Y=0.7,Z=0.7))
    Weight=1
    bPropagateToChildren=false
   )
   &RigUnit_ModifyTransforms(
    ItemToModify="((Item=(Type=Null,Name=\"arm_bend_2_r_ctrl_null\"),Transform=(Rotation=(X=0,Y=0,Z=0,W=1),Translation=(X=0,Y=0,Z=12),Scale3D=(X=1,Y=1,Z=1))),(Item=(Type=Null,Name=\"arm_bend_4_r_ctrl_null\"),Transform=(Rotation=(X=0,Y=0,Z=0,W=1),Translation=(X=0,Y=0,Z=8),Scale3 ...(truncated)"
    WeightMinimum=0
    Weight=@DeltaRotation_1.outputWeight
    ExecutePin=@RigVMFunction_ControlFlowBranch.True
   )
   > n2 = &RigUnit_SphericalPoseReader(
    ActiveRegionScaleFactors=(PositiveWidth=1.2,NegativeWidth=1.2,PositiveHeight=1.8,NegativeHeight=1.8)
    ActiveRegionSize=0.1
    Debug=(bDrawDebug=false,bDraw2D=False,bDrawLocalAxes=False,DebugScale=25,DebugSegments=20,DebugThickness=0.25)
    DriverAxis=(X=-1,Y=0,Z=0)
    DriverItem=(Type=Bone,Name="upperarmBend_r")
    FalloffRegionScaleFactors=(PositiveWidth=1,NegativeWidth=1,PositiveHeight=2.5,NegativeHeight=2.5)
    FalloffSize=0.25
    FlipHeightScaling=False
    FlipWidthScaling=False
    OptionalParentItem=(Type=Bone,Name="None")
    RotationOffset=(X=25,Y=215,Z=-10)
   )
   > PoseSpaceDeformation = RigVMFunctionReferenceNode(
    Child=(Type=Bone,Name="upperarmBend_r")
    Parent=(Type=Bone,Name="spine_05")
    coneAngle=90
    drawLines=false
    primaryAxis=(X=-10,Y=0,Z=0)
    quotient=0.5
    targets=((X=-0.2316,Y=-0.8644,Z=-0.1155,W=0.431))
   )
   Remap = &RigVMFunction_MathDoubleRemap(
    SourceMaximum=1
    SourceMinimum=0
    TargetMaximum=0.5
    TargetMinimum=1
    bClamp=False
    Value=RigVMDispatchNode(Index=0, Array=@PoseSpaceDeformation.weights) {template="DISPATCH_RigVMDispatch_ArrayGetAtIndex(in Array,in Index,out Element)"}.Element
   )
   Multiply_2 = RigVMAggregateNode(A=@Remap.Result, C=RigVMVariableNode() {variable=pub isTargeting}.Value, B=@n2.OutputParam)
   > &RigUnit_ModifyTransforms(
    ItemToModify="((Item=(Type=Null,Name=\"arm_bend_2_r_ctrl_null\"),Transform=(Rotation=(X=0.2588,Y=0,Z=0,W=0.9659),Translation=(X=0,Y=15,Z=3),Scale3D=(X=1,Y=1,Z=1))),(Item=(Type=Null,Name=\"arm_bend_3_r_ctrl_null\"),Transform=(Rotation=(X=0.1736,Y=0,Z=0,W=0.9848),Translation= ...(truncated)"
    WeightMinimum=0
    Weight=@Multiply_2.Result
   )
   > &RigUnit_ModifyTransforms(
    ItemToModify="((Item=(Type=Null,Name=\"arm_bend_2_r_ctrl_null\"),Transform=(Rotation=(X=0,Y=0,Z=0,W=1),Translation=(X=0,Y=4,Z=0),Scale3D=(X=1,Y=1,Z=1))),(Item=(Type=Null,Name=\"arm_bend_3_r_ctrl_null\"),Transform=(Rotation=(X=0,Y=0,Z=0,W=1),Translation=(X=0,Y=4,Z=0),Scale3D ...(truncated)"
    WeightMinimum=0
    Weight=RigVMDispatchNode(False=0, True=RigVMVariableNode() {variable=rUpperarmDown}.Value, Condition=RigVMVariableNode() {variable=pub isFrontEnd}.Value) {template="DISPATCH_RigVMDispatch_If(in Condition,in True,in False,out Result)"}.Result
   )
  ]}
 CollapseNode_ContainedGraph {nodes [
   n1 = RigVMFunctionEntryNode()
   RigVMFunction_ControlFlowBranch = &RigVMFunction_ControlFlowBranch(ExecuteContext=@n1.ExecutePin, Condition=@n1.enable)
   RigVMFunctionReturnNode(ExecutePin=@RigVMFunction_ControlFlowBranch.Completed)
   n2 = &RigUnit_SphericalPoseReader(
    ActiveRegionScaleFactors=(PositiveWidth=1,NegativeWidth=1,PositiveHeight=1,NegativeHeight=1)
    ActiveRegionSize=0.15
    Debug=(bDrawDebug=false,bDraw2D=False,bDrawLocalAxes=False,DebugScale=25,DebugSegments=20,DebugThickness=0.25)
    DriverAxis=(X=1,Y=0,Z=0)
    DriverItem=(Type=Bone,Name="upperarm_l")
    FalloffRegionScaleFactors=(PositiveWidth=1,NegativeWidth=1,PositiveHeight=1,NegativeHeight=1)
    FalloffSize=0.3
    FlipHeightScaling=False
    FlipWidthScaling=False
    OptionalParentItem=(Type=Bone,Name="spine_05")
    RotationOffset=(X=140.7187,Y=-6.7774,Z=-71.1958)
    ExecutePin=@RigVMFunction_ControlFlowBranch.True
   )
   > ApplyFBIKChainCorrection = RigVMFunctionReferenceNode(End=(Type=Bone,Name="lowerarmBend_l_end"), QuatAmount=-0.5, Start=(Type=Bone,Name="upperarmBend_l"), Weight=@n2.OutputParam)
  ]}
 CollapseNode_ContainedGraph {nodes [
   n1 = RigVMFunctionEntryNode()
   RigVMFunction_ControlFlowBranch = &RigVMFunction_ControlFlowBranch(ExecuteContext=@n1.ExecutePin, Condition=@n1.enable)
   RigVMFunctionReturnNode(ExecutePin=@RigVMFunction_ControlFlowBranch.Completed)
   n2 = &RigUnit_SphericalPoseReader(
    ActiveRegionScaleFactors=(PositiveWidth=1,NegativeWidth=1,PositiveHeight=1,NegativeHeight=1)
    ActiveRegionSize=0.15
    Debug=(bDrawDebug=false,bDraw2D=False,bDrawLocalAxes=False,DebugScale=25,DebugSegments=20,DebugThickness=0.25)
    DriverAxis=(X=-1,Y=0,Z=0)
    DriverItem=(Type=Bone,Name="upperarm_r")
    FalloffRegionScaleFactors=(PositiveWidth=1,NegativeWidth=1,PositiveHeight=1,NegativeHeight=1)
    FalloffSize=0.3
    FlipHeightScaling=False
    FlipWidthScaling=False
    OptionalParentItem=(Type=Bone,Name="spine_05")
    RotationOffset=(X=-39.2814,Y=-6.7774,Z=-71.1958)
    ExecutePin=@RigVMFunction_ControlFlowBranch.True
   )
   > RigVMVariableNode(Value=@n2.OutputParam) {variable=rUpperarmDown}
   > ApplyFBIKChainCorrection_1 = RigVMFunctionReferenceNode(End=(Type=Bone,Name="lowerarmBend_r_end"), QuatAmount=0.5, Start=(Type=Bone,Name="upperarmBend_r"), Weight=@n2.OutputParam)
  ]}
 CollapseNode_ContainedGraph {nodes [
   n1 = RigVMFunctionEntryNode()
   &RigUnit_ParentConstraint(
    AdvancedSettings=(RotationOrderForFilter=XZY)
    Child=(Type=Null,Name="dyn_bandoliersBotBack_1_l_ctrl_null")
    Parents=((Item=(Type=Bone,Name="deform_glute_l"),Weight=0.027),(Item=(Type=Bone,Name="pelvis"),Weight=0.711),(Item=(Type=Bone,Name="spine_02"),Weight=0.1),(Item=(Type=Bone,Name="spine_01"),Weight=0.093),(Item=(Type=Bone,Name="spine_03"),Weight=0.069))
    Weight=1
    ExecutePin=@n1.ExecutePin
   )
   > &RigUnit_ParentConstraint(
    AdvancedSettings=(RotationOrderForFilter=XZY)
    Child=(Type=Null,Name="dyn_bandoliersBotBack_2_l_ctrl_null")
    Parents="((Item=(Type=Bone,Name=\"thigh_twist_01_r\"),Weight=0.015),(Item=(Type=Bone,Name=\"deform_glute_l\"),Weight=0.221),(Item=(Type=Bone,Name=\"deform_glute_r\"),Weight=0.058),(Item=(Type=Bone,Name=\"thigh_twist_01_l\"),Weight=0.053),(Item=(Type=Bone,Name=\"pelvis\"),Wei ...(truncated)"
    Weight=1
   )
   > &RigUnit_ParentConstraint(
    AdvancedSettings=(RotationOrderForFilter=XZY)
    Child=(Type=Null,Name="dyn_bandoliersBotBack_3_l_ctrl_null")
    Parents="((Item=(Type=Bone,Name=\"thigh_twist_01_r\"),Weight=0.011),(Item=(Type=Bone,Name=\"deform_glute_l\"),Weight=0.233),(Item=(Type=Bone,Name=\"deform_glute_r\"),Weight=0.018),(Item=(Type=Bone,Name=\"thigh_l\"),Weight=0.052),(Item=(Type=Bone,Name=\"thigh_twist_01_l\"),We ...(truncated)"
    Weight=1
   )
   > &RigUnit_ParentConstraint(
    AdvancedSettings=(RotationOrderForFilter=XZY)
    Child=(Type=Null,Name="dyn_bandoliersBotBack_4_l_ctrl_null")
    Parents=((Item=(Type=Bone,Name="deform_glute_l"),Weight=0.088),(Item=(Type=Bone,Name="thigh_l"),Weight=0.155),(Item=(Type=Bone,Name="thigh_twist_01_l"),Weight=0.561),(Item=(Type=Bone,Name="pelvis"),Weight=0.196))
    Weight=1
   )
   > &RigUnit_ParentConstraint(
    AdvancedSettings=(RotationOrderForFilter=XZY)
    Child=(Type=Null,Name="dyn_bandoliersBotBack_5_l_ctrl_null")
    Parents=((Item=(Type=Bone,Name="thigh_l"),Weight=0.27),(Item=(Type=Bone,Name="thigh_twist_01_l"),Weight=0.643),(Item=(Type=Bone,Name="deform_groin_l"),Weight=0.014),(Item=(Type=Bone,Name="pelvis"),Weight=0.073))
    Weight=1
   )
   > &RigUnit_ParentConstraint(
    AdvancedSettings=(RotationOrderForFilter=XZY)
    Child=(Type=Null,Name="dyn_bandoliersBotFront_1_l_ctrl_null")
    Parents="((Item=(Type=Bone,Name=\"thigh_twist_01_r\"),Weight=0.013),(Item=(Type=Bone,Name=\"deform_groin_r\"),Weight=0.021),(Item=(Type=Bone,Name=\"thigh_twist_01_l\"),Weight=0.066),(Item=(Type=Bone,Name=\"deform_groin_l\"),Weight=0.179),(Item=(Type=Bone,Name=\"pelvis\"),Wei ...(truncated)"
    Weight=1
   )
   > &RigUnit_ParentConstraint(
    AdvancedSettings=(RotationOrderForFilter=XZY)
    Child=(Type=Null,Name="dyn_bandoliersBotFront_2_l_ctrl_null")
    Parents="((Item=(Type=Bone,Name=\"thigh_twist_01_r\"),Weight=0.012),(Item=(Type=Bone,Name=\"deform_groin_r\"),Weight=0.021),(Item=(Type=Bone,Name=\"thigh_twist_01_l\"),Weight=0.113),(Item=(Type=Bone,Name=\"deform_groin_l\"),Weight=0.291),(Item=(Type=Bone,Name=\"pelvis\"),Wei ...(truncated)"
    Weight=1
   )
   > &RigUnit_ParentConstraint(
    AdvancedSettings=(RotationOrderForFilter=XZY)
    Child=(Type=Null,Name="dyn_bandoliersBotFront_3_l_ctrl_null")
    Parents=((Item=(Type=Bone,Name="thigh_l"),Weight=0.044),(Item=(Type=Bone,Name="thigh_twist_01_l"),Weight=0.325),(Item=(Type=Bone,Name="deform_groin_l"),Weight=0.415),(Item=(Type=Bone,Name="pelvis"),Weight=0.215))
    Weight=1
   )
   > &RigUnit_ParentConstraint(
    AdvancedSettings=(RotationOrderForFilter=XZY)
    Child=(Type=Null,Name="dyn_bandoliersBotFront_4_l_ctrl_null")
    Parents=((Item=(Type=Bone,Name="thigh_l"),Weight=0.133),(Item=(Type=Bone,Name="thigh_twist_01_l"),Weight=0.488),(Item=(Type=Bone,Name="deform_groin_l"),Weight=0.281),(Item=(Type=Bone,Name="pelvis"),Weight=0.099))
    Weight=1
   )
   > &RigUnit_ParentConstraint(
    AdvancedSettings=(RotationOrderForFilter=XZY)
    Child=(Type=Null,Name="dyn_bandoliersBotFront_5_l_ctrl_null")
    Parents=((Item=(Type=Bone,Name="thigh_l"),Weight=0.294),(Item=(Type=Bone,Name="thigh_twist_01_l"),Weight=0.562),(Item=(Type=Bone,Name="deform_groin_l"),Weight=0.094),(Item=(Type=Bone,Name="pelvis"),Weight=0.05))
    Weight=1
   )
   > &RigUnit_ParentConstraint(
    AdvancedSettings=(InterpolationType=Average,RotationOrderForFilter=XZY)
    Child=(Type=Bone,Name="base_bandoliersBotMid_1_l")
    Parents=((Item=(Type=Bone,Name="thigh_l"),Weight=0.393),(Item=(Type=Bone,Name="thigh_twist_01_l"),Weight=0.543),(Item=(Type=Bone,Name="deform_groin_l"),Weight=0.03),(Item=(Type=Bone,Name="pelvis"),Weight=0.034))
    Weight=1
   )
   > &RigUnit_ParentConstraint(
    AdvancedSettings=(RotationOrderForFilter=XZY)
    Child=(Type=Null,Name="dyn_bandoliersBotBack_1_r_ctrl_null")
    Parents=((Item=(Type=Bone,Name="deform_glute_r"),Weight=0.027),(Item=(Type=Bone,Name="pelvis"),Weight=0.703),(Item=(Type=Bone,Name="spine_02"),Weight=0.103),(Item=(Type=Bone,Name="spine_01"),Weight=0.096),(Item=(Type=Bone,Name="spine_03"),Weight=0.072))
    Weight=1
   )
   > &RigUnit_ParentConstraint(
    AdvancedSettings=(RotationOrderForFilter=XZY)
    Child=(Type=Null,Name="dyn_bandoliersBotBack_2_r_ctrl_null")
    Parents="((Item=(Type=Bone,Name=\"thigh_twist_01_r\"),Weight=0.053),(Item=(Type=Bone,Name=\"deform_glute_l\"),Weight=0.058),(Item=(Type=Bone,Name=\"deform_glute_r\"),Weight=0.221),(Item=(Type=Bone,Name=\"thigh_twist_01_l\"),Weight=0.015),(Item=(Type=Bone,Name=\"pelvis\"),Wei ...(truncated)"
    Weight=1
   )
   > &RigUnit_ParentConstraint(
    AdvancedSettings=(RotationOrderForFilter=XZY)
    Child=(Type=Null,Name="dyn_bandoliersBotBack_3_r_ctrl_null")
    Parents="((Item=(Type=Bone,Name=\"thigh_twist_01_r\"),Weight=0.288),(Item=(Type=Bone,Name=\"deform_glute_l\"),Weight=0.018),(Item=(Type=Bone,Name=\"deform_glute_r\"),Weight=0.233),(Item=(Type=Bone,Name=\"thigh_twist_01_l\"),Weight=0.011),(Item=(Type=Bone,Name=\"thigh_r\"),We ...(truncated)"
    Weight=1
   )
   > &RigUnit_ParentConstraint(
    AdvancedSettings=(RotationOrderForFilter=XZY)
    Child=(Type=Null,Name="dyn_bandoliersBotBack_4_r_ctrl_null")
    Parents=((Item=(Type=Bone,Name="thigh_twist_01_r"),Weight=0.561),(Item=(Type=Bone,Name="deform_glute_r"),Weight=0.088),(Item=(Type=Bone,Name="thigh_r"),Weight=0.155),(Item=(Type=Bone,Name="pelvis"),Weight=0.196))
    Weight=1
   )
   > &RigUnit_ParentConstraint(
    AdvancedSettings=(RotationOrderForFilter=XZY)
    Child=(Type=Null,Name="dyn_bandoliersBotBack_5_r_ctrl_null")
    Parents=((Item=(Type=Bone,Name="thigh_twist_01_r"),Weight=0.643),(Item=(Type=Bone,Name="deform_groin_r"),Weight=0.014),(Item=(Type=Bone,Name="thigh_r"),Weight=0.27),(Item=(Type=Bone,Name="pelvis"),Weight=0.073))
    Weight=1
   )
   > &RigUnit_ParentConstraint(
    AdvancedSettings=(RotationOrderForFilter=XZY)
    Child=(Type=Null,Name="dyn_bandoliersBotFront_1_r_ctrl_null")
    Parents="((Item=(Type=Bone,Name=\"thigh_twist_01_r\"),Weight=0.066),(Item=(Type=Bone,Name=\"deform_groin_r\"),Weight=0.179),(Item=(Type=Bone,Name=\"thigh_twist_01_l\"),Weight=0.013),(Item=(Type=Bone,Name=\"deform_groin_l\"),Weight=0.021),(Item=(Type=Bone,Name=\"pelvis\"),Wei ...(truncated)"
    Weight=1
   )
   > &RigUnit_ParentConstraint(
    AdvancedSettings=(RotationOrderForFilter=XZY)
    Child=(Type=Null,Name="dyn_bandoliersBotFront_2_r_ctrl_null")
    Parents="((Item=(Type=Bone,Name=\"thigh_twist_01_r\"),Weight=0.113),(Item=(Type=Bone,Name=\"deform_groin_r\"),Weight=0.291),(Item=(Type=Bone,Name=\"thigh_twist_01_l\"),Weight=0.012),(Item=(Type=Bone,Name=\"deform_groin_l\"),Weight=0.021),(Item=(Type=Bone,Name=\"pelvis\"),Wei ...(truncated)"
    Weight=1
   )
   > &RigUnit_ParentConstraint(
    AdvancedSettings=(RotationOrderForFilter=XZY)
    Child=(Type=Null,Name="dyn_bandoliersBotFront_3_r_ctrl_null")
    Parents=((Item=(Type=Bone,Name="thigh_twist_01_r"),Weight=0.325),(Item=(Type=Bone,Name="deform_groin_r"),Weight=0.417),(Item=(Type=Bone,Name="thigh_r"),Weight=0.044),(Item=(Type=Bone,Name="pelvis"),Weight=0.214))
    Weight=1
   )
   > &RigUnit_ParentConstraint(
    AdvancedSettings=(RotationOrderForFilter=XZY)
    Child=(Type=Null,Name="dyn_bandoliersBotFront_4_r_ctrl_null")
    Parents=((Item=(Type=Bone,Name="thigh_twist_01_r"),Weight=0.488),(Item=(Type=Bone,Name="deform_groin_r"),Weight=0.281),(Item=(Type=Bone,Name="thigh_r"),Weight=0.133),(Item=(Type=Bone,Name="pelvis"),Weight=0.098))
    Weight=1
   )
   > &RigUnit_ParentConstraint(
    AdvancedSettings=(RotationOrderForFilter=XZY)
    Child=(Type=Null,Name="dyn_bandoliersBotFront_5_r_ctrl_null")
    Parents=((Item=(Type=Bone,Name="thigh_twist_01_r"),Weight=0.562),(Item=(Type=Bone,Name="deform_groin_r"),Weight=0.094),(Item=(Type=Bone,Name="thigh_r"),Weight=0.293),(Item=(Type=Bone,Name="pelvis"),Weight=0.05))
    Weight=1
   )
   > n2 = &RigUnit_ParentConstraint(
    AdvancedSettings=(InterpolationType=Average,RotationOrderForFilter=XZY)
    Child=(Type=Bone,Name="base_bandoliersBotMid_1_r")
    Parents=((Item=(Type=Bone,Name="thigh_twist_01_r"),Weight=0.572),(Item=(Type=Bone,Name="deform_groin_r"),Weight=0.038),(Item=(Type=Bone,Name="thigh_r"),Weight=0.345),(Item=(Type=Bone,Name="pelvis"),Weight=0.044))
    Weight=1
   )
   RigVMFunctionReturnNode(ExecutePin=@n2.ExecutePin)
  ]}
 CollapseNode_ContainedGraph {nodes [
   n1 = RigVMFunctionEntryNode()
   &RigUnit_ParentConstraint(
    AdvancedSettings=(RotationOrderForFilter=XZY)
    Child=(Type=Null,Name="dyn_shoulderStraps_1_l_ctrl_null")
    Parents=((Item=(Type=Bone,Name="spine_05"),Weight=1.0))
    Weight=1
    ExecutePin=@n1.ExecutePin
   )
   > &RigUnit_ParentConstraint(
    AdvancedSettings=(RotationOrderForFilter=XZY)
    Child=(Type=Null,Name="dyn_shoulderStraps_2_l_ctrl_null")
    Parents=((Item=(Type=Bone,Name="spine_01"),Weight=0.0005),(Item=(Type=Bone,Name="spine_02"),Weight=0.0042),(Item=(Type=Bone,Name="spine_03"),Weight=0.0136),(Item=(Type=Bone,Name="spine_04"),Weight=0.0162),(Item=(Type=Bone,Name="spine_05"),Weight=0.9655))
    Weight=1
   )
   > &RigUnit_ParentConstraint(
    AdvancedSettings=(RotationOrderForFilter=XZY)
    Child=(Type=Null,Name="dyn_shoulderStraps_3_l_ctrl_null")
    Parents="((Item=(Type=Bone,Name=\"pelvis\"),Weight=0.0007),(Item=(Type=Bone,Name=\"spine_01\"),Weight=0.005),(Item=(Type=Bone,Name=\"spine_02\"),Weight=0.0311),(Item=(Type=Bone,Name=\"spine_03\"),Weight=0.0819),(Item=(Type=Bone,Name=\"spine_04\"),Weight=0.083),(Item=(Type=Bo ...(truncated)"
    Weight=1
   )
   > &RigUnit_ParentConstraint(
    AdvancedSettings=(RotationOrderForFilter=XZY)
    Child=(Type=Null,Name="dyn_shoulderStraps_4_l_ctrl_null")
    Parents="((Item=(Type=Bone,Name=\"pelvis\"),Weight=0.0074),(Item=(Type=Bone,Name=\"spine_01\"),Weight=0.0357),(Item=(Type=Bone,Name=\"spine_02\"),Weight=0.1415),(Item=(Type=Bone,Name=\"spine_04\"),Weight=0.1661),(Item=(Type=Bone,Name=\"spine_03\"),Weight=0.238),(Item=(Type=B ...(truncated)"
    Weight=1
   )
   > &RigUnit_ParentConstraint(
    AdvancedSettings=(RotationOrderForFilter=XZY)
    Child=(Type=Null,Name="dyn_shoulderStraps_5_l_ctrl_null")
    Parents="((Item=(Type=Bone,Name=\"spine_05\"),Weight=0.0556),(Item=(Type=Bone,Name=\"pelvis\"),Weight=0.0583),(Item=(Type=Bone,Name=\"spine_04\"),Weight=0.0817),(Item=(Type=Bone,Name=\"spine_01\"),Weight=0.1796),(Item=(Type=Bone,Name=\"spine_03\"),Weight=0.2643),(Item=(Type= ...(truncated)"
    Weight=1
   )
   > &RigUnit_ParentConstraint(
    AdvancedSettings=(RotationOrderForFilter=XZY)
    Child=(Type=Null,Name="dyn_shoulderStraps_6_l_ctrl_null")
    Parents="((Item=(Type=Bone,Name=\"spine_05\"),Weight=0.0025),(Item=(Type=Bone,Name=\"spine_04\"),Weight=0.0143),(Item=(Type=Bone,Name=\"spine_03\"),Weight=0.1237),(Item=(Type=Bone,Name=\"pelvis\"),Weight=0.1656),(Item=(Type=Bone,Name=\"spine_01\"),Weight=0.3284),(Item=(Type= ...(truncated)"
    Weight=1
   )
   > &RigUnit_ParentConstraint(
    AdvancedSettings=(RotationOrderForFilter=XZY)
    Child=(Type=Null,Name="dyn_shoulderStraps_7_l_ctrl_null")
    Parents=((Item=(Type=Bone,Name="spine_04"),Weight=0.0003),(Item=(Type=Bone,Name="spine_03"),Weight=0.0081),(Item=(Type=Bone,Name="spine_02"),Weight=0.0853),(Item=(Type=Bone,Name="spine_01"),Weight=0.3334),(Item=(Type=Bone,Name="pelvis"),Weight=0.5729))
    Weight=1
   )
   > &RigUnit_ParentConstraint(
    AdvancedSettings=(RotationOrderForFilter=XZY)
    Child=(Type=Null,Name="dyn_shoulderStraps_1_r_ctrl_null")
    Parents=((Item=(Type=Bone,Name="spine_05"),Weight=1.0))
    Weight=1
   )
   > &RigUnit_ParentConstraint(
    AdvancedSettings=(RotationOrderForFilter=XZY)
    Child=(Type=Null,Name="dyn_shoulderStraps_2_r_ctrl_null")
    Parents=((Item=(Type=Bone,Name="spine_01"),Weight=0.0005),(Item=(Type=Bone,Name="spine_02"),Weight=0.0042),(Item=(Type=Bone,Name="spine_03"),Weight=0.0136),(Item=(Type=Bone,Name="spine_04"),Weight=0.0162),(Item=(Type=Bone,Name="spine_05"),Weight=0.9655))
    Weight=1
   )
   > &RigUnit_ParentConstraint(
    AdvancedSettings=(RotationOrderForFilter=XZY)
    Child=(Type=Null,Name="dyn_shoulderStraps_3_r_ctrl_null")
    Parents="((Item=(Type=Bone,Name=\"pelvis\"),Weight=0.0007),(Item=(Type=Bone,Name=\"spine_01\"),Weight=0.005),(Item=(Type=Bone,Name=\"spine_02\"),Weight=0.0312),(Item=(Type=Bone,Name=\"spine_03\"),Weight=0.082),(Item=(Type=Bone,Name=\"spine_04\"),Weight=0.0831),(Item=(Type=Bo ...(truncated)"
    Weight=1
   )
   > &RigUnit_ParentConstraint(
    AdvancedSettings=(RotationOrderForFilter=XZY)
    Child=(Type=Null,Name="dyn_shoulderStraps_4_r_ctrl_null")
    Parents="((Item=(Type=Bone,Name=\"pelvis\"),Weight=0.0074),(Item=(Type=Bone,Name=\"spine_01\"),Weight=0.0357),(Item=(Type=Bone,Name=\"spine_02\"),Weight=0.1414),(Item=(Type=Bone,Name=\"spine_04\"),Weight=0.1661),(Item=(Type=Bone,Name=\"spine_03\"),Weight=0.238),(Item=(Type=B ...(truncated)"
    Weight=1
   )
   > &RigUnit_ParentConstraint(
    AdvancedSettings=(RotationOrderForFilter=XZY)
    Child=(Type=Null,Name="dyn_shoulderStraps_5_r_ctrl_null")
    Parents="((Item=(Type=Bone,Name=\"spine_05\"),Weight=0.0543),(Item=(Type=Bone,Name=\"pelvis\"),Weight=0.059),(Item=(Type=Bone,Name=\"spine_04\"),Weight=0.0807),(Item=(Type=Bone,Name=\"spine_01\"),Weight=0.181),(Item=(Type=Bone,Name=\"spine_03\"),Weight=0.2633),(Item=(Type=Bo ...(truncated)"
    Weight=1
   )
   > &RigUnit_ParentConstraint(
    AdvancedSettings=(RotationOrderForFilter=XZY)
    Child=(Type=Null,Name="dyn_shoulderStraps_6_r_ctrl_null")
    Parents="((Item=(Type=Bone,Name=\"spine_05\"),Weight=0.0026),(Item=(Type=Bone,Name=\"spine_04\"),Weight=0.0145),(Item=(Type=Bone,Name=\"spine_03\"),Weight=0.125),(Item=(Type=Bone,Name=\"pelvis\"),Weight=0.1636),(Item=(Type=Bone,Name=\"spine_01\"),Weight=0.3272),(Item=(Type=B ...(truncated)"
    Weight=1
   )
   > n2 = &RigUnit_ParentConstraint(
    AdvancedSettings=(RotationOrderForFilter=XZY)
    Child=(Type=Null,Name="dyn_shoulderStraps_7_r_ctrl_null")
    Parents=((Item=(Type=Bone,Name="spine_04"),Weight=0.0003),(Item=(Type=Bone,Name="spine_03"),Weight=0.0078),(Item=(Type=Bone,Name="spine_02"),Weight=0.0838),(Item=(Type=Bone,Name="spine_01"),Weight=0.3322),(Item=(Type=Bone,Name="pelvis"),Weight=0.5759))
    Weight=1
   )
   RigVMFunctionReturnNode(ExecutePin=@n2.ExecutePin)
  ]}
 Multiply_3_ContainedGraph {nodes [
   n1 = RigVMFunctionEntryNode()
   RigVMFunctionReturnNode(Result=&RigVMFunction_MathDoubleMul(B=@n1.C, A=&RigVMFunction_MathDoubleMul(A=@n1.A, B=@n1.B).Result).Result)
  ]}
 Multiply_3_ContainedGraph {nodes [
   n1 = RigVMFunctionEntryNode()
   RigVMFunctionReturnNode(Result=&RigVMFunction_MathTransformMul(B=@n1.C, A=&RigVMFunction_MathTransformMul(A=@n1.A, B=@n1.B) {template="Multiply::Execute(in A,in B,out Result)"}.Result) {template="Multiply::Execute(in A,in B,out Result)"}.Result)
  ]}
 New Function_ContainedGraph {nodes [
   n1 = RigVMFunctionEntryNode()
   RigVMFunction_ControlFlowBranch = &RigVMFunction_ControlFlowBranch(ExecuteContext=@n1.ExecuteContext, Condition=@n1.debug)
   RigVMFunctionReturnNode(ExecuteContext=@RigVMFunction_ControlFlowBranch.Completed)
   RigVMFunction_Sequence = &RigVMFunction_Sequence(ExecuteContext=@RigVMFunction_ControlFlowBranch.True)
   RigUnit_ItemArray = &RigUnit_ItemArray(Items="((Type=Bone,Name=\"spine_01\"),(Type=Bone,Name=\"spine_02\"),(Type=Bone,Name=\"spine_03\"),(Type=Bone,Name=\"spine_04\"),(Type=Bone,Name=\"spine_05\"),(Type=Bone,Name=\"thigh_r\"),(Type=Bone,Name=\"calf_r\"),(Type=Bone,Name=\"foot_r\"),(Type=Bone,Name=\"thigh_l\"),(Type=Bon ...(truncated)")
   DISPATCH_RigVMDispatch_ArrayIterator = RigVMDispatchNode(Array=@RigUnit_ItemArray.Items, ExecuteContext=@RigVMFunction_Sequence.A) {template="DISPATCH_RigVMDispatch_ArrayIterator(in Array,out Element,out Index,out Count,out Ratio)"}
   Concat = /Script/RigVM.RigVMFunction_NameConcat(B=_ctrl, A=@DISPATCH_RigVMDispatch_ArrayIterator.Element.Name) {is=struct RigVMUnitNode, template="Concat::Execute(in A,in B,out Result)"}
   > &RigUnit_SetTransform(
    Item=(Type=Bone,Name="spine_01")
    Weight=1
    bInitial=False
    Item.Name=@DISPATCH_RigVMDispatch_ArrayIterator.Element.Name
    Value=&RigUnit_GetTransform(Item=(Type=Control,Name="spine_01_ctrl"), bInitial=False, Item.Name=@Concat.Result).Transform
   )
   RigVMFunction_MathQuaternionSwingTwist = &RigVMFunction_MathQuaternionSwingTwist(
    TwistAxis=(X=1,Y=0,Z=0)
    Input=&RigUnit_GetRelativeTransformForItem(Child=(Type=Bone,Name="thigh_l"), Parent=(Type=Null,Name="thigh_l_null"), bChildInitial=False, bParentInitial=False).RelativeTransform.Rotation
   )
   Inverse = /Script/RigVM.RigVMFunction_MathQuaternionInverse(Value=@RigVMFunction_MathQuaternionSwingTwist.Twist) {is=struct RigVMUnitNode, template="Inverse::Execute(in Value,out Result)"}
   &RigUnit_ModifyTransforms(
    ItemToModify=((Item=(Type=Bone,Name="thigh_twist_01_l"),Transform=(Rotation=(X=0,Y=0,Z=0,W=1),Translation=(X=0,Y=0,Z=0),Scale3D=(X=1,Y=1,Z=1))))
    Weight=1
    WeightMinimum=0
    ItemToModify.0.Transform.Rotation=@Inverse.Result
    ExecutePin=@RigVMFunction_Sequence.B
   )
   n2 = &RigUnit_GetRelativeTransformForItem(Child=(Type=Bone,Name="thigh_r"), Parent=(Type=Null,Name="thigh_r_null"), bChildInitial=False, bParentInitial=False)
   > &RigUnit_ModifyTransforms(
    ItemToModify=((Item=(Type=Bone,Name="thigh_twist_01_r"),Transform=(Rotation=(X=0,Y=0,Z=0,W=1),Translation=(X=0,Y=0,Z=0),Scale3D=(X=1,Y=1,Z=1))))
    Weight=1
    WeightMinimum=0
    ItemToModify.0.Transform.Rotation=&RigVMFunction_MathQuaternionInverse(Value=&RigVMFunction_MathQuaternionSwingTwist(TwistAxis=(X=1,Y=0,Z=0), Input=@n2.RelativeTransform.Rotation).Twist) {template="Inverse::Execute(in Value,out Result)"}.Result
   )
  ]}
 New Function_ContainedGraph {nodes [
   RigVMFunctionEntryNode()
   > RigVMFunction_Sequence = RigVMAggregateNode()
   &RigUnit_ParentConstraint(
    AdvancedSettings=(RotationOrderForFilter=XZY)
    Child=(Type=Bone,Name="deform_knee_out_base_l")
    Parents=((Item=(Type=Bone,Name="calf_l"),Weight=0.1),(Item=(Type=Bone,Name="thigh_l"),Weight=0.9))
    Weight=1
    ExecutePin=@RigVMFunction_Sequence.A
   )
   > &RigUnit_ParentConstraint(
    AdvancedSettings=(RotationOrderForFilter=XZY)
    Child=(Type=Bone,Name="deform_knee_in_base_l")
    Parents=((Item=(Type=Bone,Name="calf_l"),Weight=0.5),(Item=(Type=Bone,Name="thigh_l"),Weight=0.5))
    Weight=1
   )
   > DeltaRotation = RigVMFunctionReferenceNode(Rotation Order=ZYX, child=calf_l, max=100, parent=calf_l_null, parentType=Null, premult=1, primAxis=2, twistAxis=(X=1,Y=0,Z=0))
   > RigVMVariableNode(Value=@DeltaRotation.outputWeight) {variable=calfBend_in_l}
   > &RigUnit_ModifyTransforms(
    ItemToModify="((Item=(Type=Bone,Name=\"deform_knee_out_l\"),Transform=(Rotation=(X=0,Y=0,Z=0,W=1),Translation=(X=-3,Y=3,Z=0),Scale3D=(X=1,Y=1,Z=1))),(Item=(Type=Bone,Name=\"deform_knee_in_l\"),Transform=(Rotation=(X=0,Y=0,Z=0,W=1),Translation=(X=0,Y=-12,Z=0),Scale3D=(X=1,Y= ...(truncated)"
    Weight=@DeltaRotation.outputWeight
   )
   > &RigUnit_ParentConstraint(
    AdvancedSettings=(RotationOrderForFilter=XZY)
    Child=(Type=Bone,Name="deform_groin_base_l")
    Parents=((Item=(Type=Bone,Name="thigh_twist_01_l"),Weight=0.5),(Item=(Type=Bone,Name="pelvis"),Weight=0.5))
    Weight=1
   )
   > &RigUnit_ParentConstraint(
    AdvancedSettings=(RotationOrderForFilter=XZY)
    Child=(Type=Bone,Name="deform_glute_base_l")
    Parents=((Item=(Type=Bone,Name="thigh_twist_01_l"),Weight=0.25),(Item=(Type=Bone,Name="pelvis"),Weight=0.75))
    Weight=1
   )
   > DeltaRotation_2 = RigVMFunctionReferenceNode(Rotation Order=ZYX, child=thigh_l, max=-100, parent=thigh_l_null, parentType=Null, premult=1, primAxis=2, twistAxis=(X=1,Y=0,Z=0))
   > RigVMVariableNode(Value=@DeltaRotation_2.outputWeight) {variable=thighBend_out_l}
   > &RigUnit_ModifyTransforms(
    ItemToModify="((Item=(Type=Bone,Name=\"deform_glute_l\"),Transform=(Rotation=(X=0,Y=0,Z=0,W=1),Translation=(X=-7,Y=0,Z=0),Scale3D=(X=1,Y=1.5,Z=1))),(Item=(Type=Bone,Name=\"deform_groin_l\"),Transform=(Rotation=(X=0,Y=0,Z=0,W=1),Translation=(X=-3,Y=15,Z=0),Scale3D=(X=0.5,Y=1 ...(truncated)"
    Weight=@DeltaRotation_2.outputWeight
   )
   &RigUnit_ParentConstraint(
    AdvancedSettings=(RotationOrderForFilter=XZY)
    Child=(Type=Bone,Name="deform_knee_out_base_r")
    Parents=((Item=(Type=Bone,Name="calf_r"),Weight=0.1),(Item=(Type=Bone,Name="thigh_r"),Weight=0.9))
    Weight=1
    ExecutePin=@RigVMFunction_Sequence.B
   )
   > &RigUnit_ParentConstraint(
    AdvancedSettings=(RotationOrderForFilter=XZY)
    Child=(Type=Bone,Name="deform_knee_in_base_r")
    Parents=((Item=(Type=Bone,Name="calf_r"),Weight=0.5),(Item=(Type=Bone,Name="thigh_r"),Weight=0.5))
    Weight=1
   )
   > DeltaRotation_1 = RigVMFunctionReferenceNode(Rotation Order=ZYX, child=calf_r, max=100, parent=calf_r_null, parentType=Null, premult=1, primAxis=2, twistAxis=(X=1,Y=0,Z=0))
   > RigVMVariableNode(Value=@DeltaRotation_1.outputWeight) {variable=calfBend_in_r}
   > &RigUnit_ModifyTransforms(
    ItemToModify="((Item=(Type=Bone,Name=\"deform_knee_out_r\"),Transform=(Rotation=(X=0,Y=0,Z=0,W=1),Translation=(X=3,Y=-3,Z=0),Scale3D=(X=1,Y=1,Z=1))),(Item=(Type=Bone,Name=\"deform_knee_in_r\"),Transform=(Rotation=(X=0,Y=0,Z=0,W=1),Translation=(X=0,Y=12,Z=0),Scale3D=(X=1,Y=1 ...(truncated)"
    Weight=@DeltaRotation_1.outputWeight
   )
   > &RigUnit_ParentConstraint(
    AdvancedSettings=(RotationOrderForFilter=XZY)
    Child=(Type=Bone,Name="deform_groin_base_r")
    Parents=((Item=(Type=Bone,Name="thigh_twist_01_r"),Weight=0.7),(Item=(Type=Bone,Name="pelvis"),Weight=0.3))
    Weight=1
   )
   > &RigUnit_ParentConstraint(
    AdvancedSettings=(RotationOrderForFilter=XZY)
    Child=(Type=Bone,Name="deform_glute_base_r")
    Parents=((Item=(Type=Bone,Name="thigh_twist_01_r"),Weight=0.25),(Item=(Type=Bone,Name="pelvis"),Weight=0.75))
    Weight=1
   )
   > DeltaRotation_3 = RigVMFunctionReferenceNode(Rotation Order=ZYX, child=thigh_r, max=-100, parent=thigh_r_null, parentType=Null, premult=1, primAxis=2, twistAxis=(X=1,Y=0,Z=0))
   > RigVMVariableNode(Value=@DeltaRotation_3.outputWeight) {variable=thighBend_out_r}
   > &RigUnit_ModifyTransforms(
    ItemToModify="((Item=(Type=Bone,Name=\"deform_glute_r\"),Transform=(Rotation=(X=0,Y=0,Z=0,W=1),Translation=(X=7,Y=0,Z=0),Scale3D=(X=1,Y=1.5,Z=1))),(Item=(Type=Bone,Name=\"deform_groin_r\"),Transform=(Rotation=(X=0,Y=0,Z=0,W=1),Translation=(X=3,Y=-15,Z=0),Scale3D=(X=0.5,Y=1, ...(truncated)"
    Weight=@DeltaRotation_3.outputWeight
   )
   RigVMFunctionReturnNode(ExecuteContext=@RigVMFunction_Sequence.E)
  ]}
 New Function_ContainedGraph {nodes [
   n1 = RigVMFunctionEntryNode()
   RigVMFunction_MathBoolToInteger = &RigVMFunction_MathBoolToInteger(Value=@n1.onlyTwist)
   RigVMFunction_MathQuaternionSwingTwist = &RigVMFunction_MathQuaternionSwingTwist(
    TwistAxis=@n1.twistAxis
    Input=&RigUnit_GetRelativeTransformForItem(
     Child=(Type=Bone,Name="None")
     Parent=(Type=Bone,Name="None")
     bChildInitial=False
     bParentInitial=False
     Child.Name=@n1.child
     Parent.Name=@n1.parent
     Parent.Type=@n1.parentType
    ).RelativeTransform.Rotation
   )
   RigVMFunction_MathQuaternionToEuler = &RigVMFunction_MathQuaternionToEuler(
    Value=RigVMDispatchNode(
     Values=((),())
     Index=@RigVMFunction_MathBoolToInteger.Result
     Values.0=@RigVMFunction_MathQuaternionSwingTwist.Swing
     Values.1=@RigVMFunction_MathQuaternionSwingTwist.Twist
    ) {template="DISPATCH_RigVMDispatch_SelectInt32(in Index,in Values,out Result)"}.Result
    RotationOrder=@n1.Rotation Order
   )
   > DISPATCH_RigVMDispatch_Print = RigVMDispatchNode(ScreenColor=(R=1,G=1,B=1,A=1), ScreenDuration=1.5, Enabled=@n1.debugRotation, Value=@RigVMFunction_MathQuaternionToEuler.Result) {template="DISPATCH_RigVMDispatch_Print(in Prefix,in Value,in Enabled,in ScreenDuration,in ScreenColor)"}
   Clamp = &RigVMFunction_MathIntClamp(Value=@n1.primAxis)
   DISPATCH_RigVMDispatch_SelectInt32 = RigVMDispatchNode(
    Values=(0,0,0)
    Index=@Clamp.Result
    Values.0=@RigVMFunction_MathQuaternionToEuler.Result.X
    Values.1=@RigVMFunction_MathQuaternionToEuler.Result.Y
    Values.2=@RigVMFunction_MathQuaternionToEuler.Result.Z
   ) {template="DISPATCH_RigVMDispatch_SelectInt32(in Index,in Values,out Result)"}
   Remap = &RigVMFunction_MathDoubleRemap(
    TargetMaximum=1
    TargetMinimum=0
    bClamp=False
    Value=@DISPATCH_RigVMDispatch_SelectInt32.Result
    SourceMaximum=@n1.max
    SourceMinimum=@n1.min
   )
   Multiply = &RigVMFunction_MathDoubleMul(A=&RigVMFunction_MathDoubleClamp(Value=@Remap.Result).Result, B=@n1.premult)
   > RigVMFunctionReturnNode(outputWeight=@Multiply.Result)
  ]}
 New Function_ContainedGraph {nodes [
   RigVMFunctionEntryNode()
   > &RigUnit_RotationConstraintLocalSpaceOffset(
    AdvancedSettings=(RotationOrderForFilter=XZY)
    Child=(Type=Bone,Name="calfRePos_l")
    Filter=(bX=True,bY=True,bZ=True)
    Parents=((Item=(Type=Bone,Name="calf_l"),Weight=1))
    Weight=1
   )
   > &RigUnit_RotationConstraintLocalSpaceOffset(
    AdvancedSettings=(RotationOrderForFilter=XZY)
    Child=(Type=Bone,Name="calfRePos_r")
    Filter=(bX=True,bY=True,bZ=True)
    Parents=((Item=(Type=Bone,Name="calf_r"),Weight=1))
    Weight=1
   )
   > &RigUnit_AimConstraintLocalSpaceOffset(
    AimAxis=(X=0,Y=-1,Z=0)
    Child=(Type=Bone,Name="footRePos_l")
    Filter=(bX=True,bY=true,bZ=true)
    Parents=((Item=(Type=Bone,Name="ball_l"),Weight=1))
    UpAxis=(X=0,Y=0,Z=-1)
    Weight=1
    WorldUp=(Target=(X=0,Y=0,Z=1),Kind=Location,Space=(Type=Null,Name="footRePos_l_upVector"))
   )
   > &RigUnit_AimConstraintLocalSpaceOffset(
    AimAxis=(X=0,Y=1,Z=0)
    Child=(Type=Bone,Name="footRePos_r")
    Filter=(bX=True,bY=true,bZ=true)
    Parents=((Item=(Type=Bone,Name="ball_r"),Weight=1))
    Weight=1
    WorldUp=(Target=(X=0,Y=0,Z=1),Kind=Location,Space=(Type=Null,Name="footRePos_r_upVector"))
   )
   > &RigUnit_ParentConstraint(
    AdvancedSettings=(RotationOrderForFilter=XZY)
    Child=(Type=Bone,Name="kin_frontPack_l")
    Parents=((Item=(Type=Bone,Name="spine_04"),Weight=0.0006),(Item=(Type=Bone,Name="spine_03"),Weight=0.0099),(Item=(Type=Bone,Name="spine_02"),Weight=0.0473),(Item=(Type=Bone,Name="spine_01"),Weight=0.2514),(Item=(Type=Bone,Name="pelvis"),Weight=0.6907))
    Weight=1
   )
   > &RigUnit_ParentConstraint(
    AdvancedSettings=(RotationOrderForFilter=XZY)
    Child=(Type=Bone,Name="kin_frontPack_r")
    Parents=((Item=(Type=Bone,Name="spine_04"),Weight=0.0006),(Item=(Type=Bone,Name="spine_03"),Weight=0.0099),(Item=(Type=Bone,Name="spine_02"),Weight=0.0474),(Item=(Type=Bone,Name="spine_01"),Weight=0.2516),(Item=(Type=Bone,Name="pelvis"),Weight=0.6905))
    Weight=1
   )
   > DeltaRotation_1 = RigVMFunctionReferenceNode(Rotation Order=ZYX, child=thigh_l, max=-100, min=-40, parent=thigh_l_null, parentType=Null, premult=1, primAxis=2, twistAxis=(X=1,Y=0,Z=0))
   > RigVMVariableNode(Value=@DeltaRotation_1.outputWeight) {variable=thighBend_out_n40_n100}
   > &RigUnit_ModifyTransforms(
    ItemToModify=((Item=(Type=Bone,Name="kin_frontPack_l"),Transform=(Rotation=(X=0,Y=-0.2079,Z=0,W=0.9781),Translation=(X=0.95,Y=0,Z=7),Scale3D=(X=1,Y=1,Z=1))))
    WeightMinimum=0
    Weight=@DeltaRotation_1.outputWeight
   )
   > DeltaRotation_4 = RigVMFunctionReferenceNode(Rotation Order=ZYX, child=thigh_r, max=-100, min=-40, parent=thigh_r_null, parentType=Null, premult=1, primAxis=2, twistAxis=(X=1,Y=0,Z=0))
   > n1 = &RigUnit_ModifyTransforms(
    ItemToModify=((Item=(Type=Bone,Name="kin_frontPack_r"),Transform=(Rotation=(X=0,Y=-0.2079,Z=0,W=0.9781),Translation=(X=0.95,Y=0,Z=7),Scale3D=(X=1,Y=1,Z=1))))
    WeightMinimum=0
    Weight=@DeltaRotation_4.outputWeight
   )
   RigVMFunction_ControlFlowBranch = &RigVMFunction_ControlFlowBranch(ExecuteContext=@n1.ExecutePin, Condition=RigVMVariableNode() {variable=pub isWearingShoes}.Value)
   RigVMFunctionReturnNode(ExecuteContext=@RigVMFunction_ControlFlowBranch.Completed)
   DeltaRotation_2 = RigVMFunctionReferenceNode(
    Rotation Order=ZYX
    child=calf_l
    debugRotation=false
    max=160
    min=100
    parent=calf_l_null
    parentType=Null
    premult=0.9
    primAxis=2
    twistAxis=(X=1,Y=0,Z=0)
    ExecuteContext=@RigVMFunction_ControlFlowBranch.False
   )
   > &RigUnit_ModifyTransforms(
    ItemToModify="((Item=(Type=Bone,Name=\"calfRePos_bend_1_l\"),Transform=(Rotation=(X=0,Y=0,Z=-0.4384,W=0.8988),Translation=(X=0,Y=0,Z=0),Scale3D=(X=1,Y=1,Z=1))),(Item=(Type=Bone,Name=\"calfRePos_bend_2_l\"),Transform=(Rotation=(X=0,Y=0,Z=-0.2672,W=0.9636),Translation=(X=1.8, ...(truncated)"
    WeightMinimum=0
    Weight=@DeltaRotation_2.outputWeight
   )
   > DeltaRotation_3 = RigVMFunctionReferenceNode(
    Rotation Order=ZYX
    child=calf_r
    debugRotation=false
    max=160
    min=100
    parent=calf_r_null
    parentType=Null
    premult=0.9
    primAxis=2
    twistAxis=(X=1,Y=0,Z=0)
   )
   > &RigUnit_ModifyTransforms(
    ItemToModify="((Item=(Type=Bone,Name=\"calfRePos_bend_1_r\"),Transform=(Rotation=(X=0,Y=0,Z=-0.4384,W=0.8988),Translation=(X=0,Y=0,Z=0),Scale3D=(X=1,Y=1,Z=1))),(Item=(Type=Bone,Name=\"calfRePos_bend_2_r\"),Transform=(Rotation=(X=0,Y=0,Z=-0.2672,W=0.9636),Translation=(X=-1.8 ...(truncated)"
    WeightMinimum=0
    Weight=@DeltaRotation_3.outputWeight
   )
  ]}
 New Function_ContainedGraph {nodes [
   RigVMFunctionEntryNode()
   > RigVMVariableNode(Value=RigVMDispatchNode(False=true, True=false, Condition=RigVMVariableNode() {variable=pub isWearingGauntlets}.Value) {template="DISPATCH_RigVMDispatch_If(in Condition,in True,in False,out Result)"}.Result) {variable=enableBendSystem}
   n1 = &RigVMFunction_MathBoolNot(Value=RigVMVariableNode() {variable=enableBendSystem}.Value)
   > TwoBoneIKWithSoftness = RigVMFunctionReferenceNode(
    effectorBone=(Type=Bone,Name="lowerarmBend_l_end")
    firstBone=(Type=Bone,Name="upperarmBend_l")
    firstBoneOrig=(Type=Bone,Name="upperarm_l")
    ikEffector=(Type=Bone,Name="hand_ik_l")
    itemStartStretch=(Type=Null,Name="upperarm_l_null")
    offsetPv=(X=0,Y=0,Z=0)
    poleVector=(Type=Control,Name="arm_l_pv_ctrl")
    primary Axis=(X=1,Y=0,Z=0)
    secondBone=(Type=Bone,Name="lowerarmBend_l")
    secondBoneOrig=(Type=Bone,Name="lowerarm_l")
    secondary Axis=(X=0,Y=0,Z=-1)
    softness=2.5
    lockPV=@n1.Result
   )
   > TwoBoneIKWithSoftness_1 = RigVMFunctionReferenceNode(
    effectorBone=(Type=Bone,Name="lowerarmBend_r_end")
    firstBone=(Type=Bone,Name="upperarmBend_r")
    firstBoneOrig=(Type=Bone,Name="upperarm_r")
    ikEffector=(Type=Bone,Name="hand_ik_r")
    itemStartStretch=(Type=Null,Name="upperarm_r_null")
    offsetPv=(X=0,Y=0,Z=0)
    poleVector=(Type=Control,Name="arm_r_pv_ctrl")
    primary Axis=(X=-1,Y=0,Z=0)
    secondBone=(Type=Bone,Name="lowerarmBend_r")
    secondBoneOrig=(Type=Bone,Name="lowerarm_r")
    secondary Axis=(X=0,Y=0,Z=1)
    softness=2.5
    lockPV=@n1.Result
   )
   > RigVMFunction_Sequence = RigVMAggregateNode()
   RigVMFunctionReturnNode(ExecuteContext=@RigVMFunction_Sequence.C)
   n2 = RigVMVariableNode() {variable=enableBendSystem}
   preCorrectives_l_arm = RigVMCollapseNode(enable=@n2.Value, ExecutePin=@RigVMFunction_Sequence.A)
   constraints_l_arm = RigVMCollapseNode(ExecuteContext=@preCorrectives_l_arm.ExecutePin)
   > correctives_l_arm = RigVMCollapseNode(enable=@n2.Value)
   RigUnit_ItemArray = &RigUnit_ItemArray(Items=((Type=Control,Name="arm_bend_1_l_ctrl"),(Type=Control,Name="arm_bend_2_l_ctrl"),(Type=Control,Name="arm_bend_3_l_ctrl"),(Type=Control,Name="arm_bend_4_l_ctrl"),(Type=Control,Name="arm_bend_5_l_ctrl")))
   n3 = &RigUnit_ItemArray(Items="((Type=Bone,Name=\"arm_bend_1_l\"),(Type=Bone,Name=\"arm_bend_2_l\"),(Type=Bone,Name=\"arm_bend_3_l\"),(Type=Bone,Name=\"arm_bend_4_l\"),(Type=Bone,Name=\"arm_bend_5_l\"),(Type=Bone,Name=\"arm_bend_6_l\"),(Type=Bone,Name=\"arm_bend_7_l\"),(Type=Bone,Name=\"arm_bend_8_l\") ...(truncated)")
   > SplineIK = RigVMFunctionReferenceNode(
    Primary Axis=(X=1,Y=0,Z=0)
    Samples Per Segment=3
    Secondary Spline Direction=(X=0,Y=5,Z=0)
    Spline Mode=BSpline
    Stretch=true
    Up Axis=(X=0,Y=1,Z=0)
    debug=false
    Controls=@RigUnit_ItemArray.Items
    Bones=@n3.Items
   )
   > DeltaQuatExtraction = RigVMFunctionReferenceNode(child=(Type=Bone,Name="upperarm_l"), invert=true, onlyTwist=true, parent=(Type=Null,Name="upperarm_l_null"), twistAxis=(X=1,Y=0,Z=0))
   n4 = &RigUnit_ItemArray(Items="((Type=Bone,Name=\"arm_bend_1_l\"),(Type=Bone,Name=\"arm_bend_2_l\"),(Type=Bone,Name=\"arm_bend_3_l\"),(Type=Bone,Name=\"arm_bend_4_l\"),(Type=Bone,Name=\"arm_bend_5_l\"),(Type=Bone,Name=\"arm_bend_6_l\"),(Type=Bone,Name=\"arm_bend_7_l\"),(Type=Bone,Name=\"arm_bend_8_l\") ...(truncated)")
   > SplineTwistTopBot_4 = RigVMFunctionReferenceNode(botRotQuat=(), enable=true, topRotQuat=@DeltaQuatExtraction.outputQuat, chainArray=@n4.Items)
   > DeltaQuatExtraction_1 = RigVMFunctionReferenceNode(child=(Type=Bone,Name="hand_l"), invert=false, onlyTwist=true, parent=(Type=Null,Name="hand_l_null"), twistAxis=(X=1,Y=0,Z=0))
   > SplineTwistTopBot = RigVMFunctionReferenceNode(
    topRotQuat=()
    botRotQuat=@DeltaQuatExtraction_1.outputQuat
    enable=RigVMVariableNode() {variable=enableBendSystem}.Value
    chainArray=@n4.Items
   )
   n5 = RigVMVariableNode() {variable=enableBendSystem}
   preCorrectives_r_arm = RigVMCollapseNode(enable=@n5.Value, ExecutePin=@RigVMFunction_Sequence.B)
   constraints_r_arm = RigVMCollapseNode(ExecuteContext=@preCorrectives_r_arm.ExecutePin)
   correctives_r_arm = RigVMCollapseNode(ExecutePin=@constraints_r_arm.ExecuteContext, enable=@n5.Value)
   n6 = &RigUnit_ItemArray(Items="((Type=Bone,Name=\"arm_bend_1_r\"),(Type=Bone,Name=\"arm_bend_2_r\"),(Type=Bone,Name=\"arm_bend_3_r\"),(Type=Bone,Name=\"arm_bend_4_r\"),(Type=Bone,Name=\"arm_bend_5_r\"),(Type=Bone,Name=\"arm_bend_6_r\"),(Type=Bone,Name=\"arm_bend_7_r\"),(Type=Bone,Name=\"arm_bend_8_r\") ...(truncated)")
   n7 = &RigUnit_ItemArray(Items=((Type=Control,Name="arm_bend_1_r_ctrl"),(Type=Control,Name="arm_bend_2_r_ctrl"),(Type=Control,Name="arm_bend_3_r_ctrl"),(Type=Control,Name="arm_bend_4_r_ctrl"),(Type=Control,Name="arm_bend_5_r_ctrl")))
   SplineIK_1 = RigVMFunctionReferenceNode(
    Primary Axis=(X=-1,Y=0,Z=0)
    Samples Per Segment=3
    Secondary Spline Direction=(X=0,Y=5,Z=0)
    Spline Mode=BSpline
    Stretch=true
    Up Axis=(X=0,Y=1,Z=0)
    debug=false
    Bones=@n6.Items
    Controls=@n7.Items
    ExecuteContext=@correctives_r_arm.ExecutePin
   )
   > DeltaQuatExtraction_2 = RigVMFunctionReferenceNode(child=(Type=Bone,Name="upperarm_r"), invert=true, onlyTwist=true, parent=(Type=Null,Name="upperarm_r_null"), twistAxis=(X=1,Y=0,Z=0))
   n8 = &RigUnit_ItemArray(Items="((Type=Bone,Name=\"arm_bend_1_r\"),(Type=Bone,Name=\"arm_bend_2_r\"),(Type=Bone,Name=\"arm_bend_3_r\"),(Type=Bone,Name=\"arm_bend_4_r\"),(Type=Bone,Name=\"arm_bend_5_r\"),(Type=Bone,Name=\"arm_bend_6_r\"),(Type=Bone,Name=\"arm_bend_7_r\"),(Type=Bone,Name=\"arm_bend_8_r\") ...(truncated)")
   > SplineTwistTopBot_5 = RigVMFunctionReferenceNode(botRotQuat=(), enable=true, topRotQuat=@DeltaQuatExtraction_2.outputQuat, chainArray=@n8.Items)
   > DeltaQuatExtraction_3 = RigVMFunctionReferenceNode(child=(Type=Bone,Name="hand_r"), invert=false, onlyTwist=true, parent=(Type=Null,Name="hand_r_null"), twistAxis=(X=1,Y=0,Z=0))
   > SplineTwistTopBot_1 = RigVMFunctionReferenceNode(
    topRotQuat=()
    botRotQuat=@DeltaQuatExtraction_3.outputQuat
    enable=RigVMVariableNode() {variable=enableBendSystem}.Value
    chainArray=@n8.Items
   )
  ]}
 New Function_ContainedGraph {nodes [
   n1 = RigVMFunctionEntryNode()
   RigVMFunction_ControlFlowBranch = &RigVMFunction_ControlFlowBranch(ExecuteContext=@n1.ExecuteContext, Condition=@n1.enable)
   RigVMFunctionReturnNode(ExecuteContext=@RigVMFunction_ControlFlowBranch.Completed)
   For_Each = RigVMDispatchNode(Array=@n1.chainArray, ExecuteContext=@RigVMFunction_ControlFlowBranch.True) {template="DISPATCH_RigVMDispatch_ArrayIterator(in Array,out Element,out Index,out Count,out Ratio)"}
   Interpolate = /Script/RigVM.RigVMFunction_MathQuaternionSlerp(T=@For_Each.Ratio, B=@n1.botRotQuat, A=@n1.topRotQuat) {is=struct RigVMUnitNode, template="Interpolate::Execute(in A,in B,in T,out Result)"}
   > &RigUnit_OffsetTransformForItem(
    OffsetTransform=(Rotation=(X=0,Y=0,Z=0,W=1),Translation=(X=0,Y=0,Z=0),Scale3D=(X=1,Y=1,Z=1))
    Weight=1
    bPropagateToChildren=false
    OffsetTransform.Rotation=@Interpolate.Result
    Item=@For_Each.Element
   )
  ]}
 New Function_ContainedGraph {nodes [
   n1 = RigVMFunctionEntryNode()
   RigVMFunction_MathBoolToInteger = &RigVMFunction_MathBoolToInteger(Value=@n1.onlyTwist)
   RigVMFunction_MathQuaternionSwingTwist = &RigVMFunction_MathQuaternionSwingTwist(
    TwistAxis=@n1.twistAxis
    Input=&RigUnit_GetRelativeTransformForItem(bChildInitial=False, bParentInitial=False, Child=@n1.child, Parent=@n1.parent).RelativeTransform.Rotation
   )
   n2 = RigVMDispatchNode(
    Values=((),())
    Index=@RigVMFunction_MathBoolToInteger.Result
    Values.0=@RigVMFunction_MathQuaternionSwingTwist.Swing
    Values.1=@RigVMFunction_MathQuaternionSwingTwist.Twist
   ) {template="DISPATCH_RigVMDispatch_SelectInt32(in Index,in Values,out Result)"}
   Inverse = &RigVMFunction_MathQuaternionInverse(Value=@n2.Result) {template="Inverse::Execute(in Value,out Result)"}
   > RigVMFunctionReturnNode(outputQuat=RigVMDispatchNode(True=@Inverse.Result, Condition=@n1.invert, False=@n2.Result) {template="DISPATCH_RigVMDispatch_If(in Condition,in True,in False,out Result)"}.Result)
  ]}
 New Function_ContainedGraph {nodes [
   RigVMFunctionEntryNode()
   > &RigUnit_ParentConstraint(
    AdvancedSettings=(RotationOrderForFilter=XZY)
    Child=(Type=Bone,Name="base_armPacks_l")
    Filter=(TranslationFilter=(bX=True,bY=True,bZ=True),RotationFilter=(bX=false,bY=false,bZ=false),ScaleFilter=(bX=True,bY=True,bZ=True))
    Parents="((Item=(Type=Bone,Name=\"arm_bend_4_l\"),Weight=0.0023),(Item=(Type=Bone,Name=\"spine_05\"),Weight=0.0302),(Item=(Type=Bone,Name=\"arm_bend_3_l\"),Weight=0.0717),(Item=(Type=Bone,Name=\"arm_bend_1_l\"),Weight=0.427),(Item=(Type=Bone,Name=\"arm_bend_2_l\"),Weight=0.4 ...(truncated)"
    Weight=1
   )
   > &RigUnit_AimConstraintLocalSpaceOffset(
    Child=(Type=Bone,Name="base_armPacks_l")
    Filter=(bX=True,bY=True,bZ=True)
    Parents=((Item=(Type=Bone,Name="arm_bend_4_l"),Weight=1))
    Weight=1
    WorldUp=(Target=(X=0,Y=0,Z=1),Kind=Location,Space=(Type=Bone,Name="arm_bend_4_l"))
   )
   > &RigUnit_ParentConstraint(
    AdvancedSettings=(RotationOrderForFilter=XZY)
    Child=(Type=Bone,Name="base_armPacks_r")
    Filter=(TranslationFilter=(bX=True,bY=True,bZ=True),RotationFilter=(bX=false,bY=false,bZ=false),ScaleFilter=(bX=True,bY=True,bZ=True))
    Parents="((Item=(Type=Bone,Name=\"arm_bend_4_r\"),Weight=0.0023),(Item=(Type=Bone,Name=\"spine_05\"),Weight=0.0302),(Item=(Type=Bone,Name=\"arm_bend_3_r\"),Weight=0.0717),(Item=(Type=Bone,Name=\"arm_bend_1_r\"),Weight=0.427),(Item=(Type=Bone,Name=\"arm_bend_2_r\"),Weight=0.4 ...(truncated)"
    Weight=1
   )
   > &RigUnit_AimConstraintLocalSpaceOffset(
    Child=(Type=Bone,Name="base_armPacks_r")
    Filter=(bX=True,bY=True,bZ=True)
    Parents=((Item=(Type=Bone,Name="arm_bend_4_r"),Weight=1))
    Weight=1
    WorldUp=(Target=(X=0,Y=0,Z=1),Kind=Location,Space=(Type=Bone,Name="arm_bend_4_r"))
   )
   > &RigUnit_ParentConstraint(
    AdvancedSettings=(RotationOrderForFilter=XZY)
    Child=(Type=Bone,Name="base_calfOrn_l")
    Parents="((Item=(Type=Bone,Name=\"calfRePos_bend_1_r\"),Weight=0.0001),(Item=(Type=Bone,Name=\"calfRePos_bend_2_r\"),Weight=0.0047),(Item=(Type=Bone,Name=\"calfRePos_bend_3_r\"),Weight=0.0274),(Item=(Type=Bone,Name=\"calfRePos_bend_4_r\"),Weight=0.1924),(Item=(Type=Bone,Na ...(truncated)"
    Weight=1
   )
   > &RigUnit_ParentConstraint(
    AdvancedSettings=(RotationOrderForFilter=XZY)
    Child=(Type=Bone,Name="kin_legsFitter_l")
    Parents="((Item=(Type=Bone,Name=\"calfRePos_bend_2_l\"),Weight=0.0056),(Item=(Type=Bone,Name=\"deform_knee_in_l\"),Weight=0.0295),(Item=(Type=Bone,Name=\"thigh_twist_01_l\"),Weight=0.0499),(Item=(Type=Bone,Name=\"deform_knee_out_l\"),Weight=0.1092),(Item=(Type=Bone,Name=\"c ...(truncated)"
    Weight=1
   )
   > &RigUnit_ParentConstraint(
    AdvancedSettings=(RotationOrderForFilter=XZY)
    Child=(Type=Bone,Name="kin_legsFitter_r")
    Parents="((Item=(Type=Bone,Name=\"calfRePos_bend_2_r\"),Weight=0.0079),(Item=(Type=Bone,Name=\"deform_knee_in_r\"),Weight=0.028),(Item=(Type=Bone,Name=\"thigh_twist_01_r\"),Weight=0.0471),(Item=(Type=Bone,Name=\"deform_knee_out_r\"),Weight=0.1127),(Item=(Type=Bone,Name=\"ca ...(truncated)"
    Weight=1
   )
   > &RigUnit_ParentConstraint(
    AdvancedSettings=(RotationOrderForFilter=XZY)
    Child=(Type=Bone,Name="base_frontLogo")
    Parents="((Item=(Type=Bone,Name=\"spine_05\"),Weight=0.0043),(Item=(Type=Bone,Name=\"spine_04\"),Weight=0.0159),(Item=(Type=Bone,Name=\"kin_frontPack_r\"),Weight=0.0298),(Item=(Type=Bone,Name=\"spine_03\"),Weight=0.0833),(Item=(Type=Bone,Name=\"spine_02\"),Weight=0.1615),(It ...(truncated)"
    Weight=1
   )
   > &RigUnit_ParentConstraint(
    AdvancedSettings=(RotationOrderForFilter=XZY)
    Child=(Type=Bone,Name="base_back_zipperSlider")
    Parents="((Item=(Type=Bone,Name=\"pelvis\"),Weight=0.0002),(Item=(Type=Bone,Name=\"spine_01\"),Weight=0.0013),(Item=(Type=Bone,Name=\"spine_02\"),Weight=0.0097),(Item=(Type=Bone,Name=\"spine_03\"),Weight=0.0304),(Item=(Type=Bone,Name=\"spine_04\"),Weight=0.0374),(Item=(Type= ...(truncated)"
    Weight=1
   )
   > &RigUnit_ParentConstraint(
    AdvancedSettings=(RotationOrderForFilter=XZY)
    Child=(Type=Bone,Name="base_zipperFront_A")
    Parents="((Item=(Type=Bone,Name=\"spine_05\"),Weight=0.0658),(Item=(Type=Bone,Name=\"pelvis\"),Weight=0.0831),(Item=(Type=Bone,Name=\"spine_04\"),Weight=0.1188),(Item=(Type=Bone,Name=\"spine_01\"),Weight=0.1797),(Item=(Type=Bone,Name=\"spine_03\"),Weight=0.2726),(Item=(Type= ...(truncated)"
    Weight=1
   )
   > &RigUnit_ParentConstraint(
    AdvancedSettings=(RotationOrderForFilter=XZY)
    Child=(Type=Bone,Name="base_zipperFront_B")
    Parents="((Item=(Type=Bone,Name=\"kin_frontPack_r\"),Weight=0.0089),(Item=(Type=Bone,Name=\"spine_05\"),Weight=0.0205),(Item=(Type=Bone,Name=\"spine_04\"),Weight=0.0534),(Item=(Type=Bone,Name=\"kin_frontPack_l\"),Weight=0.1218),(Item=(Type=Bone,Name=\"pelvis\"),Weight=0.1262 ...(truncated)"
    Weight=1
   )
   > &RigUnit_AimConstraintLocalSpaceOffset(
    Child=(Type=Bone,Name="base_zipperFront_A")
    Filter=(bX=True,bY=True,bZ=True)
    Parents=((Item=(Type=Null,Name="base_zipperFront_A_aimTarget_null"),Weight=1))
    Weight=1
    WorldUp=(Target=(X=0,Y=0,Z=-1),Kind=Direction,Space=(Type=Bone,Name="spine_02"))
    bMaintainOffset=true
   )
   n1 = RigVMVariableNode() {variable=twistSpine_0_50}
   > &RigUnit_ModifyTransforms(
    ItemToModify=((Item=(Type=Bone,Name="base_zipperFront_A"),Transform=(Rotation=(X=0,Y=0,Z=0.2842,W=0.9588),Translation=(X=0,Y=0,Z=0),Scale3D=(X=1,Y=1,Z=1))))
    WeightMinimum=0
    Weight=@n1.Value
   )
   > &RigUnit_ModifyTransforms(
    ItemToModify=((Item=(Type=Bone,Name="base_zipperFront_B"),Transform=(Rotation=(X=-0.0368,Y=-0.079,Z=-0.421,W=0.9029),Translation=(X=0,Y=0,Z=0),Scale3D=(X=1,Y=1,Z=1))))
    WeightMinimum=0
    Weight=@n1.Value
   )
   > &RigUnit_ModifyTransforms(
    ItemToModify=((Item=(Type=Bone,Name="kin_sidePack_retract"),Transform=(Rotation=(X=0,Y=-0.1736,Z=0,W=0.9848),Translation=(X=3.15,Y=0,Z=2.5),Scale3D=(X=1,Y=1,Z=1))))
    WeightMinimum=0
    Weight=RigVMVariableNode() {variable=thighBend_out_l}.Value
   )
   > &RigUnit_ModifyTransforms(
    ItemToModify=((Item=(Type=Bone,Name="base_zipperFront_B"),Transform=(Rotation=(X=0,Y=0,Z=-0.4226,W=0.9063),Translation=(X=0,Y=0,Z=0),Scale3D=(X=1,Y=1,Z=1))))
    WeightMinimum=0
    Weight=RigVMVariableNode() {variable=thighBend_out_n40_n100}.Value
   )
   > RigVMFunctionReturnNode()
  ]}
 New Function_ContainedGraph {nodes [
   RigVMFunctionEntryNode()
   > &RigUnit_ParentConstraint(
    AdvancedSettings=(RotationOrderForFilter=XZY)
    Child=(Type=Bone,Name="base_bandoliersTop_l")
    Filter=(TranslationFilter=(bX=true,bY=true,bZ=true),RotationFilter=(bX=True,bY=True,bZ=True),ScaleFilter=(bX=true,bY=true,bZ=true))
    Parents=((Item=(Type=Bone,Name="thigh_twist_01_l"),Weight=0.8),(Item=(Type=Bone,Name="pelvis"),Weight=0.2))
    Weight=1
   )
   > &RigUnit_ParentConstraint(
    AdvancedSettings=(RotationOrderForFilter=XZY)
    Child=(Type=Bone,Name="base_bandoliersTop_r")
    Filter=(TranslationFilter=(bX=true,bY=true,bZ=true),RotationFilter=(bX=True,bY=True,bZ=True),ScaleFilter=(bX=true,bY=true,bZ=true))
    Parents=((Item=(Type=Bone,Name="thigh_twist_01_r"),Weight=0.8),(Item=(Type=Bone,Name="pelvis"),Weight=0.2))
    Weight=1
   )
   > &RigUnit_ModifyTransforms(
    ItemToModify="((Item=(Type=Bone,Name=\"base_bandoliersTop_l\"),Transform=(Rotation=(X=-0.1736,Y=0,Z=0,W=0.9848),Translation=(X=0,Y=0,Z=0),Scale3D=(X=1,Y=1,Z=1))),(Item=(Type=Bone,Name=\"kin_bandoliersTopFront_l\"),Transform=(Rotation=(X=-0.043,Y=-0.1735,Z=-0.0076,W=0.9839), ...(truncated)"
    WeightMinimum=0
    Weight=RigVMVariableNode() {variable=thighBend_out_l}.Value
   )
   > &RigUnit_ModifyTransforms(
    ItemToModify="((Item=(Type=Bone,Name=\"base_bandoliersTop_r\"),Transform=(Rotation=(X=0.1736,Y=0,Z=0,W=0.9848),Translation=(X=0,Y=0,Z=0),Scale3D=(X=1,Y=1,Z=1))),(Item=(Type=Bone,Name=\"kin_bandoliersTopFront_r\"),Transform=(Rotation=(X=0.043,Y=-0.1735,Z=0.0076,W=0.9839),Tra ...(truncated)"
    WeightMinimum=0
    Weight=RigVMVariableNode() {variable=thighBend_out_r}.Value
   )
   > n1 = &RigUnit_SphericalPoseReader(
    ActiveRegionScaleFactors=(PositiveWidth=1,NegativeWidth=1,PositiveHeight=1,NegativeHeight=1)
    ActiveRegionSize=0.05
    Debug=(bDrawDebug=false,bDraw2D=false,bDrawLocalAxes=False,DebugScale=25,DebugSegments=20,DebugThickness=0.25)
    DriverAxis=(X=1,Y=0,Z=0)
    DriverItem=(Type=Bone,Name="thigh_r")
    FalloffRegionScaleFactors=(PositiveWidth=1,NegativeWidth=1,PositiveHeight=1,NegativeHeight=1)
    FalloffSize=0.1
    FlipHeightScaling=False
    FlipWidthScaling=False
    OptionalParentItem=(Type=Bone,Name="None")
    RotationOffset=(X=0,Y=-109,Z=6)
   )
   > &RigUnit_ModifyTransforms(
    ItemToModify=((Item=(Type=Bone,Name="kin_bandoliersTopBack_r"),Transform=(Rotation=(X=0,Y=0,Z=0,W=-1),Translation=(X=0,Y=2,Z=0),Scale3D=(X=1,Y=1,Z=1))))
    WeightMinimum=0
    Weight=@n1.OutputParam
   )
   > n2 = &RigUnit_SphericalPoseReader(
    ActiveRegionScaleFactors=(PositiveWidth=1,NegativeWidth=1,PositiveHeight=1,NegativeHeight=1)
    ActiveRegionSize=0.05
    Debug=(bDrawDebug=false,bDraw2D=false,bDrawLocalAxes=False,DebugScale=25,DebugSegments=20,DebugThickness=0.25)
    DriverAxis=(X=-1,Y=0,Z=0)
    DriverItem=(Type=Bone,Name="thigh_l")
    FalloffRegionScaleFactors=(PositiveWidth=1,NegativeWidth=1,PositiveHeight=1,NegativeHeight=1)
    FalloffSize=0.1
    FlipHeightScaling=False
    FlipWidthScaling=False
    OptionalParentItem=(Type=Bone,Name="None")
    RotationOffset=(X=0,Y=71,Z=6)
   )
   > n3 = &RigUnit_ModifyTransforms(
    ItemToModify=((Item=(Type=Bone,Name="kin_bandoliersTopBack_l"),Transform=(Rotation=(X=0,Y=0,Z=0,W=-1),Translation=(X=0,Y=-2,Z=0),Scale3D=(X=1,Y=1,Z=1))))
    WeightMinimum=0
    Weight=@n2.OutputParam
   )
   constraints = RigVMCollapseNode(ExecutePin=@n3.ExecutePin)
   RigVMFunction_Sequence = RigVMAggregateNode(ExecuteContext=@constraints.ExecutePin)
   RigVMFunctionReturnNode(ExecuteContext=@RigVMFunction_Sequence.C)
   &RigUnit_ModifyTransforms(
    ItemToModify="((Item=(Type=Null,Name=\"dyn_bandoliersBotBack_3_r_ctrl_null\"),Transform=(Rotation=(X=0,Y=0,Z=0,W=-1),Translation=(X=0,Y=-4.4142,Z=0),Scale3D=(X=1,Y=1,Z=1))),(Item=(Type=Null,Name=\"dyn_bandoliersBotBack_4_r_ctrl_null\"),Transform=(Rotation=(X=0,Y=0,Z=0,W=-1) ...(truncated)"
    WeightMinimum=0
    Weight=RigVMVariableNode() {variable=thighBend_out_r}.Value
    ExecutePin=@RigVMFunction_Sequence.B
   )
   n4 = &RigUnit_ItemArray(Items="((Type=Control,Name=\"dyn_bandoliersBotBack_1_r_ctrl\"),(Type=Control,Name=\"dyn_bandoliersBotBack_2_r_ctrl\"),(Type=Control,Name=\"dyn_bandoliersBotBack_3_r_ctrl\"),(Type=Control,Name=\"dyn_bandoliersBotBack_4_r_ctrl\"),(Type=Control,Name=\"dyn_bandoliersBotBack_5 ...(truncated)")
   n5 = &RigUnit_ItemArray(Items=((Type=Bone,Name="dyn_bandoliersBotBack_1_r"),(Type=Bone,Name="dyn_bandoliersBotBack_2_r"),(Type=Bone,Name="dyn_bandoliersBotBack_3_r"),(Type=Bone,Name="dyn_bandoliersBotBack_4_r"),(Type=Bone,Name="dyn_bandoliersBotBack_5_r")))
   > SplineIK_2 = RigVMFunctionReferenceNode(
    Primary Axis=(X=1,Y=0,Z=0)
    Samples Per Segment=3
    Secondary Spline Direction=(X=0,Y=5,Z=0)
    Spline Mode=BSpline
    Stretch=True
    Up Axis=(X=0,Y=1,Z=0)
    Controls=@n4.Items
    Bones=@n5.Items
   )
   n6 = &RigUnit_ItemArray(Items="((Type=Control,Name=\"dyn_bandoliersBotFront_1_r_ctrl\"),(Type=Control,Name=\"dyn_bandoliersBotFront_2_r_ctrl\"),(Type=Control,Name=\"dyn_bandoliersBotFront_3_r_ctrl\"),(Type=Control,Name=\"dyn_bandoliersBotFront_4_r_ctrl\"),(Type=Control,Name=\"dyn_bandoliersBotFr ...(truncated)")
   n7 = &RigUnit_ItemArray(Items=((Type=Bone,Name="dyn_bandoliersBotFront_1_r"),(Type=Bone,Name="dyn_bandoliersBotFront_2_r"),(Type=Bone,Name="dyn_bandoliersBotFront_3_r"),(Type=Bone,Name="dyn_bandoliersBotFront_4_r"),(Type=Bone,Name="dyn_bandoliersBotFront_5_r")))
   > SplineIK_3 = RigVMFunctionReferenceNode(
    Primary Axis=(X=1,Y=0,Z=0)
    Samples Per Segment=2
    Secondary Spline Direction=(X=0,Y=5,Z=0)
    Spline Mode=BSpline
    Stretch=True
    Up Axis=(X=0,Y=1,Z=0)
    Controls=@n6.Items
    Bones=@n7.Items
   )
   &RigUnit_ModifyTransforms(
    ItemToModify="((Item=(Type=Null,Name=\"dyn_bandoliersBotBack_3_l_ctrl_null\"),Transform=(Rotation=(X=0,Y=0,Z=0,W=-1),Translation=(X=0,Y=-4.4142,Z=0),Scale3D=(X=1,Y=1,Z=1))),(Item=(Type=Null,Name=\"dyn_bandoliersBotBack_4_l_ctrl_null\"),Transform=(Rotation=(X=0,Y=0,Z=0,W=-1) ...(truncated)"
    WeightMinimum=0
    Weight=RigVMVariableNode() {variable=thighBend_out_l}.Value
    ExecutePin=@RigVMFunction_Sequence.A
   )
   n8 = &RigUnit_ItemArray(Items=((Type=Bone,Name="dyn_bandoliersBotFront_1_l"),(Type=Bone,Name="dyn_bandoliersBotFront_2_l"),(Type=Bone,Name="dyn_bandoliersBotFront_3_l"),(Type=Bone,Name="dyn_bandoliersBotFront_4_l"),(Type=Bone,Name="dyn_bandoliersBotFront_5_l")))
   n9 = &RigUnit_ItemArray(Items="((Type=Control,Name=\"dyn_bandoliersBotFront_1_l_ctrl\"),(Type=Control,Name=\"dyn_bandoliersBotFront_2_l_ctrl\"),(Type=Control,Name=\"dyn_bandoliersBotFront_3_l_ctrl\"),(Type=Control,Name=\"dyn_bandoliersBotFront_4_l_ctrl\"),(Type=Control,Name=\"dyn_bandoliersBotFr ...(truncated)")
   > SplineIK = RigVMFunctionReferenceNode(
    Primary Axis=(X=1,Y=0,Z=0)
    Samples Per Segment=3
    Secondary Spline Direction=(X=0,Y=5,Z=0)
    Spline Mode=BSpline
    Stretch=true
    Up Axis=(X=0,Y=1,Z=0)
    Bones=@n8.Items
    Controls=@n9.Items
   )
   RigUnit_ItemArray = &RigUnit_ItemArray(Items=((Type=Bone,Name="dyn_bandoliersBotBack_1_l"),(Type=Bone,Name="dyn_bandoliersBotBack_2_l"),(Type=Bone,Name="dyn_bandoliersBotBack_3_l"),(Type=Bone,Name="dyn_bandoliersBotBack_4_l"),(Type=Bone,Name="dyn_bandoliersBotBack_5_l")))
   n10 = &RigUnit_ItemArray(Items="((Type=Control,Name=\"dyn_bandoliersBotBack_1_l_ctrl\"),(Type=Control,Name=\"dyn_bandoliersBotBack_2_l_ctrl\"),(Type=Control,Name=\"dyn_bandoliersBotBack_3_l_ctrl\"),(Type=Control,Name=\"dyn_bandoliersBotBack_4_l_ctrl\"),(Type=Control,Name=\"dyn_bandoliersBotBack_5 ...(truncated)")
   > SplineIK_1 = RigVMFunctionReferenceNode(
    Primary Axis=(X=1,Y=0,Z=0)
    Samples Per Segment=2
    Secondary Spline Direction=(X=0,Y=5,Z=0)
    Spline Mode=BSpline
    Stretch=true
    Up Axis=(X=0,Y=1,Z=0)
    Bones=@RigUnit_ItemArray.Items
    Controls=@n10.Items
   )
  ]}
 New Function_ContainedGraph {nodes [
   n1 = RigVMFunctionEntryNode()
   RigVMFunction_ControlFlowBranch = &RigVMFunction_ControlFlowBranch(ExecuteContext=@n1.ExecuteContext, Condition=RigVMVariableNode() {variable=pub isWearingBackBling}.Value)
   n2 = &RigVMFunction_ControlFlowBranch(Condition=true, ExecuteContext=@RigVMFunction_ControlFlowBranch.Completed)
   RigVMFunctionReturnNode(ExecuteContext=@n2.Completed)
   constraints = RigVMCollapseNode(ExecutePin=@n2.True)
   DeltaRotation_0 = RigVMFunctionReferenceNode(
    Rotation Order=ZYX
    child=spine_04
    debugRotation=false
    max=50
    min=10
    parent=spine_01
    parentType=Bone
    premult=1
    primAxis=2
    twistAxis=(X=1,Y=0,Z=0)
    ExecuteContext=@constraints.ExecutePin
   )
   > DeltaRotation v4 = RigVMFunctionReferenceNode(
    Rotation Order=ZYX
    absoluteValue=true
    child=(Type=Bone,Name="spine_05")
    debugRotation=false
    onlyTwist=true
    outputMinMax=(X=-1,Y=1)
    parent=(Type=Bone,Name="pelvis")
    premult=1
    rotationMinMax=(X=-60,Y=60)
    triggerAxis=X
    twistAxis=(X=1,Y=0,Z=0)
   )
   Maximum = &RigVMFunction_MathDoubleMax(A=@"DeltaRotation v4".outputWeight, B=@DeltaRotation_0.outputWeight)
   > &RigUnit_ModifyTransforms(
    ItemToModify="((Item=(Type=Null,Name=\"dyn_shoulderStraps_6_l_ctrl_null\"),Transform=(Rotation=(X=0,Y=0,Z=0,W=1),Translation=(X=0,Y=0,Z=5),Scale3D=(X=1,Y=1,Z=1))),(Item=(Type=Null,Name=\"dyn_shoulderStraps_6_r_ctrl_null\"),Transform=(Rotation=(X=0,Y=0,Z=0,W=1),Translation=( ...(truncated)"
    WeightMinimum=0
    Weight=@Maximum.Result
   )
   n3 = &RigUnit_ItemArray(Items="((Type=Bone,Name=\"dyn_shoulderStraps_1_l\"),(Type=Bone,Name=\"dyn_shoulderStraps_2_l\"),(Type=Bone,Name=\"dyn_shoulderStraps_3_l\"),(Type=Bone,Name=\"dyn_shoulderStraps_4_l\"),(Type=Bone,Name=\"dyn_shoulderStraps_5_l\"),(Type=Bone,Name=\"dyn_shoulderStraps_6_l\"),(Ty ...(truncated)")
   n4 = &RigUnit_ItemArray(Items="((Type=Control,Name=\"dyn_shoulderStraps_1_l_ctrl\"),(Type=Control,Name=\"dyn_shoulderStraps_2_l_ctrl\"),(Type=Control,Name=\"dyn_shoulderStraps_3_l_ctrl\"),(Type=Control,Name=\"dyn_shoulderStraps_4_l_ctrl\"),(Type=Control,Name=\"dyn_shoulderStraps_5_l_ctrl\"),(Type ...(truncated)")
   > SplineIK = RigVMFunctionReferenceNode(
    Primary Axis=(X=1,Y=0,Z=0)
    Samples Per Segment=3
    Secondary Spline Direction=(X=0,Y=5,Z=0)
    Spline Mode=BSpline
    Stretch=true
    Up Axis=(X=0,Y=1,Z=0)
    Bones=@n3.Items
    Controls=@n4.Items
   )
   RigUnit_ItemArray = &RigUnit_ItemArray(Items="((Type=Control,Name=\"dyn_shoulderStraps_1_r_ctrl\"),(Type=Control,Name=\"dyn_shoulderStraps_2_r_ctrl\"),(Type=Control,Name=\"dyn_shoulderStraps_3_r_ctrl\"),(Type=Control,Name=\"dyn_shoulderStraps_4_r_ctrl\"),(Type=Control,Name=\"dyn_shoulderStraps_5_r_ctrl\"),(Type ...(truncated)")
   n5 = &RigUnit_ItemArray(Items="((Type=Bone,Name=\"dyn_shoulderStraps_1_r\"),(Type=Bone,Name=\"dyn_shoulderStraps_2_r\"),(Type=Bone,Name=\"dyn_shoulderStraps_3_r\"),(Type=Bone,Name=\"dyn_shoulderStraps_4_r\"),(Type=Bone,Name=\"dyn_shoulderStraps_5_r\"),(Type=Bone,Name=\"dyn_shoulderStraps_6_r\"),(Ty ...(truncated)")
   > SplineIK_1 = RigVMFunctionReferenceNode(
    Primary Axis=(X=1,Y=0,Z=0)
    Samples Per Segment=3
    Secondary Spline Direction=(X=0,Y=5,Z=0)
    Spline Mode=BSpline
    Stretch=true
    Up Axis=(X=0,Y=1,Z=0)
    Controls=@RigUnit_ItemArray.Items
    Bones=@n5.Items
   )
   &RigUnit_ModifyTransforms(
    ItemToModify="((Item=(Type=Bone,Name=\"base_shoulderStraps_1_l\"),Transform=(Rotation=(X=0,Y=-0.0872,Z=0,W=0.9962),Translation=(X=0,Y=0,Z=0),Scale3D=(X=1,Y=1,Z=1))),(Item=(Type=Bone,Name=\"base_shoulderStraps_1_r\"),Transform=(Rotation=(X=0,Y=-0.0872,Z=0,W=0.9962),Translati ...(truncated)"
    WeightMinimum=0
    Weight=RigVMVariableNode() {variable=twistSpine_20_50}.Value
    ExecutePin=@RigVMFunction_ControlFlowBranch.False
   )
  ]}
 New Function_ContainedGraph {nodes [
   n1 = RigVMFunctionEntryNode()
   n2 = &RigUnit_GetTransform(bInitial=False, Item=@n1.Start)
   Subtract = /Script/RigVM.RigVMFunction_MathVectorSub(B=@n2.Transform.Translation, A=&RigUnit_GetTransform(bInitial=False, Item=@n1.End).Transform.Translation) {is=struct RigVMUnitNode, template="Subtract::Execute(in A,in B,out Result)"}
   Unit = /Script/RigVM.RigVMFunction_MathVectorUnit(Value=@Subtract.Result) {is=struct RigVMUnitNode, template="Unit::Execute(in Value,out Result)"}
   RigVMFunction_MathQuaternionFromAxisAndAngle = /Script/RigVM.RigVMFunction_MathQuaternionFromAxisAndAngle(Axis=@Unit.Result, Angle=@n1.QuatAmount) {is=struct RigVMUnitNode, template="FromAxisAndAngle::Execute(in Axis,in Angle,out Result)"}
   Multiply = /Script/RigVM.RigVMFunction_MathQuaternionMul(A=@RigVMFunction_MathQuaternionFromAxisAndAngle.Result, B=@n2.Transform.Rotation) {is=struct RigVMUnitNode, template="Multiply::Execute(in A,in B,out Result)"}
   > &RigUnit_SetRotation(Space=GlobalSpace, bInitial=False, Item=@n1.Start, Weight=@n1.Weight, Value=@Multiply.Result)
   > RigVMFunctionReturnNode()
  ]}
 New Function_ContainedGraph {nodes [
   RigVMFunctionEntryNode()
   > &RigUnit_ParentConstraint(
    AdvancedSettings=(RotationOrderForFilter=XZY)
    Child=(Type=Bone,Name="base_thermos")
    Parents="((Item=(Type=Bone,Name=\"spine_05\"),Weight=0.0013),(Item=(Type=Bone,Name=\"pelvis\"),Weight=0.0037),(Item=(Type=Bone,Name=\"spine_04\"),Weight=0.006),(Item=(Type=Bone,Name=\"spine_01\"),Weight=0.0161),(Item=(Type=Bone,Name=\"spine_03\"),Weight=0.0199),(Item=(Type=B ...(truncated)"
    Weight=1
   )
   > &RigUnit_ParentConstraint(
    AdvancedSettings=(RotationOrderForFilter=XZY)
    Child=(Type=Bone,Name="kin_sidePack_2")
    Parents="((Item=(Type=Bone,Name=\"pelvis\"),Weight=0.0232),(Item=(Type=Bone,Name=\"spine_05\"),Weight=0.0646),(Item=(Type=Bone,Name=\"spine_01\"),Weight=0.1248),(Item=(Type=Bone,Name=\"spine_04\"),Weight=0.1541),(Item=(Type=Bone,Name=\"spine_02\"),Weight=0.3071),(Item=(Type= ...(truncated)"
    Weight=1
   )
   > &RigUnit_ParentConstraint(
    AdvancedSettings=(RotationOrderForFilter=XZY)
    Child=(Type=Bone,Name="kin_sidePack_1")
    Parents="((Item=(Type=Bone,Name=\"kin_frontPack_l\"),Weight=0.0002),(Item=(Type=Bone,Name=\"pelvis\"),Weight=0.0032),(Item=(Type=Bone,Name=\"spine_01\"),Weight=0.0176),(Item=(Type=Bone,Name=\"spine_02\"),Weight=0.0844),(Item=(Type=Bone,Name=\"spine_03\"),Weight=0.2063),(Item ...(truncated)"
    Weight=1
   )
   > DeltaRotation = RigVMFunctionReferenceNode(
    Rotation Order=ZYX
    child=spine_04
    debugRotation=false
    max=80
    parent=spine_01
    parentType=Bone
    premult=1
    primAxis=2
    twistAxis=(X=1,Y=0,Z=0)
   )
   > &RigUnit_ModifyTransforms(
    ItemToModify=((Item=(Type=Bone,Name="base_thermos"),Transform=(Rotation=(X=0,Y=0,Z=-0.2588,W=0.9659),Translation=(X=0,Y=0,Z=0),Scale3D=(X=1,Y=1,Z=1))))
    WeightMinimum=0
    Weight=@DeltaRotation.outputWeight
   )
   > DeltaRotation_3 = RigVMFunctionReferenceNode(
    Rotation Order=ZYX
    child=spine_05
    debugRotation=false
    max=50
    min=20
    onlyTwist=True
    parent=spine_01
    parentType=Bone
    premult=1
    twistAxis=(X=1,Y=0,Z=0)
   )
   > RigVMVariableNode(Value=@DeltaRotation_3.outputWeight) {variable=twistSpine_20_50}
   > &RigUnit_ModifyTransforms(
    ItemToModify=((Item=(Type=Bone,Name="base_thermos"),Transform=(Rotation=(X=0,Y=0,Z=-0.2588,W=0.9659),Translation=(X=0,Y=2,Z=0),Scale3D=(X=1,Y=1,Z=1))))
    WeightMinimum=0
    Weight=@DeltaRotation_3.outputWeight
   )
   > DeltaRotation_1 = RigVMFunctionReferenceNode(Rotation Order=ZYX, child=spine_04, max=50, parent=pelvis, parentType=Bone, premult=1, primAxis=1, twistAxis=(X=1,Y=0,Z=0))
   > &RigUnit_ModifyTransforms(
    ItemToModify="((Item=(Type=Bone,Name=\"kin_sidePack_retract\"),Transform=(Rotation=(X=0,Y=-0.2588,Z=0,W=0.9659),Translation=(X=2.5,Y=0,Z=10),Scale3D=(X=1,Y=1,Z=1))),(Item=(Type=Bone,Name=\"base_mug\"),Transform=(Rotation=(X=-0.0868,Y=-0.4924,Z=-0.1504,W=0.8529),Translation= ...(truncated)"
    WeightMinimum=0
    Weight=@DeltaRotation_1.outputWeight
   )
   > DeltaRotation_2 = RigVMFunctionReferenceNode(
    Rotation Order=ZYX
    child=spine_05
    debugRotation=false
    max=-20
    min=-2
    parent=spine_02
    parentType=Bone
    premult=1
    primAxis=2
    twistAxis=(X=1,Y=0,Z=0)
   )
   > &RigUnit_ModifyTransforms(
    ItemToModify=((Item=(Type=Bone,Name="attach_backpack"),Transform=(Rotation=(X=-0.1305,Y=0,Z=0,W=0.9914),Translation=(X=0,Y=-1,Z=0),Scale3D=(X=1,Y=1,Z=1))))
    WeightMinimum=0
    Weight=@DeltaRotation_2.outputWeight
   )
   > DeltaRotation_4 = RigVMFunctionReferenceNode(
    Rotation Order=ZYX
    child=spine_05
    debugRotation=false
    max=50
    min=0
    onlyTwist=true
    parent=spine_01
    parentType=Bone
    premult=1
    primAxis=0
    twistAxis=(X=1,Y=0,Z=0)
   )
   > RigVMVariableNode(Value=@DeltaRotation_4.outputWeight) {variable=twistSpine_0_50}
   > &RigUnit_ModifyTransforms(
    ItemToModify=((Item=(Type=Bone,Name="attach_backpack"),Transform=(Rotation=(X=-0.0784,Y=0.0034,Z=-0.0435,W=0.996),Translation=(X=0,Y=0,Z=0),Scale3D=(X=1,Y=1,Z=1))))
    WeightMinimum=0
    Weight=@DeltaRotation_4.outputWeight
   )
   > &RigUnit_ModifyTransforms(
    ItemToModify=((Item=(Type=Bone,Name="base_thermos"),Transform=(Rotation=(X=-0.0061,Y=0.0695,Z=0.0869,W=0.9938),Translation=(X=0,Y=0,Z=0),Scale3D=(X=1,Y=1,Z=1))))
    WeightMinimum=0
    Weight=RigVMDispatchNode(False=0, True=RigVMVariableNode() {variable=rUpperarmDown}.Value, Condition=RigVMVariableNode() {variable=pub isFrontEnd}.Value) {template="DISPATCH_RigVMDispatch_If(in Condition,in True,in False,out Result)"}.Result
   )
   > RigVMFunctionReturnNode()
  ]}
 New Function_ContainedGraph {nodes [
   n1 = RigVMFunctionEntryNode()
   RigVMFunction_MathBoolToInteger = &RigVMFunction_MathBoolToInteger(Value=@n1.onlyTwist)
   RigVMFunction_MathQuaternionSwingTwist = &RigVMFunction_MathQuaternionSwingTwist(
    TwistAxis=@n1.twistAxis
    Input=&RigUnit_GetRelativeTransformForItem(bChildInitial=False, bParentInitial=False, Child=@n1.child, Parent=@n1.parent).RelativeTransform.Rotation
   )
   n2 = RigVMDispatchNode(
    Values=((),())
    Index=@RigVMFunction_MathBoolToInteger.Result
    Values.0=@RigVMFunction_MathQuaternionSwingTwist.Swing
    Values.1=@RigVMFunction_MathQuaternionSwingTwist.Twist
   ) {template="DISPATCH_RigVMDispatch_SelectInt32(in Index,in Values,out Result)"}
   RigVMFunction_MathQuaternionToEuler = &RigVMFunction_MathQuaternionToEuler(Value=@n2.Result, RotationOrder=@n1.Rotation Order)
   > DISPATCH_RigVMDispatch_Print = RigVMDispatchNode(ScreenColor=(R=1,G=1,B=1,A=1), ScreenDuration=1.5, Enabled=@n1.debugRotation, Value=@RigVMFunction_MathQuaternionToEuler.Result) {template="DISPATCH_RigVMDispatch_Print(in Prefix,in Value,in Enabled,in ScreenDuration,in ScreenColor)"}
   Cast = RigVMDispatchNode(Value=@n1.triggerAxis) {template="DISPATCH_RigVMDispatch_CastEnumToInt(in Value,out Result)"}
   DISPATCH_RigVMDispatch_SelectInt32 = RigVMDispatchNode(
    Values=(0,0,0)
    Index=@Cast.Result
    Values.0=@RigVMFunction_MathQuaternionToEuler.Result.X
    Values.1=@RigVMFunction_MathQuaternionToEuler.Result.Y
    Values.2=@RigVMFunction_MathQuaternionToEuler.Result.Z
   ) {template="DISPATCH_RigVMDispatch_SelectInt32(in Index,in Values,out Result)"}
   Remap = &RigVMFunction_MathDoubleRemap(
    bClamp=False
    Value=@DISPATCH_RigVMDispatch_SelectInt32.Result
    TargetMinimum=@n1.outputMinMax.X
    TargetMaximum=@n1.outputMinMax.Y
    SourceMinimum=@n1.rotationMinMax.X
    SourceMaximum=@n1.rotationMinMax.Y
   )
   Absolute = &RigVMFunction_MathDoubleAbs(Value=@Remap.Result)
   Multiply = &RigVMFunction_MathDoubleMul(
    A=&RigVMFunction_MathDoubleClamp(Value=RigVMDispatchNode(True=@Absolute.Result, Condition=@n1.absoluteValue, False=@Remap.Result) {template="DISPATCH_RigVMDispatch_If(in Condition,in True,in False,out Result)"}.Result).Result
    B=@n1.premult
   )
   > RigVMFunctionReturnNode(outputQuat=@n2.Result, outputWeight=@Multiply.Result)
  ]}
 New Function_ContainedGraph {nodes [
   n1 = RigVMFunctionEntryNode()
   Num = RigVMDispatchNode(Array=@n1.targets) {template="DISPATCH_RigVMDispatch_ArrayGetNum(in Array,out Num)"}
   > RigVMDispatchNode(Element=0, Num=@Num.Num, Array=RigVMVariableNode() {variable=outputValues}.Value) {template="DISPATCH_RigVMDispatch_ArrayInit(io Array,in Num,in Element)"}
   n2 = &RigVMFunction_MathTransformMul(
    A=&RigUnit_GetRelativeTransformForItem(bChildInitial=true, bParentInitial=true, Child=@n1.Child, Parent=@n1.Parent).RelativeTransform
    B=&RigUnit_GetTransform(bInitial=False, Item=@n1.Parent).Transform
   ) {template="Multiply::Execute(in A,in B,out Result)"}
   n3 = &RigVMFunction_MathTransformMul(
    A=(Rotation=(X=0,Y=0,Z=0,W=1),Translation=(X=0,Y=0,Z=0),Scale3D=(X=1,Y=1,Z=1))
    A.Translation=@n1.primaryAxis
    B=&RigUnit_GetTransform(bInitial=False, Item=@n1.Child).Transform
   ) {template="Multiply::Execute(in A,in B,out Result)"}
   > /Script/RigVM.RigVMFunction_DebugLineNoSpace(
    Color=(R=0,G=0,B=1,A=1)
    DebugDrawSettings=(DepthPriority=SDPG_Foreground,Lifetime=-1)
    Thickness=0.5
    WorldOffset=(Rotation=(X=0,Y=0,Z=0,W=1),Translation=(X=0,Y=0,Z=0),Scale3D=(X=1,Y=1,Z=1))
    bEnabled=@n1.drawLines
    A=@n2.Result.Translation
    B=@n3.Result.Translation
   ) {is=struct RigVMUnitNode}
   > For_Each = RigVMDispatchNode(Array=@n1.targets) {template="DISPATCH_RigVMDispatch_ArrayIterator(in Array,out Element,out Index,out Count,out Ratio)"}
   Multiply_1 = RigVMAggregateNode(
    A=(Rotation=(X=0,Y=0,Z=0,W=1),Translation=(X=0,Y=0,Z=0),Scale3D=(X=1,Y=1,Z=1))
    B=(Rotation=(X=0,Y=0,Z=0,W=1),Translation=(X=0,Y=0,Z=0),Scale3D=(X=1,Y=1,Z=1))
    A.Translation=@n1.primaryAxis
    B.Rotation=@For_Each.Element
    C=@n2.Result
   )
   > &RigVMFunction_DebugLineNoSpace(
    Color=(R=1,G=0,B=0,A=1)
    DebugDrawSettings=(DepthPriority=SDPG_Foreground,Lifetime=-1)
    Thickness=0.5
    WorldOffset=(Rotation=(X=0,Y=0,Z=0,W=1),Translation=(X=0,Y=0,Z=0),Scale3D=(X=1,Y=1,Z=1))
    bEnabled=@n1.drawLines
    B=@Multiply_1.Result.Translation
    A=@n2.Result.Translation
   )
   RigVMFunction_MathVectorAngle = /Script/RigVM.RigVMFunction_MathVectorAngle(
    B=&RigVMFunction_MathVectorSub(A=@Multiply_1.Result.Translation, B=@n2.Result.Translation) {template="Subtract::Execute(in A,in B,out Result)"}.Result
    A=&RigVMFunction_MathVectorSub(B=@n2.Result.Translation, A=@n3.Result.Translation) {template="Subtract::Execute(in A,in B,out Result)"}.Result
   ) {is=struct RigVMUnitNode, template="AngleBetween::Execute(in A,in B,out Result)"}
   Degrees = /Script/RigVM.RigVMFunction_MathDoubleDeg(Value=@RigVMFunction_MathVectorAngle.Result) {is=struct RigVMUnitNode, template="Degrees::Execute(in Value,out Result)"}
   Multiply = &RigVMFunction_MathDoubleMul(A=@n1.coneAngle, B=@n1.quotient)
   Remap = &RigVMFunction_MathDoubleRemap(SourceMaximum=0, TargetMaximum=1, TargetMinimum=0, Value=@Degrees.Result, SourceMinimum=@Multiply.Result)
   > Set_At = RigVMDispatchNode(Index=@For_Each.Index, Element=@Remap.Result, Array=RigVMVariableNode() {variable=outputValues}.Value) {template="DISPATCH_RigVMDispatch_ArraySetAtIndex(io Array,in Index,in Element)"}
   RigVMFunctionReturnNode(ExecuteContext=@For_Each.Completed, weights=RigVMVariableNode() {variable=outputValues}.Value)
  ]}
 New Function_ContainedGraph {nodes [
   n1 = RigVMFunctionEntryNode()
   n2 = &RigUnit_GetTransform(bInitial=true, Item=@n1.secondBone)
   n3 = &RigVMFunction_MathVectorDistance(B=@n2.Transform.Translation, A=&RigUnit_GetTransform(bInitial=true, Item=@n1.firstBone).Transform.Translation)
   n4 = &RigVMFunction_MathVectorDistance(A=@n2.Transform.Translation, B=&RigUnit_GetTransform(bInitial=true, Item=@n1.effectorBone).Transform.Translation)
   n5 = &RigVMFunction_MathDoubleAdd(A=@n3.Result, B=@n4.Result)
   > RigVMVariableNode(Value=@n5.Result) {variable=Entry_segmentsSum}
   > RigVMVariableNode(Value=&RigVMFunction_MathDoubleDiv(B=@n5.Result, A=@n3.Result).Result) {variable=Entry_segmentAFactor}
   > RigVMVariableNode(Value=&RigVMFunction_MathDoubleDiv(B=@n5.Result, A=@n4.Result).Result) {variable=Entry_segmentBFactor}
   > &RigUnit_ParentConstraint(
    AdvancedSettings=(InterpolationType=Average,RotationOrderForFilter=XZY)
    Filter=(TranslationFilter=(bX=True,bY=True,bZ=True),RotationFilter=(bX=false,bY=false,bZ=false),ScaleFilter=(bX=false,bY=false,bZ=false))
    Parents=((Item=(Type=Bone,Name="None"),Weight=1))
    Weight=1
    bMaintainOffset=true
    Parents.0.Item=@n1.firstBoneOrig
    Child=@n1.poleVector
   )
   > &RigUnit_ModifyTransforms(
    ItemToModify=((Item=(Type=Bone,Name="None"),Transform=(Rotation=(X=0,Y=0,Z=0,W=1),Translation=(X=0,Y=0,Z=0),Scale3D=(X=1,Y=1,Z=1))))
    Weight=1
    WeightMinimum=0
    ItemToModify.0.Item=@n1.poleVector
    ItemToModify.0.Transform.Translation=@n1.offsetPv
   )
   n6 = &RigUnit_GetTransform(bInitial=False, Item=@n1.ikEffector)
   RigVMFunction_MathVectorDistance = &RigVMFunction_MathVectorDistance(B=@n6.Transform.Translation, A=&RigUnit_GetTransform(bInitial=False, Item=@n1.itemStartStretch).Transform.Translation)
   n7 = &RigVMFunction_MathDoubleSub(A=RigVMVariableNode() {variable=Entry_segmentsSum}.Value, B=@n1.softness)
   Divide = &RigVMFunction_MathDoubleDiv(
    A=&RigVMFunction_MathDoubleMul(B=-1, A=&RigVMFunction_MathDoubleSub(A=@RigVMFunction_MathVectorDistance.Result, B=@n7.Result).Result).Result
    B=@n1.softness
   )
   Exponential = /Script/RigVM.RigVMFunction_MathDoubleExponential(Value=@Divide.Result) {is=struct RigVMUnitNode, template="Exponential::Execute(in Value,out Result)"}
   Add = &RigVMFunction_MathDoubleAdd(
    B=RigVMDispatchNode(
     False=0
     Condition=/Script/RigVM.RigVMFunction_MathDoubleGreater(A=@RigVMFunction_MathVectorDistance.Result, B=@n7.Result) {is=struct RigVMUnitNode, template="Greater::Execute(in A,in B,out Result)"}.Result
     True=&RigVMFunction_MathDoubleMul(A=@Exponential.Result, B=@n1.softness).Result
    ) {template="DISPATCH_RigVMDispatch_If(in Condition,in True,in False,out Result)"}.Result
    A=@RigVMFunction_MathVectorDistance.Result
   )
   Greater = &RigVMFunction_MathDoubleGreater(A=@Add.Result, B=RigVMVariableNode() {variable=Entry_segmentsSum}.Value) {template="Greater::Execute(in A,in B,out Result)"}
   n8 = RigVMDispatchNode(True=@Add.Result, Condition=@Greater.Result, False=RigVMVariableNode() {variable=Entry_segmentsSum}.Value) {template="DISPATCH_RigVMDispatch_If(in Condition,in True,in False,out Result)"}
   Multiply = &RigVMFunction_MathDoubleMul(B=@n8.Result, A=RigVMVariableNode() {variable=Entry_segmentAFactor}.Value)
   n9 = &RigUnit_GetTransform(bInitial=false, Item=@n1.secondBoneOrig)
   n10 = &RigUnit_GetTransform(bInitial=false, Item=@n1.firstBone)
   n11 = RigVMVariableNode() {variable=Entry_segmentBFactor}
   n12 = &RigUnit_GetTransform(bInitial=false, Item=@n1.ikEffector)
   > /Script/ControlRig.RigUnit_TwoBoneIKSimplePerItem(
    DebugSettings=(bEnabled=False,Scale=10,WorldOffset=(Rotation=(X=0,Y=0,Z=0,W=1),Translation=(X=0,Y=0,Z=0),Scale3D=(X=1,Y=1,Z=1)))
    PoleVectorKind=Direction
    PoleVectorSpace=(Type=Bone,Name="None")
    SecondaryAxisWeight=1
    StretchMaximumRatio=2
    StretchStartRatio=1
    Weight=1
    bEnableStretch=false
    bPropagateToChildren=True
    PrimaryAxis=@n1.primary Axis
    SecondaryAxis=@n1.secondary Axis
    PoleVector=&RigUnit_GetTransform(bInitial=False, Item=@n1.poleVector).Transform.Translation
    ItemALength=RigVMDispatchNode(
     False=@Multiply.Result
     True=&RigVMFunction_MathVectorDistance(B=@n9.Transform.Translation, A=@n10.Transform.Translation).Result
     Condition=@n1.lockPV
    ) {template="DISPATCH_RigVMDispatch_If(in Condition,in True,in False,out Result)"}.Result
    ItemBLength=RigVMDispatchNode(
     False=&RigVMFunction_MathDoubleMul(A=@n8.Result, B=@n11.Value).Result
     True=&RigVMFunction_MathVectorDistance(A=@n9.Transform.Translation, B=@n12.Transform.Translation).Result
     Condition=@n1.lockPV
    ) {template="DISPATCH_RigVMDispatch_If(in Condition,in True,in False,out Result)"}.Result
    EffectorItem=@n1.effectorBone
    ItemB=@n1.secondBone
    ItemA=@n1.firstBone
    Effector=@n6.Transform
   ) {is=struct RigVMUnitNode}
   > &RigUnit_RotationConstraintLocalSpaceOffset(
    AdvancedSettings=(InterpolationType=Average,RotationOrderForFilter=XYZ)
    Filter=(bX=True,bY=True,bZ=True)
    Parents=((Item=(Type=Bone,Name="None"),Weight=1))
    Weight=1
    Child=@n1.effectorBone
    Parents.0.Item=@n1.secondBone
   )
   > RigVMFunctionReturnNode()
  ]}
 New Function_ContainedGraph {nodes [
   RigVMFunctionEntryNode()
   > RigVMFunction_Sequence = RigVMAggregateNode()
   &RigUnit_ParentConstraint(
    AdvancedSettings=(RotationOrderForFilter=XZY)
    Child=(Type=Bone,Name="deform_upperarm_fr_base_l")
    Parents=((Item=(Type=Bone,Name="arm_bend_1_l"),Weight=0.3),(Item=(Type=Bone,Name="spine_05"),Weight=0.7))
    Weight=1
    bMaintainOffset=true
    ExecutePin=@RigVMFunction_Sequence.A
   )
   > &RigUnit_ParentConstraint(
    AdvancedSettings=(RotationOrderForFilter=XZY)
    Child=(Type=Bone,Name="deform_upperarm_bk_base_l")
    Parents=((Item=(Type=Bone,Name="arm_bend_1_l"),Weight=0.3),(Item=(Type=Bone,Name="spine_05"),Weight=0.7))
    Weight=1
   )
   > n1 = &RigUnit_SphericalPoseReader(
    ActiveRegionScaleFactors=(PositiveWidth=0.5,NegativeWidth=3,PositiveHeight=1.5,NegativeHeight=1.5)
    ActiveRegionSize=0.1
    Debug=(bDrawDebug=false,DebugScale=25,DebugSegments=20,DebugThickness=0.25)
    DriverAxis=(X=1,Y=0,Z=0)
    DriverItem=(Type=Bone,Name="arm_bend_1_l")
    FalloffRegionScaleFactors=(PositiveWidth=1,NegativeWidth=1,PositiveHeight=1,NegativeHeight=1)
    FalloffSize=0.4
    OptionalParentItem=(Type=Bone)
    RotationOffset=(X=0,Y=0,Z=0)
   )
   > &RigUnit_ModifyTransforms(
    ItemToModify="((Item=(Type=Bone,Name=\"deform_upperarm_fr_l\"),Transform=(Rotation=(X=0,Y=0,Z=0,W=1),Translation=(X=0,Y=10,Z=0),Scale3D=(X=0.25,Y=1,Z=1))),(Item=(Type=Bone,Name=\"deform_upperarm_bk_l\"),Transform=(Rotation=(X=0,Y=0,Z=0,W=1),Translation=(X=0,Y=-6,Z=0),Scale3 ...(truncated)"
    Weight=@n1.OutputParam
   )
   > n2 = &RigUnit_SphericalPoseReader(
    ActiveRegionScaleFactors=(PositiveWidth=3,NegativeWidth=1,PositiveHeight=1.5,NegativeHeight=1.5)
    ActiveRegionSize=0.1
    Debug=(bDrawDebug=false,DebugScale=25,DebugSegments=20,DebugThickness=0.25)
    DriverAxis=(X=1,Y=0,Z=0)
    DriverItem=(Type=Bone,Name="arm_bend_1_l")
    FalloffRegionScaleFactors=(PositiveWidth=1,NegativeWidth=1,PositiveHeight=1,NegativeHeight=1)
    FalloffSize=0.4
    OptionalParentItem=(Type=Bone)
    RotationOffset=(X=-180,Y=0,Z=180)
   )
   > &RigUnit_ModifyTransforms(
    ItemToModify=((Item=(Type=Bone,Name="deform_upperarm_bk_l"),Transform=(Rotation=(X=0,Y=0,Z=0,W=1),Translation=(X=0,Y=-4,Z=0),Scale3D=(X=1,Y=1,Z=1))))
    Weight=@n2.OutputParam
   )
   &RigUnit_ParentConstraint(
    AdvancedSettings=(RotationOrderForFilter=XZY)
    Child=(Type=Bone,Name="deform_upperarm_fr_base_r")
    Parents=((Item=(Type=Bone,Name="arm_bend_1_r"),Weight=0.3),(Item=(Type=Bone,Name="spine_05"),Weight=0.7))
    Weight=1
    bMaintainOffset=true
    ExecutePin=@RigVMFunction_Sequence.B
   )
   > &RigUnit_ParentConstraint(
    AdvancedSettings=(RotationOrderForFilter=XZY)
    Child=(Type=Bone,Name="deform_upperarm_bk_base_r")
    Parents=((Item=(Type=Bone,Name="arm_bend_1_r"),Weight=0.3),(Item=(Type=Bone,Name="spine_05"),Weight=0.7))
    Weight=1
   )
   > n3 = &RigUnit_SphericalPoseReader(
    ActiveRegionScaleFactors=(PositiveWidth=0.5,NegativeWidth=3,PositiveHeight=1.5,NegativeHeight=1.5)
    ActiveRegionSize=0.1
    Debug=(bDrawDebug=false,DebugScale=25,DebugSegments=20,DebugThickness=0.25)
    DriverAxis=(X=-1,Y=0,Z=0)
    DriverItem=(Type=Bone,Name="arm_bend_1_r")
    FalloffRegionScaleFactors=(PositiveWidth=1,NegativeWidth=1,PositiveHeight=1,NegativeHeight=1)
    FalloffSize=0.4
    OptionalParentItem=(Type=Bone)
    RotationOffset=(X=180,Y=0,Z=180)
   )
   > &RigUnit_ModifyTransforms(
    ItemToModify="((Item=(Type=Bone,Name=\"deform_upperarm_fr_r\"),Transform=(Rotation=(X=0,Y=0,Z=0,W=1),Translation=(X=0,Y=-10,Z=0),Scale3D=(X=0.25,Y=1,Z=1))),(Item=(Type=Bone,Name=\"deform_upperarm_bk_l\"),Transform=(Rotation=(X=0,Y=0,Z=0,W=1),Translation=(X=0,Y=-6,Z=0),Scale ...(truncated)"
    Weight=@n3.OutputParam
   )
   > n4 = &RigUnit_SphericalPoseReader(
    ActiveRegionScaleFactors=(PositiveWidth=3,NegativeWidth=1,PositiveHeight=1.5,NegativeHeight=1.5)
    ActiveRegionSize=0.1
    Debug=(bDrawDebug=false,DebugScale=25,DebugSegments=20,DebugThickness=0.25)
    DriverAxis=(X=-1,Y=0,Z=0)
    DriverItem=(Type=Bone,Name="arm_bend_1_r")
    FalloffRegionScaleFactors=(PositiveWidth=1,NegativeWidth=1,PositiveHeight=1,NegativeHeight=1)
    FalloffSize=0.4
    OptionalParentItem=(Type=Bone)
    RotationOffset=(X=0,Y=0,Z=0)
   )
   > &RigUnit_ModifyTransforms(
    ItemToModify=((Item=(Type=Bone,Name="deform_upperarm_bk_r"),Transform=(Rotation=(X=0,Y=0,Z=0,W=1),Translation=(X=0,Y=4,Z=0),Scale3D=(X=1,Y=1,Z=1))))
    Weight=@n4.OutputParam
   )
   RigVMFunctionReturnNode(ExecuteContext=@RigVMFunction_Sequence.C)
  ]}
 RigVMFunctionLibrary {nodes [
   ApplyFBIKChainCorrection = RigVMCollapseNode(End=(Type=None,Name="None"), ExecuteContext=(), Start=(Type=None,Name="None"))
   DeltaQuatExtraction = RigVMCollapseNode(ExecuteContext=(), child=(Type=Bone,Name="None"), parent=(Type=Bone,Name="None"), twistAxis=(X=1,Y=0,Z=0))
   DeltaRotation = RigVMCollapseNode(ExecuteContext=(), Rotation Order=ZYX, premult=1, twistAxis=(X=1,Y=0,Z=0))
   DeltaRotation v4 = RigVMCollapseNode(
    ExecuteContext=()
    Rotation Order=ZYX
    child=(Type=Bone,Name="None")
    outputMinMax=(X=0,Y=1)
    parent=(Type=Bone,Name="None")
    premult=1
    rotationMinMax=(X=0,Y=0)
    twistAxis=(X=1,Y=0,Z=0)
   )
   PoseSpaceDeformation = RigVMCollapseNode(
    Child=(Type=Bone,Name="None")
    ExecuteContext=()
    Parent=(Type=Bone,Name="None")
    coneAngle=90
    drawLines=true
    primaryAxis=(X=0,Y=0,Z=0)
    quotient=0.5
    targets=()
   )
   SplineTwistTopBot = RigVMCollapseNode(
    ExecuteContext=()
    botRotQuat=()
    chainArray="((Type=Bone,Name=\"arm_bend_1_l\"),(Type=Bone,Name=\"arm_bend_2_l\"),(Type=Bone,Name=\"arm_bend_3_l\"),(Type=Bone,Name=\"arm_bend_4_l\"),(Type=Bone,Name=\"arm_bend_5_l\"),(Type=Bone,Name=\"arm_bend_6_l\"),(Type=Bone,Name=\"arm_bend_7_l\"),(Type=Bone,Name=\"arm_bend_8_l\") ...(truncated)"
    enable=true
    topRotQuat=()
   )
   TwoBoneIKWithSoftness = RigVMCollapseNode(
    ExecuteContext=()
    effectorBone=(Type=None,Name="None")
    firstBone=(Type=None,Name="None")
    firstBoneOrig=(Type=None,Name="None")
    ikEffector=(Type=None,Name="None")
    itemStartStretch=(Type=None,Name="None")
    offsetPv=(X=0,Y=0,Z=0)
    poleVector=(Type=None,Name="None")
    primary Axis=(X=1,Y=0,Z=0)
    secondBone=(Type=None,Name="None")
    secondBoneOrig=(Type=Bone,Name="None")
    secondary Axis=(X=0,Y=1,Z=0)
    softness=1
   )
   armSetup = RigVMCollapseNode(ExecuteContext=())
   bandoliers = RigVMCollapseNode(ExecuteContext=())
   controllers = RigVMCollapseNode(ExecuteContext=(), debug=False)
   custom_deformRig = RigVMCollapseNode(ExecuteContext=())
   custom_deformRig_postKinArms = RigVMCollapseNode(ExecuteContext=())
   legSetup = RigVMCollapseNode(ExecuteContext=())
   misc = RigVMCollapseNode(ExecuteContext=())
   shoulderStraps = RigVMCollapseNode(ExecuteContext=())
   torso = RigVMCollapseNode(ExecuteContext=())
  ]}
 RigVMFunction_Sequence_1_ContainedGraph {nodes [
   n1 = RigVMFunctionEntryNode()
   RigVMFunction_Sequence = &RigVMFunction_Sequence(ExecuteContext=@n1.ExecuteContext)
   n2 = &RigVMFunction_Sequence(ExecuteContext=@RigVMFunction_Sequence.B)
   n3 = &RigVMFunction_Sequence(ExecuteContext=@n2.B)
   n4 = &RigVMFunction_Sequence(ExecuteContext=@n3.B)
   RigVMFunctionReturnNode(A=@RigVMFunction_Sequence.A, B=@n2.A, C=@n3.A, D=@n4.A, E=@n4.B)
  ]}
 RigVMFunction_Sequence_1_ContainedGraph {nodes [
   n1 = RigVMFunctionEntryNode()
   RigVMFunction_Sequence = &RigVMFunction_Sequence(ExecuteContext=@n1.ExecuteContext)
   n2 = &RigVMFunction_Sequence(ExecuteContext=@RigVMFunction_Sequence.B)
   RigVMFunctionReturnNode(A=@RigVMFunction_Sequence.A, B=@n2.A, E=@n2.B)
  ]}
 RigVMFunction_Sequence_1_ContainedGraph {nodes [
   n1 = RigVMFunctionEntryNode()
   RigVMFunction_Sequence = &RigVMFunction_Sequence(ExecuteContext=@n1.ExecuteContext)
   n2 = &RigVMFunction_Sequence(ExecuteContext=@RigVMFunction_Sequence.B)
   RigVMFunctionReturnNode(A=@RigVMFunction_Sequence.A, B=@n2.A, C=@n2.B)
  ]}
 RigVMFunction_Sequence_1_ContainedGraph {nodes [
   n1 = RigVMFunctionEntryNode()
   RigVMFunction_Sequence = &RigVMFunction_Sequence(ExecuteContext=@n1.ExecuteContext)
   n2 = &RigVMFunction_Sequence(ExecuteContext=@RigVMFunction_Sequence.B)
   RigVMFunctionReturnNode(A=@RigVMFunction_Sequence.A, B=@n2.A, C=@n2.B)
  ]}
 RigVMFunction_Sequence_1_ContainedGraph {nodes [
   n1 = RigVMFunctionEntryNode()
   RigVMFunction_Sequence = &RigVMFunction_Sequence(ExecuteContext=@n1.ExecuteContext)
   n2 = &RigVMFunction_Sequence(ExecuteContext=@RigVMFunction_Sequence.B)
   RigVMFunctionReturnNode(A=@RigVMFunction_Sequence.A, B=@n2.A, C=@n2.B)
  ]}
 RigVMModel {nodes [
   RigUnit_BeginExecution = &RigUnit_BeginExecution()
   RigVMFunction_Sequence = RigVMAggregateNode(ExecuteContext=@RigUnit_BeginExecution.ExecutePin)
   legSetup = RigVMFunctionReferenceNode(ExecuteContext=@RigVMFunction_Sequence.B)
   > armSetup = RigVMFunctionReferenceNode()
   custom_deformRig_postKinArms = RigVMFunctionReferenceNode(ExecuteContext=@RigVMFunction_Sequence.C)
   bandoliers = RigVMFunctionReferenceNode(ExecuteContext=@RigVMFunction_Sequence.D)
   > torso = RigVMFunctionReferenceNode()
   shoulderStraps = RigVMFunctionReferenceNode(ExecuteContext=@RigVMFunction_Sequence.E)
   > misc = RigVMFunctionReferenceNode()
   custom_deformRig = RigVMFunctionReferenceNode(ExecuteContext=@RigVMFunction_Sequence.A)
  ]}
 RigVMModel Rig {nodes [
   n1 = &RigUnit_PrepareForExecution()
   ComputePoleVector = RigVMFunctionReferenceNode(
    Bone A=(Type=Bone,Name="upperarm_l")
    Bone B=(Type=Bone,Name="lowerarm_l")
    Bone C=(Type=Bone,Name="hand_l")
    Draw Transform=true
    Initial=true
    OffsetFactor=10
    Origin Scale=(X=0,Y=0,Z=0)
    ExecuteContext=@n1.ExecutePin
   )
   > &RigUnit_SetTransform(
    Item=(Type=Control,Name="arm_l_pv_ctrl")
    Value=(Rotation=(X=0,Y=0,Z=0,W=1),Translation=(X=0,Y=0,Z=0),Scale3D=(X=1,Y=1,Z=1))
    Weight=1
    bInitial=true
    Value.Translation=@ComputePoleVector.Transform.Translation
   )
   > ComputePoleVector_1 = RigVMFunctionReferenceNode(
    Bone A=(Type=Bone,Name="upperarm_r")
    Bone B=(Type=Bone,Name="lowerarm_r")
    Bone C=(Type=Bone,Name="hand_r")
    Draw Transform=true
    Initial=true
    OffsetFactor=10
    Origin Scale=(X=0,Y=0,Z=0)
   )
   > &RigUnit_SetTransform(
    Item=(Type=Control,Name="arm_r_pv_ctrl")
    Value=(Rotation=(X=0,Y=0,Z=0,W=1),Translation=(X=0,Y=0,Z=0),Scale3D=(X=1,Y=1,Z=1))
    Weight=1
    bInitial=true
    Value.Translation=@ComputePoleVector_1.Transform.Translation
   )
  ]}
]
hierarchy [
 root {loc=<0 0 0>, rot=<0 0 0 1>}
 pelvis {parent=root, loc=<0 -0.7321 92.4218>, rot=<0.0127 0.707 -0.0127 -0.707>}
 spine_01 {parent=pelvis, loc=<5.6002 -0.5128 0>, rot=<0 0 0.034 -0.9994>}
 spine_02 {parent=spine_01, loc=<8.0756 0.0139 0>, rot=<0 0 0.0238 -0.9997>}
 spine_03 {parent=spine_02, loc=<7.0395 0 0>, rot=<0 0 0.0149 -0.9999>}
 spine_04 {parent=spine_03, loc=<8.1023 0 0>, rot=<0 0 -0.0132 -0.9999>}
 spine_05 {parent=spine_04, loc=<10.0141 0 0>, rot=<0 0 -0.0015 -1>}
 clavicle_l {parent=spine_05, loc=<2.5856 -0.2866 -4.3777>, rot=<0.0465 -0.7289 -0.009 -0.6829>}
 upperarm_l {parent=clavicle_l, loc=<18.2264 0.1342 -0.001>, rot=<0.0253 -0.2693 0.0013 -0.9627>}
 lowerarm_l {parent=upperarm_l, loc=<29.0071 0.1686 -1.239>, rot=<0.0391 -0.0099 -0.2453 -0.9686>}
 hand_l {parent=lowerarm_l, loc=<25.8607 0.222 1.0274>, rot=<-0.7193 -0.0988 0.1193 -0.6773>}
 thumb_01_l {parent=hand_l, loc=<1.3118 -0.3191 -2.459>, rot=<0.6546 -0.4042 0.0664 -0.6354>}
 thumb_02_l {parent=thumb_01_l, loc=<5.9336 -0.0448 0.0401>, rot=<-0.0037 0.0213 0.298 -0.9543>}
 thumb_03_l {parent=thumb_02_l, loc=<2.2223 0.3682 -0.0907>, rot=<0.0118 -0.013 0.0066 -0.9998>}
 index_metacarpal_l {parent=hand_l, loc=<1.6535 0.0183 -2.96>, rot=<0.0158 -0.0124 -0.0173 -0.9996>}
 index_01_l {parent=index_metacarpal_l, loc=<8.3008 0.0357 -0.0524>, rot=<-0.0044 0.0485 0.0924 -0.9945>}
 index_02_l {parent=index_01_l, loc=<3.3587 -0.3004 0.0452>, rot=<-0.002 0.0027 0.3225 -0.9466>}
 index_03_l {parent=index_02_l, loc=<2.5578 -0.123 -0.0252>, rot=<-0.002 0.0067 0.2558 -0.9667>}
 middle_metacarpal_l {parent=hand_l, loc=<2.1856 -0.1725 -0.5984>, rot=<0.006 0.0168 -0.0329 -0.9993>}
 middle_01_l {parent=middle_metacarpal_l, loc=<7.3275 -0.2447 0.0129>, rot=<-0.0208 0.0042 0.0858 -0.9961>}
 middle_02_l {parent=middle_01_l, loc=<4.0123 0.1401 0.1175>, rot=<-0.0218 0.0258 0.3512 -0.9357>}
 middle_03_l {parent=middle_02_l, loc=<2.6077 0.0534 0.1373>, rot=<0.0082 0.0188 0.1708 -0.9851>}
 ring_metacarpal_l {parent=hand_l, loc=<2.2074 -0.3625 1.3638>, rot=<-0.0284 0.0722 -0.0458 -0.9959>}
 ring_01_l {parent=ring_metacarpal_l, loc=<7.0295 -0.6197 -0.1446>, rot=<-0.0489 -0.0252 0.1892 -0.9804>}
 ring_02_l {parent=ring_01_l, loc=<3.6174 0.4647 -0.1027>, rot=<0 0 0.2806 -0.9598>}
 ring_03_l {parent=ring_02_l, loc=<2.4152 0.114 0.0199>, rot=<0 0 0.1189 -0.9929>}
 pinky_metacarpal_l {parent=hand_l, loc=<1.7616 -0.9753 3.3814>, rot=<-0.0397 0.0814 -0.069 -0.9935>}
 pinky_01_l {parent=pinky_metacarpal_l, loc=<7.1709 -0.4831 -0.1094>, rot=<-0.0654 -0.0035 0.2242 -0.9723>}
 pinky_02_l {parent=pinky_01_l, loc=<2.8274 0.0893 -0.1043>, rot=<0.0037 0.0011 0.2854 -0.9584>}
 pinky_03_l {parent=pinky_02_l, loc=<2.2046 0.0307 -0.0078>, rot=<-0.0002 0.0098 0.1445 -0.9895>}
 hand_ik_l {parent=hand_l, loc=<0 0 0>, rot=<0 0 0 1>}
 hand_l_null {parent=lowerarm_l, type=Null, loc=<25.8607 0.222 1.0274>, rot=<-0.7193 -0.0988 0.1193 -0.6773>}
 lowerarm_twist_02_l {parent=lowerarm_l, loc=<9.6199 0 0>, rot=<0 0 0 -1>}
 lowerarm_twist_01_l {parent=lowerarm_l, loc=<17.8549 0 0>, rot=<0 0 0 -1>}
 lowerarm_l_null {parent=upperarm_l, type=Null, loc=<29.0071 0.1686 -1.239>, rot=<0.0391 -0.0099 -0.2453 -0.9686>}
 upperarm_twist_01_l {parent=upperarm_l, loc=<7.9796 0.0843 -0.6195>, rot=<0 0 0 -1>}
 upperarm_twist_02_l {parent=upperarm_l, loc=<19.5059 0.0843 -0.6195>, rot=<0 0 0 -1>}
 neck_01 {parent=spine_05, loc=<10.815 1.9196 0>, rot=<0 0 -0.1864 -0.9825>}
 neck_02 {parent=neck_01, loc=<4.7347 0.3045 0>, rot=<0 0 0.0363 -0.9993>}
 head {parent=neck_02, loc=<5.7632 0.5552 0>, rot=<0 0 0.115 -0.9934>}
 clavicle_r {parent=spine_05, loc=<2.5855 -0.2866 4.3777>, rot=<0.7289 0.0465 -0.6829 0.009>}
 upperarm_r {parent=clavicle_r, loc=<-18.2264 -0.1342 0.001>, rot=<0.0253 -0.2693 0.0013 -0.9627>}
 lowerarm_r {parent=upperarm_r, loc=<-29.0071 -0.1686 1.239>, rot=<0.0391 -0.0099 -0.2453 -0.9686>}
 hand_r {parent=lowerarm_r, loc=<-25.8607 -0.222 -1.0274>, rot=<-0.7193 -0.0988 0.1193 -0.6773>}
 thumb_01_r {parent=hand_r, loc=<-1.3118 0.3191 2.459>, rot=<0.6546 -0.4042 0.0664 -0.6354>}
 thumb_02_r {parent=thumb_01_r, loc=<-5.9336 0.0448 -0.0401>, rot=<-0.0037 0.0213 0.298 -0.9543>}
 thumb_03_r {parent=thumb_02_r, loc=<-2.2223 -0.3682 0.0907>, rot=<0.0118 -0.013 0.0066 -0.9998>}
 index_metacarpal_r {parent=hand_r, loc=<-1.6535 -0.0183 2.96>, rot=<0.0158 -0.0124 -0.0173 -0.9996>}
 index_01_r {parent=index_metacarpal_r, loc=<-8.3008 -0.0357 0.0524>, rot=<-0.0044 0.0485 0.0924 -0.9945>}
 index_02_r {parent=index_01_r, loc=<-3.3587 0.3004 -0.0452>, rot=<-0.002 0.0027 0.3225 -0.9466>}
 index_03_r {parent=index_02_r, loc=<-2.5578 0.123 0.0252>, rot=<-0.002 0.0067 0.2558 -0.9667>}
 middle_metacarpal_r {parent=hand_r, loc=<-2.1856 0.1725 0.5984>, rot=<0.006 0.0168 -0.0329 -0.9993>}
 middle_01_r {parent=middle_metacarpal_r, loc=<-7.3275 0.2447 -0.0129>, rot=<-0.0208 0.0042 0.0858 -0.9961>}
 middle_02_r {parent=middle_01_r, loc=<-4.0123 -0.1401 -0.1175>, rot=<-0.0218 0.0258 0.3512 -0.9357>}
 middle_03_r {parent=middle_02_r, loc=<-2.6077 -0.0534 -0.1373>, rot=<0.0082 0.0188 0.1708 -0.9851>}
 ring_metacarpal_r {parent=hand_r, loc=<-2.2074 0.3625 -1.3638>, rot=<-0.0284 0.0722 -0.0458 -0.9959>}
 ring_01_r {parent=ring_metacarpal_r, loc=<-7.0295 0.6197 0.1446>, rot=<-0.0489 -0.0252 0.1892 -0.9804>}
 ring_02_r {parent=ring_01_r, loc=<-3.6174 -0.4647 0.1027>, rot=<0 0 0.2806 -0.9598>}
 ring_03_r {parent=ring_02_r, loc=<-2.4152 -0.114 -0.0199>, rot=<0 0 0.1189 -0.9929>}
 pinky_metacarpal_r {parent=hand_r, loc=<-1.7616 0.9753 -3.3814>, rot=<-0.0397 0.0814 -0.069 -0.9935>}
 pinky_01_r {parent=pinky_metacarpal_r, loc=<-7.1709 0.4831 0.1094>, rot=<-0.0654 -0.0035 0.2242 -0.9723>}
 pinky_02_r {parent=pinky_01_r, loc=<-2.8274 -0.0893 0.1043>, rot=<0.0037 0.0011 0.2854 -0.9584>}
 pinky_03_r {parent=pinky_02_r, loc=<-2.2046 -0.0307 0.0078>, rot=<-0.0002 0.0098 0.1445 -0.9895>}
 hand_ik_r {parent=hand_r, loc=<0 0 0>, rot=<0 0 0 1>}
 hand_r_null {parent=lowerarm_r, type=Null, loc=<-25.8607 -0.222 -1.0274>, rot=<-0.7193 -0.0988 0.1193 -0.6773>}
 lowerarm_twist_02_r {parent=lowerarm_r, loc=<-9.6199 0 0>, rot=<0 0 0 -1>}
 lowerarm_twist_01_r {parent=lowerarm_r, loc=<-17.8549 0 0>, rot=<0 0 0 -1>}
 lowerarm_r_null {parent=upperarm_r, type=Null, loc=<-29.0071 -0.1686 1.239>, rot=<0.0391 -0.0099 -0.2453 -0.9686>}
 upperarm_twist_01_r {parent=upperarm_r, loc=<-7.9796 -0.0843 0.6195>, rot=<0 0 0 -1>}
 upperarm_twist_02_r {parent=upperarm_r, loc=<-19.5059 -0.0843 0.6195>, rot=<0 0 0 -1>}
 attach_cape {parent=spine_05, loc=<9.017 -11.714 0>, rot=<0.0616 -0.7044 -0.0616 -0.7044>}
 cape {parent=attach_cape, type=Null, loc=<0 -0.6769 -3.5744>, rot=<0 0 0 1>}
 attach_backpack {parent=spine_05, loc=<-0.123 -13.473 0>, rot=<-0.0093 -0.707 0.0093 -0.707>}
 Backpack_BR {parent=attach_backpack, type=Null, loc=<0 -1.2 -1.5>, rot=<-0.055 0 0 0.9985>}
 upperarmBend_l {parent=spine_05, loc=<1.4284 -1.1642 -22.5467>, rot=<0.3456 -0.6483 -0.6174 -0.2811>}
 lowerarmBend_l {parent=upperarmBend_l, loc=<29.034 0 0>, rot=<0 0.2483 0 -0.9687>}
 lowerarmBend_l_end {parent=lowerarmBend_l, loc=<25.8821 0 0>, rot=<0 0 0 1>}
 upperarmBend_r {parent=spine_05, loc=<1.4289 -1.1641 22.5467>, rot=<-0.6483 -0.3456 0.2811 -0.6174>}
 lowerarmBend_r {parent=upperarmBend_r, loc=<-29.0342 -0.0003 0>, rot=<0 0.2483 0 -0.9687>}
 lowerarmBend_r_end {parent=lowerarmBend_r, loc=<-25.8823 -0.0002 0.0001>, rot=<0 0 0 1>}
 base_topLid {parent=spine_05, loc=<10.7157 9.7715 0.0056>, rot=<-0.0389 -0.9992 0 0>}
 dyn_topLid {parent=base_topLid, loc=<0 0 0>, rot=<0 0 0 1>}
 upperarm_l_null {parent=spine_05, type=Null, loc=<1.4284 -1.1642 -22.5467>, rot=<-0.0654 0.8854 0.0137 0.46>}
 upperarm_r_null {parent=spine_05, type=Null, loc=<1.4284 -1.1642 22.5467>, rot=<-0.8854 -0.0654 0.46 -0.0137>}
 arm_bend_1_l {parent=spine_05, loc=<1.4284 -1.1642 -22.5467>, rot=<0.3723 -0.6388 -0.6268 -0.2459>}
 arm_bend_2_l {parent=arm_bend_1_l, loc=<4.2041 0 0>, rot=<-0.0002 0.0191 0.0017 -0.9998>}
 arm_bend_3_l {parent=arm_bend_2_l, loc=<4.2041 0 0>, rot=<-0.0001 0.012 0.0018 -0.9999>}
 arm_bend_4_l {parent=arm_bend_3_l, loc=<4.2045 0 0>, rot=<-0.0001 0.0062 0.0018 -1>}
 arm_bend_5_l {parent=arm_bend_4_l, loc=<4.2042 0 0>, rot=<0 0.0161 0.0016 -0.9999>}
 arm_bend_6_l {parent=arm_bend_5_l, loc=<4.2043 0 0>, rot=<0 0.0341 0.0011 -0.9994>}
 arm_bend_7_l {parent=arm_bend_6_l, loc=<4.2028 0 0>, rot=<0.0001 0.0523 0.0007 -0.9986>}
 arm_bend_8_l {parent=arm_bend_7_l, loc=<4.2019 0 0>, rot=<0.0003 0.0545 0.0012 -0.9985>}
 arm_bend_9_l {parent=arm_bend_8_l, loc=<4.2026 0 0>, rot=<0.0008 0.0409 0.0025 -0.9992>}
 arm_bend_10_l {parent=arm_bend_9_l, loc=<4.2037 0 0>, rot=<0.0016 0.0276 0.004 -0.9996>}
 arm_bend_11_l {parent=arm_bend_10_l, loc=<4.2042 0 0>, rot=<0.0024 0.019 0.0054 -0.9998>}
 arm_bend_12_l {parent=arm_bend_11_l, loc=<4.2042 0 0>, rot=<0.0034 0.0194 0.0067 -0.9998>}
 arm_bend_13_l {parent=arm_bend_12_l, loc=<4.2039 0 0>, rot=<0.0044 0.0202 0.0079 -0.9998>}
 arm_bend_14_l {parent=arm_bend_13_l, loc=<4.2041 0 0>, rot=<0 0 0 1>}
 deform_upperarm_bk_base_l {parent=arm_bend_1_l, loc=<1.9876 1.6737 -13.2305>, rot=<-0.7014 0.0402 -0.0029 -0.7116>}
 deform_upperarm_bk_l {parent=deform_upperarm_bk_base_l, loc=<0 0 0>, rot=<0 0 0 1>}
 deform_upperarm_fr_base_l {parent=arm_bend_1_l, loc=<-0.8756 2.9524 10.7168>, rot=<-0.7014 0.0402 -0.0029 -0.7116>}
 deform_upperarm_fr_l {parent=deform_upperarm_fr_base_l, loc=<0 0 0>, rot=<0 0 0 1>}
 deform_upperarm_up_base_l {parent=arm_bend_1_l, loc=<-3.5758 -7.9965 -1.2308>, rot=<-0.7014 0.0402 -0.0029 -0.7116>}
 deform_upperarm_up_l {parent=deform_upperarm_up_base_l, loc=<0 0 0>, rot=<0 0 0 1>}
 arm_bend_1_r {parent=spine_05, loc=<1.4289 -1.1641 22.5467>, rot=<-0.6388 -0.3723 0.2459 -0.6268>}
 arm_bend_2_r {parent=arm_bend_1_r, loc=<-4.2045 -0.0005 -0.0001>, rot=<-0.0002 0.0191 0.0017 -0.9998>}
 arm_bend_3_r {parent=arm_bend_2_r, loc=<-4.2041 0.0001 0>, rot=<-0.0001 0.012 0.0018 -0.9999>}
 arm_bend_4_r {parent=arm_bend_3_r, loc=<-4.2047 -0.0002 0>, rot=<-0.0001 0.0062 0.0018 -1>}
 arm_bend_5_r {parent=arm_bend_4_r, loc=<-4.2043 -0.0001 0>, rot=<0 0.0161 0.0016 -0.9999>}
 arm_bend_6_r {parent=arm_bend_5_r, loc=<-4.2037 0.0007 0>, rot=<0 0.0341 0.0011 -0.9994>}
 arm_bend_7_r {parent=arm_bend_6_r, loc=<-4.2029 -0.0001 0>, rot=<0.0001 0.0523 0.0007 -0.9986>}
 arm_bend_8_r {parent=arm_bend_7_r, loc=<-4.2021 -0.0002 0>, rot=<0.0003 0.0545 0.0012 -0.9985>}
 arm_bend_9_r {parent=arm_bend_8_r, loc=<-4.2025 0 0>, rot=<0.0008 0.0409 0.0025 -0.9992>}
 arm_bend_10_r {parent=arm_bend_9_r, loc=<-4.204 -0.0004 0.0001>, rot=<0.0016 0.0276 0.004 -0.9996>}
 arm_bend_11_r {parent=arm_bend_10_r, loc=<-4.2038 0.0006 -0.0002>, rot=<0.0024 0.019 0.0054 -0.9998>}
 arm_bend_12_r {parent=arm_bend_11_r, loc=<-4.2046 -0.0006 0.0002>, rot=<0.0034 0.0194 0.0067 -0.9998>}
 arm_bend_13_r {parent=arm_bend_12_r, loc=<-4.2035 0.0006 -0.0002>, rot=<0.0044 0.0202 0.0079 -0.9998>}
 arm_bend_14_r {parent=arm_bend_13_r, loc=<-4.2044 -0.0004 0.0002>, rot=<0 0 0 1>}
 deform_upperarm_bk_base_r {parent=arm_bend_1_r, loc=<-1.9878 -1.6739 13.2305>, rot=<-0.7014 0.0402 -0.0029 -0.7116>}
 deform_upperarm_bk_r {parent=deform_upperarm_bk_base_r, loc=<0 0 0>, rot=<0 0 0 1>}
 deform_upperarm_fr_base_r {parent=arm_bend_1_r, loc=<0.8751 -2.953 -10.7169>, rot=<-0.7014 0.0402 -0.0029 -0.7116>}
 deform_upperarm_fr_r {parent=deform_upperarm_fr_base_r, loc=<0 0 0>, rot=<0 0 0 1>}
 deform_upperarm_up_base_r {parent=arm_bend_1_r, loc=<3.5758 7.9964 1.2308>, rot=<-0.7014 0.0402 -0.0029 -0.7116>}
 deform_upperarm_up_r {parent=deform_upperarm_up_base_r, loc=<0 0 0>, rot=<0 0 0 1>}
 base_armPacks_l {parent=spine_05, loc=<0.7452 -0.5759 -25.2174>, rot=<-0.3703 0.6256 0.6289 0.2758>}
 dyn_armPacks_l {parent=base_armPacks_l, loc=<0 0 0>, rot=<0 0 0 1>}
 base_armPacks_r {parent=spine_05, loc=<0.7455 -0.5759 25.2174>, rot=<0.3703 -0.6256 0.6289 0.2758>}
 dyn_armPacks_r {parent=base_armPacks_r, loc=<0 0 0>, rot=<0 0 0 1>}
 base_backPackHandle {parent=spine_05, loc=<10.6065 -8.2681 0.0012>, rot=<-0.0073 0.0227 0.2299 -0.9729>}
 dyn_backPackHandle {parent=base_backPackHandle, loc=<0 0 0>, rot=<0 0 0 1>}
 base_back_zipperSlider {parent=spine_05, loc=<1.3957 -11.9077 11.53>, rot=<0.0284 -0.7065 0.7065 -0.0284>}
 dyn_back_zipperSlider {parent=base_back_zipperSlider, loc=<0 0 0>, rot=<0 0 0 1>}
 base_shoulderStraps_1_r {parent=spine_05, loc=<11.25 -9.2441 7.4584>, rot=<0.139 -0.6897 0.7009 -0.1174>}
 dyn_shoulderStraps_1_r {parent=base_shoulderStraps_1_r, loc=<0 0 0>, rot=<0 0 0 1>}
 dyn_shoulderStraps_2_r {parent=dyn_shoulderStraps_1_r, loc=<7.3883 0 -0.0001>, rot=<0 -0.0763 0 -0.9971>}
 dyn_shoulderStraps_3_r {parent=dyn_shoulderStraps_2_r, loc=<7.3802 0 0>, rot=<0.0045 -0.0721 0.0066 -0.9974>}
 dyn_shoulderStraps_4_r {parent=dyn_shoulderStraps_3_r, loc=<7.3893 0 0>, rot=<0.0005 -0.0014 -0.0047 -1>}
 dyn_shoulderStraps_5_r {parent=dyn_shoulderStraps_4_r, loc=<7.3869 0 0>, rot=<-0.01 0.0124 0.0028 -0.9999>}
 dyn_shoulderStraps_6_r {parent=dyn_shoulderStraps_5_r, loc=<7.3999 0 0>, rot=<0.0126 -0.0533 0.0101 -0.9984>}
 dyn_shoulderStraps_7_r {parent=dyn_shoulderStraps_6_r, loc=<7.3545 0 0.0001>, rot=<0 0 0 1>}
 dyn_shoulderStrapsTip_1_r {parent=dyn_shoulderStraps_7_r, loc=<0.1165 -0.001 0.0193>, rot=<-0.0004 0.0818 0.0042 -0.9966>}
 dyn_shoulderStrapsTip_2_r {parent=dyn_shoulderStrapsTip_1_r, loc=<6.0794 0 0>, rot=<0 0.056 0 -0.9984>}
 base_shoulderStraps_1_l {parent=spine_05, loc=<11.2499 -9.2441 -7.4584>, rot=<0.1174 -0.7009 0.6897 -0.139>}
 dyn_shoulderStraps_1_l {parent=base_shoulderStraps_1_l, loc=<0 0 0>, rot=<0 0 0 1>}
 dyn_shoulderStraps_2_l {parent=dyn_shoulderStraps_1_l, loc=<7.388 0 0>, rot=<0 -0.0763 0 -0.9971>}
 dyn_shoulderStraps_3_l {parent=dyn_shoulderStraps_2_l, loc=<7.3805 0 0>, rot=<-0.0045 -0.0721 -0.0066 -0.9974>}
 dyn_shoulderStraps_4_l {parent=dyn_shoulderStraps_3_l, loc=<7.389 0 0>, rot=<-0.0005 -0.0014 0.0047 -1>}
 dyn_shoulderStraps_5_l {parent=dyn_shoulderStraps_4_l, loc=<7.3876 0 0>, rot=<0.01 0.0124 -0.0028 -0.9999>}
 dyn_shoulderStraps_6_l {parent=dyn_shoulderStraps_5_l, loc=<7.3994 0 0>, rot=<-0.0126 -0.0533 -0.0101 -0.9984>}
 dyn_shoulderStraps_7_l {parent=dyn_shoulderStraps_6_l, loc=<7.3547 0 0>, rot=<0 0 0 1>}
 dyn_shoulderStrapsTip_1_l {parent=dyn_shoulderStraps_7_l, loc=<0.1165 0.001 0.0193>, rot=<0.0004 0.0818 -0.0042 -0.9966>}
 dyn_shoulderStrapsTip_2_l {parent=dyn_shoulderStrapsTip_1_l, loc=<6.0794 0 0>, rot=<0 0.056 0 -0.9984>}
 clavicleFake_l {parent=spine_05, loc=<1.4284 -1.1642 -22.3467>, rot=<0.7293 0.0451 0.6825 -0.0126>}
 clavicleFake_r {parent=spine_05, loc=<1.4289 -1.1641 22.3467>, rot=<-0.0451 0.7293 -0.0126 -0.6825>}
 upperarmBend_r_null {parent=spine_05, type=Null, loc=<1.4289 -1.1641 22.5467>, rot=<-0.6493 -0.3478 0.2787 -0.6162>}
 upperarmBend_l_null {parent=spine_05, type=Null, loc=<1.4284 -1.1642 -22.5467>, rot=<0.3456 -0.6483 -0.6174 -0.2811>}
 kin_sidePack_1 {parent=spine_04, loc=<1.2538 2.0514 -18.5576>, rot=<0.0294 -0.7065 -0.0294 -0.7065>}
 base_thermos {parent=spine_03, loc=<-2.5505 -2.4378 22.6939>, rot=<-0.004 0.0494 -0.0296 -0.9983>}
 dyn_thermos {parent=base_thermos, loc=<0 0 0>, rot=<0 0 0 1>}
 base_thermos_rope_a {parent=dyn_thermos, loc=<6.2228 2.1581 4.2453>, rot=<0.3179 0.5706 0.7372 0.1732>}
 dyn_thermos_rope_a {parent=base_thermos_rope_a, loc=<0 0 0>, rot=<0 0 0 1>}
 base_thermos_rope_c {parent=dyn_thermos, loc=<-11.663 0.1522 4.9457>, rot=<0.0496 0.0012 0.9985 -0.0251>}
 dyn_thermos_rope_c {parent=base_thermos_rope_c, loc=<0 0 0>, rot=<0 0 0 1>}
 base_thermos_rope_b {parent=dyn_thermos, loc=<-11.5648 -1.8016 4.9457>, rot=<0.0496 0.0012 0.9985 -0.0251>}
 dyn_thermos_rope_b {parent=base_thermos_rope_b, loc=<0 0 0>, rot=<0 0 0 1>}
 base_frontLogo {parent=spine_02, loc=<-3.4308 18.6419 -10.1999>, rot=<0.0432 -0.8136 -0.5797 0.0081>}
 dyn_frontLogo {parent=base_frontLogo, loc=<0 0 0>, rot=<0 0 0 1>}
 kin_sidePack_2 {parent=spine_02, loc=<9.1576 -0.4784 -18.5576>, rot=<0.0282 -0.7065 -0.0282 -0.7065>}
 kin_sidePack_retract {parent=kin_sidePack_2, loc=<3.034 -0.02 -16.8745>, rot=<0 0 -0.0718 -0.9974>}
 base_mug {parent=kin_sidePack_retract, loc=<1.2073 0.2933 -4.105>, rot=<0.0121 -0.5049 -0.1423 -0.8513>}
 dyn_mug {parent=base_mug, loc=<0 0 0>, rot=<0 0 0 1>}
 base_zipperFront_A {parent=spine_02, loc=<5.4837 12.089 -15.0559>, rot=<-0.0336 -0.9084 -0.3128 -0.2753>}
 dyn_zipperFront_A {parent=base_zipperFront_A, loc=<0 0 0>, rot=<0 0 0 1>}
 base_zipperFront_B {parent=spine_02, loc=<0.8974 17.7733 -10.6752>, rot=<-0.006 -0.8799 -0.4268 -0.2087>}
 dyn_zipperFront_B {parent=base_zipperFront_B, loc=<0 0 0>, rot=<0 0 0 1>}
 kin_frontPack_l {parent=spine_01, loc=<-1.8721 12.4039 -8.0246>, rot=<0.4919 0.508 0.508 0.4919>}
 kin_frontPack_r {parent=spine_01, loc=<-1.8722 12.4039 8.0246>, rot=<0.4919 0.508 0.508 0.4919>}
 base_zipperFront_A_aimTarget_null {parent=spine_01, type=Null, loc=<5.4777 12.2166 -16.4645>, rot=<0 0 0 1>}
 thigh_l {parent=pelvis, loc=<-1.8716 -0.0109 -11.9936>, rot=<0.0012 0.0572 0.0155 -0.9982>}
 calf_l {parent=thigh_l, loc=<-41.9178 -0.2246 0.0926>, rot=<0.0006 -0.0161 -0.0391 -0.9991>}
 foot_l {parent=calf_l, loc=<-38.7057 1.2438 0.2452>, rot=<0.0975 -0.0474 0.0317 -0.9936>}
 ball_l {parent=foot_l, loc=<-6.1614 13.1469 0.6807>, rot=<0.0001 -0.0001 -0.7063 -0.7079>}
 calf_twist_02_l {parent=calf_l, loc=<-15.9268 -0.0967 0.0397>, rot=<0.0237 0.0095 -0.0035 -0.9997>}
 calf_twist_01_l {parent=calf_l, loc=<-32.9335 -0.0967 0.0397>, rot=<0.0237 0.0095 -0.0035 -0.9997>}
 deform_knee_in_base_l {parent=calf_l, loc=<1.298 -8.0356 -0.0496>, rot=<0 0 0 -1>}
 deform_knee_in_l {parent=deform_knee_in_base_l, loc=<0 0 0>, rot=<0 0 0 1>}
 deform_knee_out_base_l {parent=calf_l, loc=<0 6.8739 0>, rot=<0 0 0 -1>}
 deform_knee_out_l {parent=deform_knee_out_base_l, loc=<0 0 0>, rot=<0 0 0 1>}
 thigh_twist_01_l {parent=thigh_l, loc=<-21.6303 0 0>, rot=<0.0013 0 0 -1>}
 deform_groin_base_l {parent=thigh_l, loc=<-2.1908 8.5661 -1.7442>, rot=<0 0 0 -1>}
 deform_groin_l {parent=deform_groin_base_l, loc=<0 0 0>, rot=<0 0 0 1>}
 calf_l_null {parent=thigh_l, type=Null, loc=<-41.9178 -0.2246 0.0926>, rot=<0.0006 -0.0161 -0.0391 -0.9991>}
 calfRePos_l {parent=thigh_l, loc=<-41.9178 -0.2246 0.0926>, rot=<-0.0061 -0.0164 -0.0583 -0.9981>}
 footRePos_l {parent=calfRePos_l, loc=<-45.4926 0 0>, rot=<0.0025 -0.0406 0.061 -0.9973>}
 footRePos_l_upVector {parent=calfRePos_l, type=Null, loc=<-45.4926 0 -5>, rot=<-0.0025 0.0406 -0.061 0.9973>}
 calfRePos_bend_1_l {parent=calfRePos_l, loc=<0 0 0>, rot=<0 0 0 1>}
 calfRePos_bend_2_l {parent=calfRePos_l, loc=<-11.3732 0 0>, rot=<0 0 0 1>}
 calfRePos_bend_3_l {parent=calfRePos_l, loc=<-22.7463 0 0>, rot=<0 0 0 1>}
 calfRePos_bend_4_l {parent=calfRePos_l, loc=<-34.1195 0 0>, rot=<0 0 0 1>}
 calfRePos_bend_5_l {parent=calfRePos_l, loc=<-45.1195 0 0>, rot=<0 0 0 1>}
 kin_legsFitter_l {parent=thigh_l, loc=<-41.5699 -0.0432 -4.3711>, rot=<0.7884 -0.5405 -0.2897 0.0491>}
 base_legsFitter_A_l {parent=kin_legsFitter_l, loc=<-0.0739 -0.4079 0.4015>, rot=<-0.5655 0.4254 -0.4779 -0.5204>}
 dyn_legsFitter_A_l {parent=base_legsFitter_A_l, loc=<0 0 0>, rot=<0 0 0 1>}
 base_legsFitter_B_l {parent=kin_legsFitter_l, loc=<-1.6055 0.2198 0.474>, rot=<-0.039 0.3623 -0.612 -0.7019>}
 dyn_legsFitter_B_l {parent=base_legsFitter_B_l, loc=<0 0 0>, rot=<0 0 0 1>}
 thigh_r {parent=pelvis, loc=<-1.8716 -0.0109 11.9936>, rot=<0.0572 -0.0012 0.9982 0.0155>}
 calf_r {parent=thigh_r, loc=<41.9178 0.2246 -0.0926>, rot=<0.0006 -0.0161 -0.0391 -0.9991>}
 foot_r {parent=calf_r, loc=<38.7057 -1.2438 -0.2452>, rot=<0.0975 -0.0474 0.0317 -0.9936>}
 ball_r {parent=foot_r, loc=<6.1614 -13.1469 -0.6807>, rot=<0.0001 -0.0001 -0.7063 -0.7079>}
 calf_twist_02_r {parent=calf_r, loc=<15.9268 0.0967 -0.0397>, rot=<0.0237 0.0095 -0.0035 -0.9997>}
 calf_twist_01_r {parent=calf_r, loc=<32.9335 0.0967 -0.0397>, rot=<0.0237 0.0095 -0.0035 -0.9997>}
 deform_knee_out_base_r {parent=calf_r, loc=<0 -6.8739 0>, rot=<0 0 0 -1>}
 deform_knee_out_r {parent=deform_knee_out_base_r, loc=<0 0 0>, rot=<0 0 0 1>}
 deform_knee_in_base_r {parent=calf_r, loc=<-1.2981 8.0356 0.0496>, rot=<0 0 0 -1>}
 deform_knee_in_r {parent=deform_knee_in_base_r, loc=<0 0 0>, rot=<0 0 0 1>}
 thigh_twist_01_r {parent=thigh_r, loc=<21.6303 0 0>, rot=<0.0013 0 0 -1>}
 deform_groin_base_r {parent=thigh_r, loc=<2.1908 -8.5661 1.7442>, rot=<0 0 0 -1>}
 deform_groin_r {parent=deform_groin_base_r, loc=<0 0 0>, rot=<0 0 0 -1>}
 calf_r_null {parent=thigh_r, type=Null, loc=<41.9178 0.2246 -0.0926>, rot=<0.0006 -0.0161 -0.0391 -0.9991>}
 calfRePos_r {parent=thigh_r, loc=<41.9178 0.2247 -0.0926>, rot=<-0.0061 -0.0164 -0.0583 -0.9981>}
 footRePos_r {parent=calfRePos_r, loc=<45.4926 0 0>, rot=<0.0025 -0.0406 0.061 -0.9973>}
 footRePos_r_upVector {parent=calfRePos_r, type=Null, loc=<45.4926 0 5>, rot=<-0.0025 0.0406 -0.061 0.9973>}
 calfRePos_bend_1_r {parent=calfRePos_r, loc=<0 0 0>, rot=<0 0 0 1>}
 calfRePos_bend_2_r {parent=calfRePos_r, loc=<11.3731 0 0>, rot=<0 0 0 1>}
 calfRePos_bend_3_r {parent=calfRePos_r, loc=<22.7463 0 0>, rot=<0 0 0 1>}
 calfRePos_bend_4_r {parent=calfRePos_r, loc=<34.1194 0 0>, rot=<0 0 0 1>}
 base_calfOrn_l {parent=calfRePos_bend_4_r, loc=<8.5358 0.3257 2.6045>, rot=<-0.0156 0.7234 0.3556 -0.5915>}
 dyn_calfOrn_l {parent=base_calfOrn_l, loc=<0 0 0>, rot=<0 0 0 1>}
 calfRePos_bend_5_r {parent=calfRePos_r, loc=<45.1194 0 0>, rot=<0 0 0 1>}
 kin_legsFitter_r {parent=thigh_r, loc=<41.5699 0.0432 4.3711>, rot=<0.7884 -0.5405 -0.2897 0.0491>}
 base_legsFitter_A_r {parent=kin_legsFitter_r, loc=<0.0739 0.408 -0.4015>, rot=<-0.4779 0.5204 0.5655 0.4254>}
 dyn_legsFitter_A_r {parent=base_legsFitter_A_r, loc=<0 0 0>, rot=<0 0 0 -1>}
 base_legsFitter_B_r {parent=kin_legsFitter_r, loc=<1.6055 -0.2197 -0.474>, rot=<0.612 -0.7019 -0.039 -0.3623>}
 dyn_legsFitter_B_r {parent=base_legsFitter_B_r, loc=<0 0 0>, rot=<0 0 0 1>}
 deform_glute_base_l {parent=pelvis, loc=<-10.1877 -11.6217 -12.899>, rot=<0.0275 0.0501 -0.4472 -0.8926>}
 deform_glute_l {parent=deform_glute_base_l, loc=<0 0 0>, rot=<0 0 0 1>}
 deform_glute_base_r {parent=pelvis, loc=<-10.1877 -11.6218 12.899>, rot=<0.0501 -0.0275 0.8926 -0.4472>}
 deform_glute_r {parent=deform_glute_base_r, loc=<0 0 0>, rot=<0 0 0 1>}
 thigh_l_null {parent=pelvis, type=Null, loc=<-1.8716 -0.0109 -11.9936>, rot=<0.0012 0.0572 0.0155 -0.9982>}
 thigh_r_null {parent=pelvis, type=Null, loc=<-1.8716 -0.0109 11.9936>, rot=<0.0572 -0.0012 0.9982 0.0155>}
 base_bandoliersTop_l {parent=pelvis, loc=<5.6084 -0.5704 -8.6756>, rot=<-0.0167 -0.934 0.0064 -0.3567>}
 dyn_bandoliersTop_l {parent=base_bandoliersTop_l, loc=<0 0 0>, rot=<0 0 0 1>}
 kin_bandoliersTopBack_l {parent=dyn_bandoliersTop_l, loc=<15.37 -6.9993 -0.7235>, rot=<0 0.1478 0 -0.989>}
 kin_bandoliersTopFront_l {parent=dyn_bandoliersTop_l, loc=<15.3045 5.6127 -1.1207>, rot=<0 0.0453 0 -0.999>}
 base_bandoliersTop_r {parent=pelvis, loc=<5.6084 -0.5704 8.6756>, rot=<0.3567 -0.0064 0.934 0.0167>}
 dyn_bandoliersTop_r {parent=base_bandoliersTop_r, loc=<0 0 0>, rot=<0 0 0 1>}
 kin_bandoliersTopBack_r {parent=dyn_bandoliersTop_r, loc=<15.37 6.9993 -0.7236>, rot=<0 0.1478 0 -0.989>}
 kin_bandoliersTopFront_r {parent=dyn_bandoliersTop_r, loc=<15.3045 -5.6127 -1.1208>, rot=<0 0.1478 0 -0.989>}
 base_bandoliersBotMid_1_l {parent=pelvis, loc=<-1.8716 -0.0109 -11.9936>, rot=<0.2206 -0.7441 -0.5978 -0.2008>}
 dyn_bandoliersBotMid_1_l {parent=base_bandoliersBotMid_1_l, loc=<0 0 0>, rot=<0 0 0 1>}
 base_bandoliersBotBack_1_l {parent=pelvis, loc=<3.6253 -9.8277 -4.4982>, rot=<0.1273 -0.9748 -0.1399 -0.1182>}
 dyn_bandoliersBotBack_1_l {parent=base_bandoliersBotBack_1_l, loc=<0 0 0>, rot=<0 0 0 1>}
 dyn_bandoliersBotBack_2_l {parent=dyn_bandoliersBotBack_1_l, loc=<6.4843 0 0>, rot=<-0.0425 0.1975 -0.1028 -0.974>}
 dyn_bandoliersBotBack_3_l {parent=dyn_bandoliersBotBack_2_l, loc=<6.5443 0 0>, rot=<-0.0732 0.1123 -0.0426 -0.9901>}
 dyn_bandoliersBotBack_4_l {parent=dyn_bandoliersBotBack_3_l, loc=<6.5722 0 0>, rot=<0.0088 0.0678 -0.0823 -0.9943>}
 dyn_bandoliersBotBack_5_l {parent=dyn_bandoliersBotBack_4_l, loc=<6.5508 0 0>, rot=<0.0133 0.038 -0.2341 -0.9714>}
 base_bandoliersBotFront_1_l {parent=pelvis, loc=<4.0468 6.8646 -2.7302>, rot=<-0.1504 0.0781 0.9829 0.0724>}
 dyn_bandoliersBotFront_1_l {parent=base_bandoliersBotFront_1_l, loc=<0 0 0>, rot=<0 0 0 1>}
 dyn_bandoliersBotFront_2_l {parent=dyn_bandoliersBotFront_1_l, loc=<7.3273 0 0>, rot=<0 -0.1524 0 -0.9883>}
 dyn_bandoliersBotFront_3_l {parent=dyn_bandoliersBotFront_2_l, loc=<7.2909 0 0>, rot=<-0.049 -0.1648 -0.0082 -0.9851>}
 dyn_bandoliersBotFront_4_l {parent=dyn_bandoliersBotFront_3_l, loc=<7.3143 0 0>, rot=<-0.0057 -0.0796 -0.1734 -0.9816>}
 dyn_bandoliersBotFront_5_l {parent=dyn_bandoliersBotFront_4_l, loc=<7.1824 0 0>, rot=<-0.0618 -0.162 -0.3293 -0.9282>}
 base_bandoliersBotMid_1_r {parent=pelvis, loc=<-1.8716 -0.0109 11.9936>, rot=<0.2206 -0.7441 0.5978 0.2008>}
 dyn_bandoliersBotMid_1_r {parent=base_bandoliersBotMid_1_r, loc=<0 0 0>, rot=<0 0 0 1>}
 base_bandoliersBotBack_1_r {parent=pelvis, loc=<3.6253 -9.8277 4.4982>, rot=<0.1273 -0.9748 0.1399 0.1182>}
 dyn_bandoliersBotBack_1_r {parent=base_bandoliersBotBack_1_r, loc=<0 0 0>, rot=<0 0 0 1>}
 dyn_bandoliersBotBack_2_r {parent=dyn_bandoliersBotBack_1_r, loc=<6.4842 0 0>, rot=<0.0425 -0.1975 -0.1028 -0.974>}
 dyn_bandoliersBotBack_3_r {parent=dyn_bandoliersBotBack_2_r, loc=<6.5443 0 0>, rot=<0.0732 -0.1123 -0.0426 -0.9901>}
 dyn_bandoliersBotBack_4_r {parent=dyn_bandoliersBotBack_3_r, loc=<6.5722 0 0.0001>, rot=<-0.0088 -0.0678 -0.0823 -0.9943>}
 dyn_bandoliersBotBack_5_r {parent=dyn_bandoliersBotBack_4_r, loc=<6.5508 0 -0.0001>, rot=<-0.0133 -0.038 -0.2341 -0.9714>}
 base_bandoliersBotFront_1_r {parent=pelvis, loc=<4.0469 6.8646 2.7302>, rot=<0.1504 -0.0781 0.9829 0.0724>}
 dyn_bandoliersBotFront_1_r {parent=base_bandoliersBotFront_1_r, loc=<0 0 0>, rot=<0 0 0 1>}
 dyn_bandoliersBotFront_2_r {parent=dyn_bandoliersBotFront_1_r, loc=<7.3274 0 0>, rot=<0 0.1524 0 -0.9883>}
 dyn_bandoliersBotFront_3_r {parent=dyn_bandoliersBotFront_2_r, loc=<7.2908 0 0>, rot=<0.049 0.1648 -0.0082 -0.9851>}
 dyn_bandoliersBotFront_4_r {parent=dyn_bandoliersBotFront_3_r, loc=<7.3143 0 0>, rot=<0.0057 0.0796 -0.1734 -0.9816>}
 dyn_bandoliersBotFront_5_r {parent=dyn_bandoliersBotFront_4_r, loc=<7.1823 0 0>, rot=<0.0618 0.162 -0.3293 -0.9282>}
 coll_bandoliersTop_1_l {parent=pelvis, loc=<-4.0857 6.832 -14.9159>, rot=<0.4299 -0.6797 -0.5485 -0.229>}
 coll_bandoliersTop_2_l {parent=pelvis, loc=<-7.3491 4.7633 -19.7775>, rot=<0.6153 -0.7142 -0.3272 0.0661>}
 coll_bandoliersTop_3_l {parent=pelvis, loc=<-8.8442 -1.1241 -21.7456>, rot=<0.7019 -0.5904 -0.0162 0.398>}
 coll_bandoliersTop_4_l {parent=pelvis, loc=<-6.9218 -6.4779 -18.9006>, rot=<0.7019 -0.5904 -0.0162 0.398>}
 arm_bend_1_l_ctrl_null {type=Null, loc=<22.5467 -4.8444 132.4895>, rot=<0.654 -0.2517 -0.2025 -0.684>}
 arm_bend_1_l_ctrl {controlType=EULER_TRANSFORM, parent=arm_bend_1_l_ctrl_null, type=Control, loc=<0 0 0>, rot=<0 0 0 1>}
 arm_bend_2_l_ctrl_null {type=Null, loc=<34.0331 -5.6029 123.6447>, rot=<0.654 -0.2517 -0.2025 -0.684>}
 arm_bend_2_l_ctrl {controlType=EULER_TRANSFORM, parent=arm_bend_2_l_ctrl_null, type=Control, loc=<0 0 0>, rot=<0 0 0 1>}
 arm_bend_3_l_ctrl_null {type=Null, loc=<45.5195 -6.3613 114.7999>, rot=<0.6236 -0.1642 -0.2828 -0.7101>}
 arm_bend_3_l_ctrl {controlType=EULER_TRANSFORM, parent=arm_bend_3_l_ctrl_null, type=Control, loc=<0 0 0>, rot=<0 0 0 1>}
 arm_bend_4_l_ctrl_null {type=Null, loc=<54.9907 -0.7488 107.998>, rot=<0.5833 -0.074 -0.3586 -0.7251>}
 arm_bend_4_l_ctrl {controlType=EULER_TRANSFORM, parent=arm_bend_4_l_ctrl_null, type=Control, loc=<0 0 0>, rot=<0 0 0 1>}
 arm_bend_5_l_ctrl_null {type=Null, loc=<64.462 4.8638 101.1961>, rot=<0.5833 -0.074 -0.3586 -0.7251>}
 arm_bend_5_l_ctrl {controlType=EULER_TRANSFORM, parent=arm_bend_5_l_ctrl_null, type=Control, loc=<0 0 0>, rot=<0 0 0 1>}
 arm_bend_1_r_ctrl_null {type=Null, loc=<-22.5467 -4.8444 132.49>, rot=<-0.684 0.2025 -0.2517 -0.654>}
 arm_bend_1_r_ctrl {controlType=EULER_TRANSFORM, parent=arm_bend_1_r_ctrl_null, type=Control, loc=<0 0 0>, rot=<0 0 0 1>}
 arm_bend_2_r_ctrl_null {type=Null, loc=<-34.0331 -5.6029 123.645>, rot=<-0.684 0.2025 -0.2517 -0.654>}
 arm_bend_2_r_ctrl {controlType=EULER_TRANSFORM, parent=arm_bend_2_r_ctrl_null, type=Control, loc=<0 0 0>, rot=<0 0 0 1>}
 arm_bend_3_r_ctrl_null {type=Null, loc=<-45.5195 -6.3614 114.8>, rot=<-0.7101 0.2828 -0.1642 -0.6236>}
 arm_bend_3_r_ctrl {controlType=EULER_TRANSFORM, parent=arm_bend_3_r_ctrl_null, type=Control, loc=<0 0 0>, rot=<0 0 0 1>}
 arm_bend_4_r_ctrl_null {type=Null, loc=<-54.9907 -0.7488 107.998>, rot=<-0.7251 0.3586 -0.074 -0.5833>}
 arm_bend_4_r_ctrl {controlType=EULER_TRANSFORM, parent=arm_bend_4_r_ctrl_null, type=Control, loc=<0 0 0>, rot=<0 0 0 1>}
 arm_bend_5_r_ctrl_null {type=Null, loc=<-64.462 4.8638 101.196>, rot=<-0.7251 0.3586 -0.074 -0.5833>}
 arm_bend_5_r_ctrl {controlType=EULER_TRANSFORM, parent=arm_bend_5_r_ctrl_null, type=Control, loc=<0 0 0>, rot=<0 0 0 1>}
 arm_l_pv_ctrl {controlType=EULER_TRANSFORM, type=Control, loc=<0 0 0>, rot=<0 0 0 1>}
 arm_r_pv_ctrl {controlType=EULER_TRANSFORM, type=Control, loc=<0 0 0>, rot=<0 0 0 1>}
 dyn_shoulderStraps_1_r_ctrl_null {type=Null, loc=<-7.4584 -13.6854 141.632>, rot=<0.4199 0.4282 -0.5772 0.5543>}
 dyn_shoulderStraps_1_r_ctrl {controlType=EULER_TRANSFORM, parent=dyn_shoulderStraps_1_r_ctrl_null, type=Control, loc=<0 0 0>, rot=<0 0 0 1>}
 dyn_shoulderStraps_2_r_ctrl_null {type=Null, loc=<-7.702 -15.756 134.544>, rot=<0.4627 0.4692 -0.5435 0.52>}
 dyn_shoulderStraps_2_r_ctrl {
  controlType=EULER_TRANSFORM
  parent=dyn_shoulderStraps_2_r_ctrl_null
  type=Control
  loc=<-35.3819 -15.756 130.6827>
  rot=<0.5028 0.4282 -0.5066 0.5543>
 }
 dyn_shoulderStraps_3_r_ctrl_null {type=Null, loc=<-7.9311 -16.7229 127.231>, rot=<-0.4952 -0.511 0.51 -0.4833>}
 dyn_shoulderStraps_3_r_ctrl {
  controlType=EULER_TRANSFORM
  parent=dyn_shoulderStraps_3_r_ctrl_null
  type=Control
  loc=<-33.2144 -18.1777 123.5645>
  rot=<0.5295 0.4804 -0.4674 0.52>
 }
 dyn_shoulderStraps_4_r_ctrl_null {type=Null, loc=<-8.2448 -16.6257 119.849>, rot=<0.498 0.5096 -0.5068 0.4852>}
 dyn_shoulderStraps_4_r_ctrl {
  controlType=EULER_TRANSFORM
  parent=dyn_shoulderStraps_4_r_ctrl_null
  type=Control
  loc=<-16.1147 -16.6069 119.7879>
  rot=<-0.5014 -0.5069 0.5081 -0.4833>
 }
 dyn_shoulderStraps_5_r_ctrl_null {type=Null, loc=<-8.4888 -16.5092 112.467>, rot=<-0.4951 -0.4998 0.5194 -0.4851>}
 dyn_shoulderStraps_5_r_ctrl {
  controlType=EULER_TRANSFORM
  parent=dyn_shoulderStraps_5_r_ctrl_null
  type=Control
  loc=<-12.9856 -14.3222 113.1356>
  rot=<0.4824 0.5019 -0.5291 0.4852>
 }
 dyn_shoulderStraps_6_r_ctrl_null {type=Null, loc=<-8.7778 -16.5756 105.073>, rot=<-0.5109 -0.5364 0.4908 -0.4587>}
 dyn_shoulderStraps_6_r_ctrl {
  controlType=EULER_TRANSFORM
  parent=dyn_shoulderStraps_6_r_ctrl_null
  type=Control
  loc=<-26.8797 -19.6274 102.3184>
  rot=<0.5366 0.5225 -0.4514 0.4851>
 }
 dyn_shoulderStraps_7_r_ctrl_null {type=Null, loc=<-9.1982 -15.8563 97.7658>, rot=<0.5109 0.5364 -0.4908 0.4587>}
 dyn_shoulderStraps_7_r_ctrl {
  controlType=EULER_TRANSFORM
  parent=dyn_shoulderStraps_7_r_ctrl_null
  type=Control
  loc=<-16.5528 -15.8563 97.7657>
  rot=<-0.5109 -0.5364 0.4908 -0.4587>
 }
 dyn_shoulderStraps_1_l_ctrl_null {type=Null, loc=<7.4584 -13.6854 141.632>, rot=<-0.4282 -0.4199 0.5543 -0.5772>}
 dyn_shoulderStraps_1_l_ctrl {controlType=EULER_TRANSFORM, parent=dyn_shoulderStraps_1_l_ctrl_null, type=Control, loc=<0 0 0>, rot=<0 0 0 1>}
 dyn_shoulderStraps_2_l_ctrl_null {type=Null, loc=<7.702 -15.756 134.5442>, rot=<-0.4692 -0.4627 0.52 -0.5435>}
 dyn_shoulderStraps_2_l_ctrl {
  controlType=EULER_TRANSFORM
  parent=dyn_shoulderStraps_2_l_ctrl_null
  type=Control
  loc=<-20.157 -15.756 133.0261>
  rot=<0.5075 0.4199 -0.4827 0.5772>
 }
 dyn_shoulderStraps_3_l_ctrl_null {type=Null, loc=<7.9311 -16.7229 127.231>, rot=<-0.511 -0.4952 0.4833 -0.51>}
 dyn_shoulderStraps_3_l_ctrl {
  controlType=EULER_TRANSFORM
  parent=dyn_shoulderStraps_3_l_ctrl_null
  type=Control
  loc=<-17.9834 -15.4636 126.1156>
  rot=<0.5455 0.4516 -0.4507 0.5435>
 }
 dyn_shoulderStraps_4_l_ctrl_null {type=Null, loc=<8.2448 -16.6257 119.8492>, rot=<-0.5096 -0.498 0.4852 -0.5068>}
 dyn_shoulderStraps_4_l_ctrl {
  controlType=EULER_TRANSFORM
  parent=dyn_shoulderStraps_4_l_ctrl_null
  type=Control
  loc=<0.6875 -16.4877 119.8694>
  rot=<0.5076 0.4995 -0.4824 0.51>
 }
 dyn_shoulderStraps_5_l_ctrl_null {type=Null, loc=<8.4888 -16.5092 112.4666>, rot=<-0.4998 -0.4951 0.4851 -0.5194>}
 dyn_shoulderStraps_5_l_ctrl {
  controlType=EULER_TRANSFORM
  parent=dyn_shoulderStraps_5_l_ctrl_null
  type=Control
  loc=<3.7945 -18.7789 112.0511>
  rot=<0.5003 0.505 -0.4877 0.5068>
 }
 dyn_shoulderStraps_6_l_ctrl_null {type=Null, loc=<8.7778 -16.5756 105.0732>, rot=<-0.5364 -0.5109 0.4587 -0.4908>}
 dyn_shoulderStraps_6_l_ctrl {
  controlType=EULER_TRANSFORM
  parent=dyn_shoulderStraps_6_l_ctrl_null
  type=Control
  loc=<-10.1393 -13.8368 104.9893>
  rot=<0.559 0.4728 -0.4408 0.5194>
 }
 dyn_shoulderStraps_7_l_ctrl_null {type=Null, loc=<9.1982 -15.8563 97.7658>, rot=<-0.5364 -0.5109 0.4587 -0.4908>}
 dyn_shoulderStraps_7_l_ctrl {
  controlType=EULER_TRANSFORM
  parent=dyn_shoulderStraps_7_l_ctrl_null
  type=Control
  loc=<1.8435 -15.8563 97.7658>
  rot=<0.5364 0.5109 -0.4587 0.4908>
 }
 dyn_bandoliersBotBack_5_l_ctrl_null {type=Null, loc=<19.4692 -8.003 76.8523>, rot=<-0.086 -0.2501 -0.4019 -0.8767>}
 dyn_bandoliersBotBack_5_l_ctrl {
  controlType=EULER_TRANSFORM
  parent=dyn_bandoliersBotBack_5_l_ctrl_null
  type=Control
  loc=<13.0065 -16.3438 75.5044>
  rot=<0.1691 0.2507 0.1852 0.935>
 }
 dyn_bandoliersBotBack_4_l_ctrl_null {type=Null, loc=<14.5599 -10.3546 80.4966>, rot=<-0.0215 -0.3017 -0.185 -0.935>}
 dyn_bandoliersBotBack_4_l_ctrl {
  controlType=EULER_TRANSFORM
  parent=dyn_bandoliersBotBack_4_l_ctrl_null
  type=Control
  loc=<16.834 -13.8179 78.5901>
  rot=<0.0669 0.36 0.1083 0.9242>
 }
 dyn_bandoliersBotBack_3_l_ctrl_null {type=Null, loc=<9.9034 -11.6033 84.9634>, rot=<0.0078 -0.3668 -0.1059 -0.9242>}
 dyn_bandoliersBotBack_3_l_ctrl {
  controlType=EULER_TRANSFORM
  parent=dyn_bandoliersBotBack_3_l_ctrl_null
  type=Control
  loc=<21.8919 -0.2652 82.9781>
  rot=<-0.0479 0.475 0.0395 0.8778>
 }
 dyn_bandoliersBotBack_2_l_ctrl_null {type=Null, loc=<6.2239 -12.0362 90.358>, rot=<0.1029 -0.4588 -0.0915 -0.8778>}
 dyn_bandoliersBotBack_2_l_ctrl {
  controlType=EULER_TRANSFORM
  parent=dyn_bandoliersBotBack_2_l_ctrl_null
  type=Control
  loc=<33.1043 -7.8651 84.568>
  rot=<-0.0723 0.6347 -0.0003 0.7694>
 }
 dyn_bandoliersBotFront_5_r_ctrl_null {type=Null, loc=<-20.4949 3.6047 75.4355>, rot=<-0.0838 0.0013 0.8228 -0.5621>}
 dyn_bandoliersBotFront_5_r_ctrl {
  controlType=EULER_TRANSFORM
  parent=dyn_bandoliersBotFront_5_r_ctrl_null
  type=Control
  loc=<1.6589 2.4534 80.3789>
  rot=<-0.0211 0.1131 -0.9625 0.2458>
 }
 dyn_bandoliersBotFront_4_r_ctrl_null {type=Null, loc=<-15.0518 6.6717 78.9784>, rot=<-0.2463 -0.0666 0.9351 -0.2458>}
 dyn_bandoliersBotFront_4_r_ctrl {
  controlType=EULER_TRANSFORM
  parent=dyn_bandoliersBotFront_4_r_ctrl_null
  type=Control
  loc=<-6.2672 10.7996 81.4029>
  rot=<0.1802 0.0475 -0.9798 0.0725>
 }
 dyn_bandoliersBotFront_2_r_ctrl_null {type=Null, loc=<-4.9799 6.8947 89.2745>, rot=<0.4592 0.0887 -0.8834 0.0285>}
 dyn_bandoliersBotFront_2_r_ctrl {
  controlType=EULER_TRANSFORM
  parent=dyn_bandoliersBotFront_2_r_ctrl_null
  type=Control
  loc=<15.1598 6.8947 88.8346>
  rot=<-0.3192 -0.0921 0.9431 -0.0146>
 }
 dyn_bandoliersBotFront_3_r_ctrl_null {type=Null, loc=<-9.1846 7.1219 83.3226>, rot=<-0.3061 -0.1223 0.9413 -0.0725>}
 dyn_bandoliersBotFront_3_r_ctrl {
  controlType=EULER_TRANSFORM
  parent=dyn_bandoliersBotFront_3_r_ctrl_null
  type=Control
  loc=<11.6365 -1.1884 84.4275>
  rot=<0.1509 0.176 -0.9723 0.0285>
 }
 dyn_bandoliersBotFront_1_r_ctrl_null {type=Null, loc=<-2.7302 6.2731 96.2202>, rot=<0.5884 0.0921 -0.8031 0.0146>}
 dyn_bandoliersBotFront_1_r_ctrl {controlType=EULER_TRANSFORM, parent=dyn_bandoliersBotFront_1_r_ctrl_null, type=Control, loc=<0 0 0>, rot=<0 0 0 1>}
 dyn_bandoliersBotBack_5_r_ctrl_null {type=Null, loc=<-19.4692 -8.0029 76.8523>, rot=<-0.4019 -0.8767 -0.086 -0.2501>}
 dyn_bandoliersBotBack_5_r_ctrl {
  controlType=EULER_TRANSFORM
  parent=dyn_bandoliersBotBack_5_r_ctrl_null
  type=Control
  loc=<-31.9302 8.0438 74.5864>
  rot=<0.589 0.7491 0.0286 0.3017>
 }
 dyn_bandoliersBotBack_4_r_ctrl_null {type=Null, loc=<-14.5599 -10.3546 80.4966>, rot=<-0.185 -0.935 -0.0215 -0.3017>}
 dyn_bandoliersBotBack_4_r_ctrl {
  controlType=EULER_TRANSFORM
  parent=dyn_bandoliersBotBack_4_r_ctrl_null
  type=Control
  loc=<-33.0918 -4.4737 76.9311>
  rot=<0.2568 0.8942 0.0008 0.3668>
 }
 dyn_bandoliersBotBack_3_r_ctrl_null {type=Null, loc=<-9.9034 -11.6033 84.9634>, rot=<-0.1059 -0.9242 0.0078 -0.3668>}
 dyn_bandoliersBotBack_3_r_ctrl {
  controlType=EULER_TRANSFORM
  parent=dyn_bandoliersBotBack_3_r_ctrl_null
  type=Control
  loc=<-36.178 -21.2896 76.5627>
  rot=<0.1719 0.8699 0.0563 0.4588>
 }
 dyn_bandoliersBotBack_2_r_ctrl_null {type=Null, loc=<-6.2239 -12.0362 90.358>, rot=<-0.0915 -0.8778 0.1029 -0.4588>}
 dyn_bandoliersBotBack_2_r_ctrl {
  controlType=EULER_TRANSFORM
  parent=dyn_bandoliersBotBack_2_r_ctrl_null
  type=Control
  loc=<-49.2033 -12.7968 76.7217>
  rot=<0.2191 0.7593 -0.092 0.6057>
 }
 dyn_bandoliersBotBack_1_r_ctrl_null {type=Null, loc=<-4.4982 -10.4237 96.3968>, rot=<-0.002 0.7694 -0.2028 0.6057>}
 dyn_bandoliersBotBack_1_r_ctrl {controlType=EULER_TRANSFORM, parent=dyn_bandoliersBotBack_1_r_ctrl_null, type=Control, loc=<0 0 0>, rot=<0 0 0 1>}
 dyn_bandoliersBotBack_1_l_ctrl_null {type=Null, loc=<4.4982 -10.4237 96.3968>, rot=<-0.2028 0.6057 -0.002 0.7694>}
 dyn_bandoliersBotBack_1_l_ctrl {controlType=EULER_TRANSFORM, parent=dyn_bandoliersBotBack_1_l_ctrl_null, type=Control, loc=<0 0 0>, rot=<0 0 0 1>}
 dyn_bandoliersBotFront_5_l_ctrl_null {type=Null, loc=<20.4949 3.6047 75.4355>, rot=<0.8228 -0.5621 -0.0838 0.0013>}
 dyn_bandoliersBotFront_5_l_ctrl {
  controlType=EULER_TRANSFORM
  parent=dyn_bandoliersBotFront_5_l_ctrl_null
  type=Control
  loc=<-7.6033 11.6305 75.4165>
  rot=<-0.5921 0.7981 -0.0898 0.0666>
 }
 dyn_bandoliersBotFront_4_l_ctrl_null {type=Null, loc=<15.0518 6.6717 78.9784>, rot=<0.9351 -0.2458 -0.2463 -0.0666>}
 dyn_bandoliersBotFront_4_l_ctrl {
  controlType=EULER_TRANSFORM
  parent=dyn_bandoliersBotFront_4_l_ctrl_null
  type=Control
  loc=<-2.7412 6.7073 79.3062>
  rot=<-0.8953 0.3996 0.1543 0.1223>
 }
 dyn_bandoliersBotFront_3_l_ctrl_null {type=Null, loc=<9.1846 7.1219 83.3226>, rot=<0.9413 -0.0725 -0.3061 -0.1223>}
 dyn_bandoliersBotFront_3_l_ctrl {
  controlType=EULER_TRANSFORM
  parent=dyn_bandoliersBotFront_3_l_ctrl_null
  type=Control
  loc=<-24.9639 15.3617 78.344>
  rot=<-0.9831 0.074 0.1418 0.0887>
 }
 dyn_bandoliersBotFront_2_l_ctrl_null {type=Null, loc=<4.9799 6.8947 89.2745>, rot=<-0.8834 0.0285 0.4592 0.0887>}
 dyn_bandoliersBotFront_2_l_ctrl {
  controlType=EULER_TRANSFORM
  parent=dyn_bandoliersBotFront_2_l_ctrl_null
  type=Control
  loc=<-29.1336 6.8947 84.4197>
  rot=<-0.9431 0.0146 0.3192 0.0921>
 }
 dyn_bandoliersBotFront_1_l_ctrl_null {type=Null, loc=<2.7302 6.2731 96.2202>, rot=<-0.8031 0.0146 0.5884 0.0921>}
 dyn_bandoliersBotFront_1_l_ctrl {controlType=EULER_TRANSFORM, parent=dyn_bandoliersBotFront_1_l_ctrl_null, type=Control, loc=<0 0 0>, rot=<0 0 0 1>}
]
variables [
 calfBend_in_l {type=double, default=0}
 calfBend_in_r {type=double, default=0}
 pub isFrontEnd {type=bool, default=False}
 pub isTargeting {type=double, default=0}
 pub isWearingBackBling {type=bool, default=False}
 pub isWearingGauntlets {type=bool, default=False}
 pub isWearingShoes {type=bool, default=False}
 rUpperarmDown {type=double, default=0}
 thighBend_out_l {type=double, default=0}
 thighBend_out_n40_n100 {type=double, default=0}
 thighBend_out_r {type=double, default=0}
 twistSpine_0_50 {type=double, default=0}
 twistSpine_20_50 {type=double, default=0}
]
