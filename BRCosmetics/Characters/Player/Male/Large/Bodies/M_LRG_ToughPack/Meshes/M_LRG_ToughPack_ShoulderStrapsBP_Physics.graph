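# Physics asset for the shoulder straps on the ToughPack large male player body, defining collision bodies and joints for the dangling strap ends.
# Capsule bodies cover the pelvis and mirrored left and right chains of three dynamic strap bones: the seventh strap bone and two tip bones.
# Each chain hangs off the pelvis body through angular-limited constraints, with tight swing and twist ranges and slightly narrower twist at the tips.
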PhysicsAsset /BRCosmetics/Characters/Player/Male/Large/Bodies/M_LRG_ToughPack/Meshes/M_LRG_ToughPack_ShoulderStrapsBP_Physics
bodies {
 dyn_shoulderStrapsTip_1_l [{
   shapeType=Capsule
   length=3.3473
   radius=2.5561
   shapeName=dyn_shoulderStrapsTip_1_l_capsule
   center=<2.9052 0 1.5542>
   rotation=<rot 0 -90 -90>
  }]
 dyn_shoulderStrapsTip_1_r [{
   shapeType=Capsule
   length=3.3473
   radius=2.5561
   shapeName=dyn_shoulderStrapsTip_1_l_capsule
   center=<2.9052 0 1.5542>
   rotation=<rot 0 -90 -90>
  }]
 dyn_shoulderStrapsTip_2_l [{
   shapeType=Capsule
   length=2.1562
   radius=2.8897
   shapeName=dyn_shoulderStrapsTip_2_l_capsule
   center=<4.5764 0 1.3885>
   rotation=<rot 0 -90 -90>
  }]
 dyn_shoulderStrapsTip_2_r [{
   shapeType=Capsule
   length=2.1562
   radius=2.8897
   shapeName=dyn_shoulderStrapsTip_2_l_capsule
   center=<4.5764 0 1.3885>
   rotation=<rot 0 -90 -90>
  }]
 dyn_shoulderStraps_7_l [{shapeType=Capsule, length=5, radius=1, shapeName=dyn_shoulderStraps_7_l_capsule, center=<0 0 0.7494>, rotation=<rot 0 0 90>}]
 dyn_shoulderStraps_7_r [{shapeType=Capsule, length=5, radius=1, shapeName=dyn_shoulderStraps_7_r_capsule, center=<0 0 0.7494>, rotation=<rot 0 0 -90>}]
 pelvis [{shapeType=Capsule, length=19.6264, radius=13.5312, shapeName=pelvis_capsule, center=<-1.9551 0.9719 0>, rotation=<rot 0 -90 0>}]
}
constraints [
 {bone1Name=dyn_shoulderStraps_7_l, bone2Name=pelvis, swing1LimitDegrees=15, swing2LimitDegrees=20, twistLimitDegrees=15}
 {bone1Name=dyn_shoulderStraps_7_r, bone2Name=pelvis, swing1LimitDegrees=15, swing2LimitDegrees=20, twistLimitDegrees=15}
 {bone1Name=dyn_shoulderStrapsTip_1_l, bone2Name=dyn_shoulderStraps_7_l, swing1LimitDegrees=15, swing2LimitDegrees=20, twistLimitDegrees=10}
 {bone1Name=dyn_shoulderStrapsTip_2_l, bone2Name=dyn_shoulderStrapsTip_1_l, swing1LimitDegrees=15, swing2LimitDegrees=20, twistLimitDegrees=10}
 {bone1Name=dyn_shoulderStrapsTip_2_r, bone2Name=dyn_shoulderStrapsTip_1_r, swing1LimitDegrees=15, swing2LimitDegrees=20, twistLimitDegrees=10}
 {bone1Name=dyn_shoulderStrapsTip_1_r, bone2Name=dyn_shoulderStraps_7_r, swing1LimitDegrees=15, swing2LimitDegrees=20, twistLimitDegrees=10}
]
