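# Physics asset for the left and right shoulder straps of the large male ToughPack player body, letting them move over the torso.
# Each side is a capsule chain from a base at spine_05 through seven segments, with swing and twist limits that generally widen down the strap.
# The seventh segment is rotation-locked to spine_01, pinning the strap at both ends, and two limited tip segments continue past it.
# Larger pelvis and spine capsules, linked by free-rotation constraints, form the torso.
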PhysicsAsset /BRCosmetics/Characters/Player/Male/Large/Bodies/M_LRG_ToughPack/Meshes/M_LRG_ToughPack_ShoulderStraps_Physics
bodies {
 base_shoulderStraps_1_l [{shapeType=Capsule, length=1, radius=0.5, shapeName=base_shoulderStraps_1_l_capsule}]
 base_shoulderStraps_1_r [{shapeType=Capsule, length=1, radius=0.5, shapeName=base_shoulderStraps_1_r_capsule}]
 dyn_shoulderStrapsTip_1_l [{
   shapeType=Capsule
   length=3.3473
   radius=2.5561
   shapeName=dyn_shoulderStrapsTip_1_l_capsule
   center=<2.9052 0 1.5542>
   rotation=<rot 0 -90 -90>
  }]
 dyn_shoulderStrapsTip_1_r [{
   shapeType=Capsule
   length=3.3473
   radius=2.5561
   shapeName=dyn_shoulderStrapsTip_1_l_capsule
   center=<2.9052 0 1.5542>
   rotation=<rot 0 -90 -90>
  }]
 dyn_shoulderStrapsTip_2_l [{
   shapeType=Capsule
   length=2.1562
   radius=2.8897
   shapeName=dyn_shoulderStrapsTip_2_l_capsule
   center=<4.5764 0 1.3885>
   rotation=<rot 0 -90 -90>
  }]
 dyn_shoulderStrapsTip_2_r [{
   shapeType=Capsule
   length=2.1562
   radius=2.8897
   shapeName=dyn_shoulderStrapsTip_2_l_capsule
   center=<4.5764 0 1.3885>
   rotation=<rot 0 -90 -90>
  }]
 dyn_shoulderStraps_1_l [{shapeType=Capsule, length=3.9288, radius=2.8862, center=<3.4883 0 2.416>, rotation=<rot 0 0 90>}]
 dyn_shoulderStraps_1_r [{shapeType=Capsule, length=3.9288, radius=2.8862, center=<3.4883 0 2.416>, rotation=<rot 0 0 90>}]
 dyn_shoulderStraps_2_l [{shapeType=Capsule, length=4.9679, radius=2.8862, center=<3.4883 0 1.5993>, rotation=<rot 0 0 90>}]
 dyn_shoulderStraps_2_r [{shapeType=Capsule, length=4.9679, radius=2.8862, center=<3.4883 0 1.5993>, rotation=<rot 0 0 90>}]
 dyn_shoulderStraps_3_l [{shapeType=Capsule, length=4.4961, radius=2.8862, center=<3.4883 0 1.1658>, rotation=<rot 0 0 90>}]
 dyn_shoulderStraps_3_r [{shapeType=Capsule, length=4.4961, radius=2.8862, center=<3.4883 0 1.1658>, rotation=<rot 0 0 90>}]
 dyn_shoulderStraps_4_l [{shapeType=Capsule, length=5.6992, radius=2.8862, center=<3.4883 0 1.5139>, rotation=<rot 0 0 90>}]
 dyn_shoulderStraps_4_r [{shapeType=Capsule, length=5.6992, radius=2.8862, center=<3.4883 0 1.5139>, rotation=<rot 0 0 90>}]
 dyn_shoulderStraps_5_l [{shapeType=Capsule, length=5.2657, radius=2.8862, center=<3.4883 0 1.4979>, rotation=<rot 0 0 90>}]
 dyn_shoulderStraps_5_r [{shapeType=Capsule, length=5.2657, radius=2.8862, center=<3.4883 0 1.4979>, rotation=<rot 0 0 90>}]
 dyn_shoulderStraps_6_l [{shapeType=Capsule, length=3.9288, radius=2.8862, center=<2.5955 0 2.3061>, rotation=<rot 0 0 90>}]
 dyn_shoulderStraps_6_r [{shapeType=Capsule, length=3.9288, radius=2.8862, center=<2.5955 0 2.3061>, rotation=<rot 0 0 90>}]
 dyn_shoulderStraps_7_l [{shapeType=Capsule, length=5, radius=1, shapeName=dyn_shoulderStraps_7_l_capsule, center=<0 0 0.7494>, rotation=<rot 0 0 90>}]
 dyn_shoulderStraps_7_r [{shapeType=Capsule, length=5, radius=1, shapeName=dyn_shoulderStraps_7_r_capsule, center=<0 0 0.7494>, rotation=<rot 0 0 -90>}]
 pelvis [{shapeType=Capsule, length=19.6264, radius=13.5312, shapeName=pelvis_capsule, center=<-1.9551 0.9719 0>, rotation=<rot 0 -90 0>}]
 spine_01 [{shapeType=Capsule, length=19.9126, radius=13.224, shapeName=spine_01_capsule, center=<5.0471 -0.3163 0>, rotation=<rot 0 -90 0>}]
 spine_02 [{shapeType=Capsule, length=20.4318, radius=15.8739, shapeName=spine_02_capsule, center=<3.0044 1.8914 0>, rotation=<rot 0 -90 0>}]
 spine_03 [{shapeType=Capsule, length=20.1085, radius=15.0808, shapeName=spine_03_capsule, center=<2.505 1.22 0>, rotation=<rot 0 -90 0>}]
 spine_04 [{shapeType=Capsule, length=20.1085, radius=15.0808, shapeName=spine_03_capsule, center=<2.505 2.3207 0>, rotation=<rot 0 -90 0>}]
 spine_05 [{shapeType=Capsule, length=20.1085, radius=15.0808, shapeName=spine_03_capsule, center=<2.505 3.7956 0>, rotation=<rot 0 -90 0>}]
}
constraints [
 {bone1Name=dyn_shoulderStraps_4_l, bone2Name=dyn_shoulderStraps_3_l, swing1LimitDegrees=10, swing2LimitDegrees=12, twistLimitDegrees=10}
 {bone1Name=dyn_shoulderStraps_5_l, bone2Name=dyn_shoulderStraps_4_l, swing1LimitDegrees=12, swing2LimitDegrees=18, twistLimitDegrees=10}
 {bone1Name=dyn_shoulderStraps_6_l, bone2Name=dyn_shoulderStraps_5_l, swing1LimitDegrees=15, swing2LimitDegrees=20, twistLimitDegrees=15}
 {bone1Name=dyn_shoulderStraps_7_l, bone2Name=dyn_shoulderStraps_6_l, swing1LimitDegrees=15, swing2LimitDegrees=20, twistLimitDegrees=15}
 {bone1Name=dyn_shoulderStraps_4_r, bone2Name=dyn_shoulderStraps_3_r, swing1LimitDegrees=10, swing2LimitDegrees=12, twistLimitDegrees=10}
 {bone1Name=dyn_shoulderStraps_5_r, bone2Name=dyn_shoulderStraps_4_r, swing1LimitDegrees=12, swing2LimitDegrees=18, twistLimitDegrees=10}
 {bone1Name=dyn_shoulderStraps_6_r, bone2Name=dyn_shoulderStraps_5_r, swing1LimitDegrees=15, swing2LimitDegrees=20, twistLimitDegrees=15}
 {bone1Name=dyn_shoulderStraps_7_r, bone2Name=dyn_shoulderStraps_6_r, swing1LimitDegrees=15, swing2LimitDegrees=20, twistLimitDegrees=15}
 {bone1Name=dyn_shoulderStraps_3_r, bone2Name=dyn_shoulderStraps_2_r, swing1LimitDegrees=10, swing2LimitDegrees=10, twistLimitDegrees=10}
 {bone1Name=dyn_shoulderStraps_3_l, bone2Name=dyn_shoulderStraps_2_l, swing1LimitDegrees=10, swing2LimitDegrees=10, twistLimitDegrees=10}
 {bone1Name=dyn_shoulderStraps_2_r, bone2Name=dyn_shoulderStraps_1_r, swing1LimitDegrees=10, swing2LimitDegrees=10, twistLimitDegrees=10}
 {bone1Name=dyn_shoulderStraps_1_r, bone2Name=base_shoulderStraps_1_r, swing1LimitDegrees=10, swing2LimitDegrees=10, twistLimitDegrees=5}
 {bone1Name=dyn_shoulderStraps_2_l, bone2Name=dyn_shoulderStraps_1_l, swing1LimitDegrees=10, swing2LimitDegrees=10, twistLimitDegrees=10}
 {bone1Name=dyn_shoulderStraps_1_l, bone2Name=base_shoulderStraps_1_l, swing1LimitDegrees=10, swing2LimitDegrees=10, twistLimitDegrees=5}
 {
  bone1Name=base_shoulderStraps_1_l
  bone2Name=spine_05
  swing1LimitDegrees=45
  swing1Motion=Free
  swing2LimitDegrees=45
  swing2Motion=Free
  twistLimitDegrees=45
  twistMotion=Free
 }
 {
  bone1Name=base_shoulderStraps_1_r
  bone2Name=spine_05
  swing1LimitDegrees=45
  swing1Motion=Free
  swing2LimitDegrees=45
  swing2Motion=Free
  twistLimitDegrees=45
  twistMotion=Free
 }
 {
  bone1Name=spine_03
  bone2Name=spine_02
  swing1LimitDegrees=45
  swing1Motion=Free
  swing2LimitDegrees=45
  swing2Motion=Free
  twistLimitDegrees=45
  twistMotion=Free
 }
 {
  bone1Name=spine_02
  bone2Name=spine_01
  swing1LimitDegrees=45
  swing1Motion=Free
  swing2LimitDegrees=45
  swing2Motion=Free
  twistLimitDegrees=45
  twistMotion=Free
 }
 {
  bone1Name=spine_01
  bone2Name=pelvis
  swing1LimitDegrees=45
  swing1Motion=Free
  swing2LimitDegrees=45
  swing2Motion=Free
  twistLimitDegrees=45
  twistMotion=Free
 }
 {
  bone1Name=spine_04
  bone2Name=spine_03
  swing1LimitDegrees=45
  swing1Motion=Free
  swing2LimitDegrees=45
  swing2Motion=Free
  twistLimitDegrees=45
  twistMotion=Free
 }
 {
  bone1Name=spine_05
  bone2Name=spine_04
  swing1LimitDegrees=45
  swing1Motion=Free
  swing2LimitDegrees=45
  swing2Motion=Free
  twistLimitDegrees=45
  twistMotion=Free
 }
 {
  bone1Name=spine_01
  bone2Name=dyn_shoulderStraps_7_l
  swing1LimitDegrees=45
  swing1Motion=Locked
  swing2LimitDegrees=45
  swing2Motion=Locked
  twistLimitDegrees=45
  twistMotion=Locked
 }
 {
  bone1Name=spine_01
  bone2Name=dyn_shoulderStraps_7_r
  swing1LimitDegrees=45
  swing1Motion=Locked
  swing2LimitDegrees=45
  swing2Motion=Locked
  twistLimitDegrees=45
  twistMotion=Locked
 }
 {bone1Name=dyn_shoulderStrapsTip_1_l, bone2Name=dyn_shoulderStraps_7_l, swing1LimitDegrees=15, swing2LimitDegrees=20, twistLimitDegrees=10}
 {bone1Name=dyn_shoulderStrapsTip_2_l, bone2Name=dyn_shoulderStrapsTip_1_l, swing1LimitDegrees=15, swing2LimitDegrees=20, twistLimitDegrees=10}
 {bone1Name=dyn_shoulderStrapsTip_2_r, bone2Name=dyn_shoulderStrapsTip_1_r, swing1LimitDegrees=15, swing2LimitDegrees=20, twistLimitDegrees=10}
 {bone1Name=dyn_shoulderStrapsTip_1_r, bone2Name=dyn_shoulderStraps_7_r, swing1LimitDegrees=15, swing2LimitDegrees=20, twistLimitDegrees=10}
]
