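# Physics asset for the mug on the large male ToughPack player body, apparently for secondary dangle motion against the character's bandolier.
# A capsule on dyn_mug is constrained to a tiny base_mug capsule, with narrow swing and twist limits keeping its range of motion tight.
# Four small capsules on the left-side coll_bandoliersTop bones form a separate, unconstrained set of bodies that act as collision proxies for the upper bandolier.
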PhysicsAsset /BRCosmetics/Characters/Player/Male/Large/Bodies/M_LRG_ToughPack/Meshes/M_LRG_ToughPack_mug_Physics
bodies {
 base_mug [{shapeType=Capsule, length=1, radius=0.5, shapeName=base_mug_capsule}]
 coll_bandoliersTop_1_l [{shapeType=Capsule, length=4.5284, radius=2.0141, shapeName=coll_bandoliersTop_1_l_capsule, rotation=<rot 0 -90 76.036>}]
 coll_bandoliersTop_2_l [{
   shapeType=Capsule
   length=4.5284
   radius=2.2065
   shapeName=coll_bandoliersTop_2_l_capsule
   center=<0.0321 0.2089 -0.129>
   rotation=<rot 0 -90 76.036>
  }]
 coll_bandoliersTop_3_l [{shapeType=Capsule, length=4.343, radius=2.1994, shapeName=coll_bandoliersTop_3_l_capsule, rotation=<rot 0 -89.9999 65.1599>}]
 coll_bandoliersTop_4_l [{
   shapeType=Capsule
   length=4.3547
   radius=2.1794
   shapeName=coll_bandoliersTop_4_l_capsule
   center=<-0.1122 0 -0.2892>
   rotation=<rot 0 -89.9999 111.2034>
  }]
 dyn_mug [{shapeType=Capsule, length=5.4542, radius=4.7284, shapeName=dyn_mug_capsule, center=<5.3397 -0.2148 0.4294>, rotation=<rot 0 -2.32 -89.7965>}]
}
constraints [{bone1Name=dyn_mug, bone2Name=base_mug, swing1LimitDegrees=20, swing2LimitDegrees=20, twistLimitDegrees=15}]
