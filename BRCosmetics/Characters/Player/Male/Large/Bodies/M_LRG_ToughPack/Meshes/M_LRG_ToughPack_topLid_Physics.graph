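# Physics asset for the top lid of the large male ToughPack body cosmetic, set up for secondary motion of the lid against the character's upper back.
# A box body on dyn_topLid hinges off a tiny base_topLid anchor capsule with narrow swing and twist limits and a zero second-swing limit.
# The anchor chains to spine_05 and spine_04 through angularly free constraints, with large capsules on both spine bones approximating the torso volume.
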PhysicsAsset /BRCosmetics/Characters/Player/Male/Large/Bodies/M_LRG_ToughPack/Meshes/M_LRG_ToughPack_topLid_Physics
bodies {
 base_topLid [{shapeType=Capsule, length=0.505, radius=0.505, shapeName=base_topLid_capsule, rotation=<rot 0 -90 0>}]
 dyn_topLid [{shapeType=Box, extentX=15.6259, extentY=3.3471, extentZ=35.0422, shapeName=dyn_topLid_box, center=<6.4633 1.0324 0>}]
 spine_04 [{shapeType=Capsule, length=32.3089, radius=14.1441, shapeName=spine_05_capsule, center=<1.1517 -3.9806 -0.349>, rotation=<rot 0 -90 0>}]
 spine_05 [{shapeType=Capsule, length=32.3089, radius=14.1441, shapeName=spine_05_capsule, center=<1.1517 -4.4213 -0.349>, rotation=<rot 0 -90 0>}]
}
constraints [
 {bone1Name=dyn_topLid, bone2Name=base_topLid, swing1LimitDegrees=20, swing2LimitDegrees=0, twistLimitDegrees=10}
 {
  bone1Name=base_topLid
  bone2Name=spine_05
  swing1LimitDegrees=45
  swing1Motion=Free
  swing2LimitDegrees=45
  swing2Motion=Free
  twistLimitDegrees=45
  twistMotion=Free
 }
 {
  bone1Name=spine_05
  bone2Name=spine_04
  swing1LimitDegrees=45
  swing1Motion=Free
  swing2LimitDegrees=45
  swing2Motion=Free
  twistLimitDegrees=45
  twistMotion=Free
 }
]
