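# Post-process-style cosmetic AnimBP for the large male ToughPack body, adding secondary motion to its pack, straps, bandoliers, zippers and thermos.
# Copied parent pose runs through Control Rig, bone fixes, AnimDynamics, weighted collider constraints, RigidBody sims and strap offsets, staged via cached poses.
# Back bling selects between two shoulder-strap simulations.
# The EventGraph tracks worn cosmetics, smooths targeting and melee values, and swaps skin weight profiles for shoes or low LOD without rigid bodies.
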AnimBlueprint /BRCosmetics/Characters/Player/Male/Large/Bodies/M_LRG_ToughPack/Meshes/M_LRG_ToughPack_AnimBP
target_skeleton=&/M_LRG_ToughPack_Skeleton
graphs [
 AnimGraph {nodes [
   Anim Dynamics - Bone: dyn_topLid!(ComponentPose, Alpha=1, GravityOverride:Vector="0,0,-980") -> (Pose) {
    angularDampingOverride=1
    angularSpringConstant=2000
    bAngularSpring=true
    bOverrideAngularDamping=true
    bUseGravityOverride=true
    gravityScale=0.35
    simulationSpace=RootRelative
    alphaBoolBlend {blendOption=Linear}
    boundBone {boneName=dyn_topLid}
    componentAppliedLinearAccClamp=<3000 3000 3000>
    componentLinearAccScale=<1 1 1>
    componentLinearVelScale=<1 1 1>
    gravityOverride {z=-980}
    physicsBodyDefinitions [{
      boundBone {boneName=dyn_topLid}
      boxExtents=<20 20 20>
      constraintSetup {angularLimitsMax=<0 0 20>, angularLimitsMin=<0 0 0>, angularTarget=<1 0 0>}
      localJointOffset=<5 0 0>
     }]
    planarLimits [{drivingBone {boneName=spine_05}, planeTransform {location=<0 9 0>, rotation=<rot -90 90 0>}}]
   }
   RigidBody!(ComponentPose, Alpha, OverrideWorldGravity:Vector="0,0,-980") -> (Pose) {
    bClampLinearTranslationLimitToRefPose=true
    bOverrideWorldGravity=true
    overridePhysicsAsset=&/M_LRG_ToughPack_Bandoliers_Physics
    simulationSpace=BaseBoneSpace
    alphaBoolBlend {blendOption=Linear}
    baseBoneRef {boneName=thigh_l}
    componentAppliedLinearAccClamp=<3000 3000 3000>
    simSpaceSettings {
     maxAngularAcceleration=1
     maxAngularVelocity=3
     maxLinearAcceleration=1000
     maxLinearVelocity=500
     worldAlpha=1
     externalLinearDragV=<0.2 0.2 0.2>
    }
   }
   RigidBody(OverrideWorldGravity="0,0,-980") {
    bClampLinearTranslationLimitToRefPose=true
    bOverrideWorldGravity=true
    overridePhysicsAsset=&/M_LRG_ToughPack_topLid_Physics
    simulationSpace=BaseBoneSpace
    alphaBoolBlend {blendOption=Linear}
    baseBoneRef {boneName=spine_05}
    componentAppliedLinearAccClamp=<3000 3000 3000>
    simSpaceSettings {
     maxAngularAcceleration=1
     maxAngularVelocity=3
     maxLinearAcceleration=1000
     maxLinearVelocity=500
     worldAlpha=1
     externalLinearDragV=<0.2 0.2 0.2>
    }
   }
   Output Pose(Result=Use cached pose 'postSimFixes'()) {layerGroup=None}
   n1 = Control Rig!(
    Source=Copy Pose From Mesh() {bUseAttachedParent=true}
    Alpha
    pub isWearingBackBling:bool=Get isWearingBackbling()
    pub isTargeting:double=Get isTargeting float() -> (isTargeting float:double)
    pub isWearingGauntlets:bool=Get isWearingGauntlets() -> (isWearingGauntlets:bool)
    pub isWearingShoes:bool=Get isWearingShoes() -> (isWearingShoes:bool)
    pub isFrontEnd:bool=Get bIsInFrontEnd()
   ) -> (Pose) {
    alphaBoolBlend {blendOption=Linear}
    controlRigAssetReference {blueprintRigClass=&/M_LRG_ToughPack_CtrlRig.M_LRG_ToughPack_CtrlRig_C}
   }
   Save cached pose 'controlRigCP'(Pose=@n1) {cacheName=controlRigCP}
   n2 = Anim Dynamics - Bone: dyn_backPackHandle!(ComponentPose=Local To Component(LocalPose=Use cached pose 'constrainsCollidersCP'()), Alpha=1, GravityOverride:Vector="0,0,980") -> (Pose) {
    angularDampingOverride=1
    angularSpringConstant=1000
    bAngularSpring=true
    bOverrideAngularDamping=true
    bUseGravityOverride=true
    gravityScale=0.35
    simulationSpace=RootRelative
    alphaBoolBlend {blendOption=Linear}
    boundBone {boneName=dyn_backPackHandle}
    componentAppliedLinearAccClamp=<3000 3000 3000>
    componentLinearAccScale=<0.2 0.2 0.5>
    componentLinearVelScale=<0.2 0.2 0.5>
    gravityOverride {z=980}
    physicsBodyDefinitions [{
      boundBone {boneName=dyn_backPackHandle}
      boxExtents=<10 10 10>
      constraintSetup {angularLimitsMax=<0 0 0>, angularLimitsMin=<0 0 -12>, angularTarget=<1 0 0>}
      localJointOffset=<1.5 0 0>
     }]
   }
   n3 = RigidBody(ComponentPose=@n2, OverrideWorldGravity="0,0,-980") {
    bClampLinearTranslationLimitToRefPose=true
    bOverrideWorldGravity=true
    overridePhysicsAsset=&/M_LRG_ToughPack_ShoulderStrapsBP_Physics
    simulationSpace=BaseBoneSpace
    alphaBoolBlend {blendOption=Linear}
    baseBoneRef {boneName=spine_02}
    componentAppliedLinearAccClamp=<5000 5000 5000>
    simSpaceSettings {maxAngularAcceleration=1, maxAngularVelocity=3, maxLinearAcceleration=1000, maxLinearVelocity=500, worldAlpha=1, externalLinearDragV=<1 1 1>}
   }
   n4 = Component To Local(ComponentPose=@n3)
   n5 = Anim Dynamics - Bone: dyn_backPackHandle(ComponentPose=Local To Component(LocalPose=Use cached pose 'constrainsCollidersCP'()), Alpha=1, GravityOverride="0,0,980") {
    angularDampingOverride=1
    angularSpringConstant=1000
    bAngularSpring=true
    bOverrideAngularDamping=true
    bUseGravityOverride=true
    gravityScale=0.35
    simulationSpace=RootRelative
    alphaBoolBlend {blendOption=Linear}
    boundBone {boneName=dyn_backPackHandle}
    componentAppliedLinearAccClamp=<3000 3000 3000>
    componentLinearAccScale=<0.2 0.2 0.5>
    componentLinearVelScale=<0.2 0.2 0.5>
    gravityOverride {z=980}
    physicsBodyDefinitions [{
      boundBone {boneName=dyn_backPackHandle}
      boxExtents=<10 10 10>
      constraintSetup {angularLimitsMax=<0 0 0>, angularLimitsMin=<0 0 -30>, angularTarget=<1 0 0>}
      localJointOffset=<1.5 0 0>
     }]
   }
   n6 = RigidBody(ComponentPose=@n5, OverrideWorldGravity="0,0,-980") {
    bClampLinearTranslationLimitToRefPose=true
    bOverrideWorldGravity=true
    overridePhysicsAsset=&/M_LRG_ToughPack_ShoulderStraps_Physics
    simulationSpace=BaseBoneSpace
    alphaBoolBlend {blendOption=Linear}
    alphaScaleBiasClamp {bInterpResult=true, interpSpeedDecreasing=20, interpSpeedIncreasing=20}
    baseBoneRef {boneName=spine_02}
    componentAppliedLinearAccClamp=<5000 5000 5000>
    simSpaceSettings {maxAngularAcceleration=1, maxAngularVelocity=3, maxLinearAcceleration=1000, maxLinearVelocity=500, worldAlpha=1, externalLinearDragV=<1 1 1>}
   }
   n7 = Component To Local(ComponentPose=@n6)
   n8 = Get isWearingBackbling()
   n9 = RigidBody(
    ComponentPose=Local To Component(LocalPose=Blend Poses by bool(BlendPose_0=@n4, BlendPose_1=@n7, bActiveValue=@n8) {blendType=Linear})
    OverrideWorldGravity="0,0,-980"
   ) {
    bClampLinearTranslationLimitToRefPose=true
    bOverrideWorldGravity=true
    overridePhysicsAsset=&/M_LRG_ToughPack_mug_Physics
    simulationSpace=BaseBoneSpace
    alphaBoolBlend {blendOption=Linear}
    baseBoneRef {boneName=spine_01}
    componentAppliedLinearAccClamp=<3000 3000 3000>
    simSpaceSettings {
     maxAngularAcceleration=1
     maxAngularVelocity=3
     maxLinearAcceleration=1500
     maxLinearVelocity=500
     worldAlpha=1
     externalLinearDragV=<0.2 0.2 0.2>
    }
   }
   Save cached pose 'dynamicsCP'(Pose=Component To Local(ComponentPose=@n9)) {cacheName=dynamicsCP}
   n10 = Anim Dynamics - Bone: dyn_calfOrn_l!(ComponentPose=Local To Component(LocalPose=Use cached pose 'generalAdjuistementsCP'()), Alpha=1, GravityOverride:Vector="0,0,-980") -> (Pose) {
    angularDampingOverride=1
    angularSpringConstant=10000
    bAngularSpring=true
    bOverrideAngularDamping=true
    bUseGravityOverride=true
    gravityScale=0.35
    simulationSpace=RootRelative
    alphaBoolBlend {blendOption=Linear}
    boundBone {boneName=dyn_calfOrn_l}
    componentAppliedLinearAccClamp=<3000 3000 3000>
    componentLinearAccScale=<1 1 1>
    componentLinearVelScale=<1 1 1>
    gravityOverride {z=-980}
    physicsBodyDefinitions [{
      boundBone {boneName=dyn_calfOrn_l}
      boxExtents=<20 30 30>
      constraintSetup {angularLimitsMax=<0 15 15>, angularLimitsMin=<0 -15 -15>, angularTarget=<1 0 0>}
      localJointOffset=<4 0 0>
     }]
   }
   n11 = Anim Dynamics - Bone: dyn_legsFitter_B_l!(ComponentPose=@n10, Alpha=1, GravityOverride:Vector="0,0,-980") -> (Pose) {
    angularDampingOverride=1
    angularSpringConstant=6000
    bAngularSpring=true
    bOverrideAngularDamping=true
    bUseGravityOverride=true
    gravityScale=0.35
    simulationSpace=RootRelative
    alphaBoolBlend {blendOption=Linear}
    boundBone {boneName=dyn_legsFitter_B_l}
    componentAppliedLinearAccClamp=<3000 3000 3000>
    componentLinearAccScale=<1 1 1>
    componentLinearVelScale=<1 1 1>
    gravityOverride {z=-980}
    physicsBodyDefinitions [{
      boundBone {boneName=dyn_legsFitter_B_l}
      boxExtents=<20 20 20>
      constraintSetup {angularLimitsMax=<15 15 15>, angularLimitsMin=<-15 -15 -15>, angularTarget=<1 0 0>}
      localJointOffset=<10 0 0>
     }]
   }
   n12 = Anim Dynamics - Bone: dyn_legsFitter_A_l!(ComponentPose=@n11, Alpha=1, GravityOverride:Vector="0,0,-980") -> (Pose) {
    angularDampingOverride=1
    angularSpringConstant=6000
    bAngularSpring=true
    bOverrideAngularDamping=true
    bUseGravityOverride=true
    gravityScale=0.35
    simulationSpace=RootRelative
    alphaBoolBlend {blendOption=Linear}
    boundBone {boneName=dyn_legsFitter_A_l}
    componentAppliedLinearAccClamp=<3000 3000 3000>
    componentLinearAccScale=<1 1 1>
    componentLinearVelScale=<1 1 1>
    gravityOverride {z=-980}
    physicsBodyDefinitions [{
      boundBone {boneName=dyn_legsFitter_A_l}
      boxExtents=<20 20 20>
      constraintSetup {angularLimitsMax=<15 15 15>, angularLimitsMin=<-15 -15 -15>, angularTarget=<1 0 0>}
      localJointOffset=<10 0 0>
     }]
   }
   n13 = Anim Dynamics - Bone: dyn_legsFitter_B_r!(ComponentPose=@n12, Alpha=1, GravityOverride:Vector="0,0,-980") -> (Pose) {
    angularDampingOverride=1
    angularSpringConstant=6000
    bAngularSpring=true
    bOverrideAngularDamping=true
    bUseGravityOverride=true
    gravityScale=0.35
    simulationSpace=RootRelative
    alphaBoolBlend {blendOption=Linear}
    boundBone {boneName=dyn_legsFitter_B_r}
    componentAppliedLinearAccClamp=<3000 3000 3000>
    componentLinearAccScale=<1 1 1>
    componentLinearVelScale=<1 1 1>
    gravityOverride {z=-980}
    physicsBodyDefinitions [{
      boundBone {boneName=dyn_legsFitter_B_r}
      boxExtents=<20 20 20>
      constraintSetup {angularLimitsMax=<15 15 15>, angularLimitsMin=<-15 -15 -15>, angularTarget=<1 0 0>}
      localJointOffset=<10 0 0>
     }]
   }
   n14 = Anim Dynamics - Bone: dyn_back_zipperSlider!(ComponentPose=@n13, Alpha=1, GravityOverride:Vector="0,0,-980") -> (Pose) {
    angularDampingOverride=1
    angularSpringConstant=8000
    bAngularSpring=true
    bOverrideAngularDamping=true
    bUseGravityOverride=true
    bUsePlanarLimit=false
    bUseSphericalLimits=true
    gravityScale=0.35
    simulationSpace=RootRelative
    alphaBoolBlend {blendOption=Linear}
    boundBone {boneName=dyn_back_zipperSlider}
    componentAppliedLinearAccClamp=<3000 3000 3000>
    componentLinearAccScale=<1 1 1>
    componentLinearVelScale=<1 1 1>
    gravityOverride {z=-980}
    physicsBodyDefinitions [{
      boundBone {boneName=dyn_back_zipperSlider}
      boxExtents=<20 30 30>
      constraintSetup {angularLimitsMax=<0 0 3>, angularLimitsMin=<0 -15 -20>, angularTarget=<1 0 0>}
      localJointOffset=<8 0 0>
     }]
    sphericalLimits [{limitRadius=15, drivingBone {boneName=spine_04}, sphereLocalOffset=<4.4536 2.3267 9.7177>}]
   }
   n15 = Anim Dynamics - Bone: dyn_zipperFront_A!(ComponentPose=@n14, Alpha=1, GravityOverride:Vector="0,0,-980") -> (Pose) {
    angularDampingOverride=1
    angularSpringConstant=8000
    bAngularSpring=true
    bOverrideAngularDamping=true
    bUseGravityOverride=true
    bUsePlanarLimit=false
    bUseSphericalLimits=true
    gravityScale=0.35
    simulationSpace=RootRelative
    alphaBoolBlend {blendOption=Linear}
    boundBone {boneName=dyn_zipperFront_A}
    componentAppliedLinearAccClamp=<3000 3000 3000>
    componentLinearAccScale=<1 1 1>
    componentLinearVelScale=<1 1 1>
    gravityOverride {z=-980}
    physicsBodyDefinitions [{
      boundBone {boneName=dyn_zipperFront_A}
      boxExtents=<20 30 30>
      constraintSetup {angularLimitsMax=<0 0 10>, angularLimitsMin=<0 0 -10>, angularTarget=<1 0 0>}
      localJointOffset=<8 0 0>
     }]
    sphericalLimits [{limitRadius=15, drivingBone {boneName=spine_04}, sphereLocalOffset=<4.4536 2.3267 9.7177>}]
   }
   n16 = Anim Dynamics - Bone: dyn_zipperFront_B!(ComponentPose=@n15, Alpha=1, GravityOverride:Vector="0,0,-980") -> (Pose) {
    angularDampingOverride=1
    angularSpringConstant=8000
    bAngularSpring=true
    bOverrideAngularDamping=true
    bUseGravityOverride=true
    bUsePlanarLimit=false
    bUseSphericalLimits=true
    gravityScale=0.35
    simulationSpace=RootRelative
    alphaBoolBlend {blendOption=Linear}
    boundBone {boneName=dyn_zipperFront_B}
    componentAppliedLinearAccClamp=<3000 3000 3000>
    componentLinearAccScale=<1 1 1>
    componentLinearVelScale=<1 1 1>
    gravityOverride {z=-980}
    physicsBodyDefinitions [{
      boundBone {boneName=dyn_zipperFront_B}
      boxExtents=<20 30 30>
      constraintSetup {angularLimitsMax=<0 0 20>, angularLimitsMin=<0 0 0>, angularTarget=<1 0 0>}
      localJointOffset=<8 0 0>
     }]
    sphericalLimits [{limitRadius=15, drivingBone {boneName=spine_04}, sphereLocalOffset=<4.4536 2.3267 9.7177>}]
   }
   n17 = Anim Dynamics - Bone: dyn_bandoliersBotMid_1_l!(ComponentPose=@n16, Alpha=1, GravityOverride:Vector="0,0,-980") -> (Pose) {
    angularDampingOverride=1
    angularSpringConstant=4000
    bAngularSpring=true
    bOverrideAngularDamping=true
    bUseGravityOverride=true
    gravityScale=0.35
    simulationSpace=RootRelative
    alphaBoolBlend {blendOption=Linear}
    boundBone {boneName=dyn_bandoliersBotMid_1_l}
    componentAppliedLinearAccClamp=<3000 3000 3000>
    componentLinearAccScale=<1 1 1>
    componentLinearVelScale=<1 1 1>
    gravityOverride {z=-980}
    physicsBodyDefinitions [{
      boundBone {boneName=dyn_bandoliersBotMid_1_l}
      boxExtents=<20 30 30>
      constraintSetup {angularLimitsMax=<0 0 0>, angularLimitsMin=<0 0 -50>, angularTarget=<1 0 0>}
      localJointOffset=<8 0 0>
     }]
   }
   n18 = Anim Dynamics - Bone: dyn_bandoliersBotMid_1_r!(ComponentPose=@n17, Alpha=1, GravityOverride:Vector="0,0,-980") -> (Pose) {
    angularDampingOverride=1
    angularSpringConstant=4000
    bAngularSpring=true
    bOverrideAngularDamping=true
    bUseGravityOverride=true
    gravityScale=0.35
    simulationSpace=RootRelative
    alphaBoolBlend {blendOption=Linear}
    boundBone {boneName=dyn_bandoliersBotMid_1_r}
    componentAppliedLinearAccClamp=<3000 3000 3000>
    componentLinearAccScale=<1 1 1>
    componentLinearVelScale=<1 1 1>
    gravityOverride {z=-980}
    physicsBodyDefinitions [{
      boundBone {boneName=dyn_bandoliersBotMid_1_r}
      boxExtents=<20 30 30>
      constraintSetup {angularLimitsMax=<0 0 0>, angularLimitsMin=<0 0 -50>, angularTarget=<1 0 0>}
      localJointOffset=<8 0 0>
     }]
   }
   n19 = Anim Dynamics - Bone: dyn_legsFitter_A_r!(ComponentPose=@n18, Alpha=1, GravityOverride:Vector="0,0,-980") -> (Pose) {
    angularDampingOverride=1
    angularSpringConstant=6000
    bAngularSpring=true
    bOverrideAngularDamping=true
    bUseGravityOverride=true
    gravityScale=0.35
    simulationSpace=RootRelative
    alphaBoolBlend {blendOption=Linear}
    boundBone {boneName=dyn_legsFitter_A_r}
    componentAppliedLinearAccClamp=<3000 3000 3000>
    componentLinearAccScale=<1 1 1>
    componentLinearVelScale=<1 1 1>
    gravityOverride {z=-980}
    physicsBodyDefinitions [{
      boundBone {boneName=dyn_legsFitter_A_r}
      boxExtents=<20 20 20>
      constraintSetup {angularLimitsMax=<15 15 15>, angularLimitsMin=<-15 -15 -15>, angularTarget=<1 0 0>}
      localJointOffset=<10 0 0>
     }]
   }
   n20 = Anim Dynamics - Bone: base_frontLogo!(ComponentPose=@n19, Alpha=1, GravityOverride:Vector="0,0,-980") -> (Pose) {
    angularDampingOverride=1
    angularSpringConstant=20000
    bAngularSpring=true
    bOverrideAngularDamping=true
    bUseGravityOverride=true
    gravityScale=0.35
    simulationSpace=RootRelative
    alphaBoolBlend {blendOption=Linear}
    boundBone {boneName=base_frontLogo}
    componentAppliedLinearAccClamp=<3000 3000 3000>
    componentLinearAccScale=<1 1 1>
    componentLinearVelScale=<1 1 1>
    gravityOverride {z=-980}
    physicsBodyDefinitions [{
      boundBone {boneName=base_frontLogo}
      boxExtents=<20 20 20>
      constraintSetup {angularLimitsMax=<0 0 25>, angularLimitsMin=<0 -10 -25>, angularTarget=<1 0 0>}
      localJointOffset=<10 0 0>
     }]
   }
   n21 = Anim Dynamics - Bone: dyn_bandoliersTop_l!(ComponentPose=@n20, Alpha=1, GravityOverride:Vector="0,0,-980") -> (Pose) {
    angularDampingOverride=1
    angularSpringConstant=4000
    bAngularSpring=true
    bOverrideAngularDamping=true
    bUseGravityOverride=true
    gravityScale=0.35
    simulationSpace=RootRelative
    alphaBoolBlend {blendOption=Linear}
    boundBone {boneName=dyn_bandoliersTop_l}
    componentAppliedLinearAccClamp=<3000 3000 3000>
    componentLinearAccScale=<0.5 0.5 0.5>
    componentLinearVelScale=<0.5 0.5 0.5>
    gravityOverride {z=-980}
    physicsBodyDefinitions [{
      boundBone {boneName=dyn_bandoliersTop_l}
      boxExtents=<20 30 30>
      constraintSetup {angularLimitsMax=<0 0 2>, angularLimitsMin=<0 -15 -2>, angularTarget=<1 0 0>}
      localJointOffset=<8 0 0>
     }]
   }
   n22 = Anim Dynamics - Bone: dyn_bandoliersTop_r!(ComponentPose=@n21, Alpha=1, GravityOverride:Vector="0,0,-980") -> (Pose) {
    angularDampingOverride=1
    angularSpringConstant=4000
    bAngularSpring=true
    bOverrideAngularDamping=true
    bUseGravityOverride=true
    gravityScale=0.35
    simulationSpace=RootRelative
    alphaBoolBlend {blendOption=Linear}
    boundBone {boneName=dyn_bandoliersTop_r}
    componentAppliedLinearAccClamp=<3000 3000 3000>
    componentLinearAccScale=<0.5 0.5 0.5>
    componentLinearVelScale=<0.5 0.5 0.5>
    gravityOverride {z=-980}
    physicsBodyDefinitions [{
      boundBone {boneName=dyn_bandoliersTop_r}
      boxExtents=<20 30 30>
      constraintSetup {angularLimitsMax=<0 0 2>, angularLimitsMin=<0 -15 -2>, angularTarget=<1 0 0>}
      localJointOffset=<8 0 0>
     }]
   }
   n23 = Anim Dynamics - Bone: dyn_thermos_rope_b!(SimulationSpace:AnimPhysSimSpaceType Enum, ComponentPose=@n22, Alpha=1, GravityOverride:Vector="0,0,-980") -> (Pose) {
    angularDampingOverride=1
    angularSpringConstant=6000
    bAngularSpring=true
    bOverrideAngularDamping=true
    bUseGravityOverride=true
    gravityScale=0.35
    simulationSpace=RootRelative
    alphaBoolBlend {blendOption=Linear}
    boundBone {boneName=dyn_thermos_rope_b}
    componentAppliedLinearAccClamp=<3000 3000 3000>
    componentLinearAccScale=<1 1 1>
    componentLinearVelScale=<1 1 1>
    gravityOverride {z=-980}
    physicsBodyDefinitions [{
      boundBone {boneName=dyn_thermos_rope_b}
      boxExtents=<20 30 30>
      constraintSetup {angularLimitsMax=<0 30 30>, angularLimitsMin=<0 -30 -30>, angularTarget=<1 0 0>}
      localJointOffset=<8 0 0>
     }]
   }
   n24 = Anim Dynamics - Bone: dyn_thermos_rope_c!(SimulationSpace:AnimPhysSimSpaceType Enum, ComponentPose=@n23, Alpha=1, GravityOverride:Vector="0,0,-980") -> (Pose) {
    angularDampingOverride=1
    angularSpringConstant=6000
    bAngularSpring=true
    bOverrideAngularDamping=true
    bUseGravityOverride=true
    gravityScale=0.35
    simulationSpace=RootRelative
    alphaBoolBlend {blendOption=Linear}
    boundBone {boneName=dyn_thermos_rope_c}
    componentAppliedLinearAccClamp=<3000 3000 3000>
    componentLinearAccScale=<1 1 1>
    componentLinearVelScale=<1 1 1>
    gravityOverride {z=-980}
    physicsBodyDefinitions [{
      boundBone {boneName=dyn_thermos_rope_c}
      boxExtents=<20 30 30>
      constraintSetup {angularLimitsMax=<0 30 30>, angularLimitsMin=<0 -30 -30>, angularTarget=<1 0 0>}
      localJointOffset=<8 0 0>
     }]
   }
   n25 = Anim Dynamics - Bone: dyn_thermos_rope_a!(SimulationSpace:AnimPhysSimSpaceType Enum, ComponentPose=@n24, Alpha=1, GravityOverride:Vector="0,0,-980") -> (Pose) {
    angularDampingOverride=1
    angularSpringConstant=6000
    bAngularSpring=true
    bOverrideAngularDamping=true
    bUseGravityOverride=true
    gravityScale=0.35
    simulationSpace=RootRelative
    alphaBoolBlend {blendOption=Linear}
    boundBone {boneName=dyn_thermos_rope_a}
    componentAppliedLinearAccClamp=<3000 3000 3000>
    componentLinearAccScale=<1 1 1>
    componentLinearVelScale=<1 1 1>
    gravityOverride {z=-980}
    physicsBodyDefinitions [{
      boundBone {boneName=dyn_thermos_rope_a}
      boxExtents=<20 30 30>
      constraintSetup {angularLimitsMax=<0 30 30>, angularLimitsMin=<0 -30 -30>, angularTarget=<1 0 0>}
      localJointOffset=<8 0 0>
     }]
   }
   Save cached pose 'animDynamicsCP'(Pose=Component To Local(ComponentPose=@n25)) {cacheName=animDynamicsCP}
   n26 = "Constraint - attach_cape (P)"(ComponentPose=Local To Component(LocalPose=Use cached pose 'presimFixesCP'()), Alpha=1, ConstraintWeights_0=1) {
    alphaBoolBlend {blendOption=Linear}
    boneToModify {boneName=attach_cape}
    constraintSetup [{targetBone {boneName=spine_05}}]
    constraintWeights [1]
   }
   n27 = "Transform (Modify) Bone - Bone: attach_cape"!(ComponentPose=@n26, bAlphaBoolEnabled:bool=True, Translation="0,-1,0", Rotation="0,0,16") -> (Pose) {
    alphaInputType=Bool
    rotationMode=BMM_Additive
    rotationSpace=BCS_BoneSpace
    translationMode=BMM_Additive
    translationSpace=BCS_BoneSpace
    alphaBoolBlend {blendInTime=0.5, blendOption=CircularOut, blendOutTime=0.5}
    boneToModify {boneName=attach_cape}
    rotation {roll=16}
    translation {y=-1}
   }
   n28 = "Transform (Modify) Bone - Bone: attach_cape"(ComponentPose=@n27, bAlphaBoolEnabled=True, Translation="0,-2,0", Rotation="0,0,20") {
    alphaInputType=Bool
    rotationMode=BMM_Additive
    rotationSpace=BCS_BoneSpace
    translationMode=BMM_Additive
    translationSpace=BCS_BoneSpace
    alphaBoolBlend {blendInTime=0.5, blendOption=CircularOut, blendOutTime=0.5}
    boneToModify {boneName=attach_cape}
    rotation {roll=20}
    translation {y=-2}
   }
   n29 = "Transform (Modify) Bone - Bone: attach_cape"(ComponentPose=@n28, bAlphaBoolEnabled=True, Translation="0,0,0", Rotation="0,0,10") {
    alphaInputType=Bool
    rotationMode=BMM_Additive
    rotationSpace=BCS_BoneSpace
    translationMode=BMM_Additive
    translationSpace=BCS_BoneSpace
    alphaBoolBlend {blendInTime=0.5, blendOption=Cubic, blendOutTime=0.5}
    boneToModify {boneName=attach_cape}
    rotation {roll=10}
   }
   Save cached pose 'generalAdjuistementsCP'(Pose=Component To Local(ComponentPose=@n29)) {cacheName=generalAdjuistementsCP}
   n30 = "Constraint - coll_bandoliersTop_1_l (PPPPPPPP)"!(
    ComponentPose=Local To Component(LocalPose=Use cached pose 'animDynamicsCP'())
    Alpha=1
    ConstraintWeights_0=0.6957
    ConstraintWeights_1:float=0.1055
    ConstraintWeights_2:float=0.0892
    ConstraintWeights_3:float=0.0868
    ConstraintWeights_4:float=0.0123
    ConstraintWeights_5:float=0.0085
    ConstraintWeights_6:float=0.0011
    ConstraintWeights_7:float=0.0009
   ) -> (Pose) {
    alphaBoolBlend {blendOption=Linear}
    boneToModify {boneName=coll_bandoliersTop_1_l}
    constraintSetup [
     {targetBone {boneName=kin_bandoliersTopFront_l}}
     {targetBone {boneName=deform_groin_l}}
     {targetBone {boneName=pelvis}}
     {targetBone {boneName=thigh_twist_01_l}}
     {targetBone {boneName=kin_bandoliersTopBack_l}}
     {targetBone {boneName=thigh_l}}
     {targetBone {boneName=deform_groin_r}}
     {targetBone {boneName=thigh_twist_01_r}}
    ]
    constraintWeights [1,1,1,1,1,1,1,1]
   }
   n31 = "Constraint - coll_bandoliersTop_2_l (PPPPPP)"!(
    ComponentPose=@n30
    Alpha=1
    ConstraintWeights_0=0.8283
    ConstraintWeights_1:float=0.1624
    ConstraintWeights_2:float=0.0052
    ConstraintWeights_3:float=0.0017
    ConstraintWeights_4:float=0.0013
    ConstraintWeights_5:float=0.0011
   ) -> (Pose) {
    alphaBoolBlend {blendOption=Linear}
    boneToModify {boneName=coll_bandoliersTop_2_l}
    constraintSetup [
     {targetBone {boneName=kin_bandoliersTopFront_l}}
     {targetBone {boneName=kin_bandoliersTopBack_l}}
     {targetBone {boneName=thigh_twist_01_l}}
     {targetBone {boneName=deform_groin_l}}
     {targetBone {boneName=pelvis}}
     {targetBone {boneName=thigh_l}}
    ]
    constraintWeights [1,1,1,1,1,1]
   }
   n32 = "Constraint - coll_bandoliersTop_3_l (PPP)"!(ComponentPose=@n31, Alpha=1, ConstraintWeights_0=0.524, ConstraintWeights_1:float=0.4759, ConstraintWeights_2:float=0.0001) -> (Pose) {
    alphaBoolBlend {blendOption=Linear}
    boneToModify {boneName=coll_bandoliersTop_3_l}
    constraintSetup [{targetBone {boneName=kin_bandoliersTopBack_l}}, {targetBone {boneName=kin_bandoliersTopFront_l}}, {targetBone {boneName=thigh_twist_01_l}}]
    constraintWeights [1,1,1]
   }
   n33 = "Constraint - coll_bandoliersTop_4_l (PPPPPPPP)"!(
    ComponentPose=@n32
    Alpha=1
    ConstraintWeights_0=0.7831
    ConstraintWeights_1:float=0.1499
    ConstraintWeights_2:float=0.0351
    ConstraintWeights_3:float=0.0164
    ConstraintWeights_4:float=0.0072
    ConstraintWeights_5:float=0.0069
    ConstraintWeights_6:float=0.0013
    ConstraintWeights_7:float=0.0001
   ) -> (Pose) {
    alphaBoolBlend {blendOption=Linear}
    boneToModify {boneName=coll_bandoliersTop_4_l}
    constraintSetup [
     {targetBone {boneName=kin_bandoliersTopBack_l}}
     {targetBone {boneName=kin_bandoliersTopFront_l}}
     {targetBone {boneName=thigh_twist_01_l}}
     {targetBone {boneName=pelvis}}
     {targetBone {boneName=thigh_l}}
     {targetBone {boneName=deform_glute_l}}
     {targetBone {boneName=deform_groin_l}}
     {targetBone {boneName=deform_glute_r}}
    ]
    constraintWeights [1,1,1,1,1,1,1,1]
   }
   Save cached pose 'constrainsCollidersCP'(Pose=Component To Local(ComponentPose=@n33)) {cacheName=constrainsCollidersCP}
   n34 = "Transform (Modify) Bone - Bone: base_shoulderStraps_1_l"!(ComponentPose=Local To Component(LocalPose=Use cached pose 'controlRigCP'()), Alpha=1, Rotation="0,0,0") -> (Pose) {
    rotationMode=BMM_Additive
    rotationSpace=BCS_BoneSpace
    translationMode=BMM_Additive
    translationSpace=BCS_BoneSpace
    alphaBoolBlend {blendOption=Linear}
    boneToModify {boneName=base_shoulderStraps_1_l}
   }
   n35 = "Transform (Modify) Bone - Bone: base_shoulderStraps_1_r"!(ComponentPose=@n34, Alpha=1, Rotation="0,0,0") -> (Pose) {
    rotationMode=BMM_Additive
    rotationSpace=BCS_BoneSpace
    translationMode=BMM_Additive
    translationSpace=BCS_BoneSpace
    alphaBoolBlend {blendOption=Linear}
    boneToModify {boneName=base_shoulderStraps_1_r}
   }
   n36 = "Transform (Modify) Bone - Bone: base_frontLogo"!(ComponentPose=@n35, bAlphaBoolEnabled:bool=True, Translation="0,0,1", Rotation="5,0,0") -> (Pose) {
    alphaInputType=Bool
    rotationMode=BMM_Additive
    rotationSpace=BCS_BoneSpace
    translationMode=BMM_Additive
    translationSpace=BCS_BoneSpace
    alphaBoolBlend {blendInTime=0.5, blendOption=Linear, blendOutTime=0.5}
    boneToModify {boneName=base_frontLogo}
    rotation {pitch=5}
    translation {z=1}
   }
   n37 = "Transform (Modify) Bone - Bone: kin_sidePack_retract"!(ComponentPose=@n36, bAlphaBoolEnabled:bool=True, Translation="0,0,3", Rotation="15,0,0") -> (Pose) {
    alphaInputType=Bool
    rotationMode=BMM_Additive
    rotationSpace=BCS_BoneSpace
    translationMode=BMM_Additive
    translationSpace=BCS_BoneSpace
    alphaBoolBlend {blendInTime=0.5, blendOption=Linear, blendOutTime=0.5}
    boneToModify {boneName=kin_sidePack_retract}
    rotation {pitch=15}
    translation {z=3}
   }
   Save cached pose 'presimFixesCP'(Pose=Component To Local(ComponentPose=@n37)) {cacheName=presimFixesCP}
   n38 = "Transform (Modify) Bone - Bone: dyn_shoulderStraps_4_l"!(ComponentPose=Local To Component(LocalPose=Use cached pose 'dynamicsCP'()), Alpha=1, Translation="1.5,0,0", Rotation="0,0,0") -> (Pose) {
    rotationMode=BMM_Additive
    rotationSpace=BCS_BoneSpace
    translationMode=BMM_Additive
    translationSpace=BCS_BoneSpace
    alphaBoolBlend {blendOption=Linear}
    boneToModify {boneName=dyn_shoulderStraps_4_l}
    translation {x=1.5}
   }
   n39 = "Transform (Modify) Bone - Bone: dyn_shoulderStraps_4_r"!(ComponentPose=@n38, Alpha=1, Translation="1.5,0,0", Rotation="0,0,0") -> (Pose) {
    rotationMode=BMM_Additive
    rotationSpace=BCS_BoneSpace
    translationMode=BMM_Additive
    translationSpace=BCS_BoneSpace
    alphaBoolBlend {blendOption=Linear}
    boneToModify {boneName=dyn_shoulderStraps_4_r}
    translation {x=1.5}
   }
   n40 = "Transform (Modify) Bone - Bone: dyn_shoulderStraps_5_l"!(ComponentPose=@n39, Alpha=1, Translation="1.5,0,0", Rotation="0,0,0") -> (Pose) {
    rotationMode=BMM_Additive
    rotationSpace=BCS_BoneSpace
    translationMode=BMM_Additive
    translationSpace=BCS_BoneSpace
    alphaBoolBlend {blendOption=Linear}
    boneToModify {boneName=dyn_shoulderStraps_5_l}
    translation {x=1.5}
   }
   n41 = "Transform (Modify) Bone - Bone: dyn_shoulderStraps_5_r"!(ComponentPose=@n40, Alpha=1, Translation="1.5,0,0", Rotation="0,0,0") -> (Pose) {
    rotationMode=BMM_Additive
    rotationSpace=BCS_BoneSpace
    translationMode=BMM_Additive
    translationSpace=BCS_BoneSpace
    alphaBoolBlend {blendOption=Linear}
    boneToModify {boneName=dyn_shoulderStraps_5_r}
    translation {x=1.5}
   }
   n42 = "Transform (Modify) Bone - Bone: dyn_shoulderStraps_6_l"!(ComponentPose=@n41, Alpha=1, Translation="1.5,0,0", Rotation="0,0,0") -> (Pose) {
    rotationMode=BMM_Additive
    rotationSpace=BCS_BoneSpace
    translationMode=BMM_Additive
    translationSpace=BCS_BoneSpace
    alphaBoolBlend {blendOption=Linear}
    boneToModify {boneName=dyn_shoulderStraps_6_l}
    translation {x=1.5}
   }
   n43 = "Transform (Modify) Bone - Bone: dyn_shoulderStraps_6_r"!(ComponentPose=@n42, Alpha=1, Translation="1.5,0,0", Rotation="0,0,0") -> (Pose) {
    rotationMode=BMM_Additive
    rotationSpace=BCS_BoneSpace
    translationMode=BMM_Additive
    translationSpace=BCS_BoneSpace
    alphaBoolBlend {blendOption=Linear}
    boneToModify {boneName=dyn_shoulderStraps_6_r}
    translation {x=1.5}
   }
   Save cached pose 'postSimFixes'(Pose=Component To Local(ComponentPose=@n43)) {cacheName=postSimFixes}
  ]}
 EventGraph {nodes [
   Event BlueprintInitializeAnimation()
   > Set OwnSkelMesh(OwnSkelMesh:Skeletal Mesh Component Object Reference=GetOwnerSkeletalMeshComponent()) -> (Output_Get:Skeletal Mesh Component Object Reference)
   Event OnParachuteOpened()
   > Switch on EGliderType(Selection=Get PlayerGliderType() -> (PlayerGliderType=HangGlider))
   > Set isUsingHangGlidingInParachute(isUsingHangGlidingInParachute:bool=true) -> (Output_Get:bool)
   Event OnParachuteClosed()
   > Set isUsingHangGlidingInParachute(isUsingHangGlidingInParachute=false)
   n1 = Event BlueprintUpdateAnimation()
   > Set Delta Time X(Delta Time X=@n1.DeltaTimeX)
   > n2 = Sequence!(execute) -> (then_0, then_1, then_2:Exec, then_3:Exec)
   > n3 = Gravity Override State Function DEPRECATED(CharacterPartAnimInstance=Self-Reference(), GameTimeInSeconds=GetGameTimeInSeconds(), GravityOverrideParams=Get Gravity Override Params())
   > Set GravityOverride(GravityOverride=@n3)
   n4 = NOT Boolean(A=Get isWearingBackbling())
   n5 = "Atan2 (Degrees)"!(Y:double="Clamp (Float)"(Value=Get PawnVelocityZ(), Min=-1000, Max=0), X:double=2800) -> (ReturnValue:double)
   n6 = Get bRigidBodyEnabled()
   n7 = Remap Float(Input=Get bIsPlayingMeleeAnim float() -> (bIsPlayingMeleeAnim float:double), OldRange=(X=0,Y=1), NewRange=(X=1,Y=0)) -> (Output=~)
   Set shoulderBase rotator(
    execute=@n2.then_2
    shoulderBase rotator:Rotator=MakeRotator(Pitch=float * float!(A=@n5, B=-1, C:double=@n7, D:bool=@n6, E:bool=@n4) -> (ReturnValue) {is=title K2Node_PromotableOperator})
   ) -> (Output_Get:Rotator)
   n8 = NOT Boolean(A=Get bRigidBodyEnabled())
   n9 = "integer >= integer"(A=Get CurrentLOD(), B=2)
   > n10 = Branch(Condition=OR Boolean(A=AND Boolean(A=@n8, B=@n9), B=Get isWearingShoes()))
   > n11 = Sequence!(execute) -> (then_0, then_1)
   > Do Once(Reset=@n13.then_0)
   n12 = GetPartSkeletalMeshComponent()
   > SetSkinWeightProfile!(execute, self:Skinned Mesh Component Object Reference=@n12, InProfileName:Name=Alternative, InLayer:ESkinWeightProfileLayer Enum=Secondary) -> (then, ReturnValue:bool)
   Do Once(execute=@n13.then_1, Reset=@n11.then_1)
   > ClearSkinWeightProfile!(execute, self:Skinned Mesh Component Object Reference=@n12, InLayer:ESkinWeightProfileLayer Enum=Secondary) -> (then)
   n13 = Sequence(execute=@n10.else)
   n14 = Sequence(execute=@n2.then_3)
   > Branch(Condition=Get bIsCrouching())
   > Set stretchShoulderStraps(stretchShoulderStraps:double=GetAlphaFromReferenceTransformDelta(
     SkeletalMeshComponent=Get OwnSkelMesh() -> (OwnSkelMesh:Skeletal Mesh Component Object Reference)
     BoneName=pelvis
     BaseName=spine_02
     InputRange=(X=-15,Y=-35)
     TwistAxis="1,0,0"
    )) -> (Output_Get:double)
   n15 = Branch(execute=@n14.then_1, Condition=Get bIsPlayingMeleeAnim())
   n16 = Get OwnSkelMesh()
   n17 = GetAlphaFromReferenceTransformDelta(
    SkeletalMeshComponent=@n16
    BoneName=pelvis
    BaseName=spine_05
    DeltaComponentType=QuaternionTwist
    InputRange=(X=-20,Y=-60)
    TwistAxis="1,0,0"
   )
   n18 = GetAlphaFromReferenceTransformDelta(
    SkeletalMeshComponent=@n16
    BoneName=pelvis
    BaseName=spine_05
    DeltaComponentType=QuaternionTwist
    InputRange=(X=20,Y=60)
    TwistAxis="1,0,0"
   )
   > Set shoulderStrapsDyn alpha(shoulderStrapsDyn alpha:double=Remap Float(Input=float + float(A=@n17, B=@n18), OldRange=(X=0,Y=1), NewRange=(X=1,Y=0.5), Clamp=true) -> (Output=~)) -> (Output_Get:double)
   Set shoulderStrapsDyn alpha(execute=@n15.else, shoulderStrapsDyn alpha=1)
   n19 = Get bIsPlayingMeleeAnim()
   n20 = Get Delta Time X()
   Set bIsPlayingMeleeAnim float(
    execute=@n2.then_1
    bIsPlayingMeleeAnim float:double=FInterpTo(Current=Get bIsPlayingMeleeAnim float(), Target="To Float (Boolean)"(InBool=@n19), DeltaTime=@n20, InterpSpeed=8)
   ) -> (Output_Get:double)
   n21 = Get bIsTargeting()
   > Set isTargeting float(isTargeting float:double=FInterpTo(Current=Get isTargeting float(), Target="To Float (Boolean)"(InBool=@n21), DeltaTime=@n20, InterpSpeed=15)) -> (Output_Get:double)
   n22 = Event OnLeftArmHidden() {is=title K2Node_Event}
   Set isWearingGauntlets(execute=[@n22.then, @n25.then], isWearingGauntlets:bool=true) -> (Output_Get:bool)
   n23 = Event OnLeftArmVisible() {is=title K2Node_Event}
   Set isWearingGauntlets(execute=[@n23.then, @n24.then], isWearingGauntlets=false)
   n24 = Event OnRightArmVisible() {is=title K2Node_Event}
   n25 = Event OnRightArmHidden() {is=title K2Node_Event}
   n26 = Event OnCharacterCustomizationLoaded()
   > n27 = Sequence()
   n28 = CheckForBackbling()
   > Set isWearingBackbling(isWearingBackbling=OR Boolean!(A=@n28, B=@n28.bIsBackblingACape, C:bool=@n28.bIsBackblingLong) -> (ReturnValue))
   > ResetDynamics(InTeleportType=TeleportPhysics)
   > ResetDynamics(InTeleportType=ResetPhysics)
   > n29 = Branch()
   > Set simSpaceAnimDyn(simSpaceAnimDyn:AnimPhysSimSpaceType Enum=World) -> (Output_Get:AnimPhysSimSpaceType Enum)
   Set simSpaceAnimDyn(execute=@n29.else)
   Set ShoeSkelMesh(
    execute=@n27.then_1
    ShoeSkelMesh:Skeletal Mesh Component Budgeted Object Reference=Get Array Item with EFort Custom Part Type(Array=@n26.CustomMeshParts, PartType=ExtraPart)
   ) -> (Output_Get:Skeletal Mesh Component Budgeted Object Reference)
   > n30 = Get!(execute:Exec) -> (then:Exec, else:Exec, ShoeSkelMesh:Skeletal Mesh Component Budgeted Object Reference)
   > n31 = Is Valid(InputObject=Get SkeletalMesh!(self:Skinned Mesh Component Object Reference=@n30) -> (SkeletalMesh:Skeletal Mesh Object Reference))
   Set isWearingShoes(execute=[@n31.Is Not Valid, @n30.else], isWearingShoes:bool=false) -> (Output_Get:bool)
   Set isWearingShoes(execute=@n31.Is Valid, isWearingShoes=true)
  ]}
]
variables [
 bIsPlayingMeleeAnim float {type=number}
 Delta Time X {type=number, default=0.0167}
 Gravity Override Params {type=GravityOverrideParamsStruct}
 GravityOverride {type=Vector}
 isTargeting float {type=number}
 isUsingHangGlidingInParachute {type=boolean}
 isWearingBackbling {type=boolean}
 isWearingGauntlets {type=boolean}
 isWearingShoes {type=boolean}
 OwnSkelMesh {type=SkeletalMeshComponent}
 ShoeSkelMesh {type=SkeletalMeshComponentBudgeted}
 shoulderBase rotator {type=Rotator}
 shoulderStrapsDyn alpha {type=number, default=1}
 simSpaceAnimDyn {type=AnimPhysSimSpaceType}
 stretchShoulderStraps {type=number}
]
