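# Physics asset of cape colliders for the large male ToughPack player body, evidently the collision volume that keeps a simulated cape off the character.
# Overlapping capsules run along the pelvis and five spine bones, with a mirrored rotated pair added at spine_02 and a second capsule at spine_05.
# Arm, leg and foot bodies carry no shapes, and no constraints exist, so it describes torso collision only, not a jointed ragdoll.
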PhysicsAsset /BRCosmetics/Characters/Player/Male/Large/Bodies/M_LRG_ToughPack/Meshes/M_LRG_ToughPack_CapeColliders
bodies {
 calf_l []
 calf_r []
 foot_l []
 foot_r []
 lowerarm_l []
 lowerarm_r []
 pelvis [{shapeType=Capsule, length=28.6376, radius=14.0488, center=<-2.4887 0.1881 0>}]
 spine_01 [{shapeType=Capsule, length=23.5831, radius=14.2784, center=<-1.7965 -0.943 0>}]
 spine_02 [
  {shapeType=Capsule, length=31.8003, radius=14.6283, center=<-1.711 -1.9809 0.0001>}
  {shapeType=Capsule, length=29.3956, radius=11.5807, center=<2.5024 0.4992 3.9825>, rotation=<rot 90 0.0049 0.0049>}
  {shapeType=Capsule, length=29.875, radius=11.1013, center=<2.4704 0.4991 -3.9825>, rotation=<rot 90 -0.0396 -0.0396>}
 ]
 spine_03 [{shapeType=Capsule, length=30.9667, radius=14.5206, center=<0.1445 -1.6678 0.0001>}]
 spine_04 [{shapeType=Capsule, length=24.3048, radius=14.9057, center=<2.766 -0.5618 0.0002>}]
 spine_05 [
  {shapeType=Capsule, length=18.0875, radius=17.3701, center=<-0.5477 2.347 0>}
  {shapeType=Capsule, length=13.9518, radius=13.9425, center=<-0.2916 0.4452 0>}
 ]
 thigh_l []
 thigh_r []
 upperarm_l []
 upperarm_r []
}
constraints []
